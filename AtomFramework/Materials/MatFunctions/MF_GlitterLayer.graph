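# Material function that produces a glitter fleck normal and mask for a surface material from the Glitter_NM texture.
# Two WorldAlignedTexture samples use object-relative position, with sizes scaled from Brick Scale in MPC_BrickData, the second smaller.
# Their blended red and green channels rebuild a normal flattened by the Fleck Normal parameter, while blended blue gives the mask.
# An alternative UV2 branch with two bump-offset parallax layers feeds neither output.
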MaterialFunction /AtomFramework/Materials/MatFunctions/MF_GlitterLayer
graph {nodes [
  n1 = TextureCoordinate() {coordinateIndex=2, uTiling=10, vTiling=10}
  n2 = Constant()
  n3 = TextureSample(UVs=BumpOffset(Coordinate=Add(A=@n1) {constB=0.6}, Height=@n2, HeightRatioInput=Constant() {r=0.09}) {heightRatio=0.5, referencePlane=1}) {samplerSource=SSM_Wrap_WorldGroupSettings, samplerType=SAMPLERTYPE_LinearColor, texture=/AtomFramework/Materials/Textures/Glitter_NM}
  n4 = TextureCoordinate() {coordinateIndex=2, uTiling=16, vTiling=16}
  n5 = TextureSample(UVs=BumpOffset(Coordinate=Add(A=@n4) {constB=0.25}, Height=@n2, HeightRatioInput=Constant() {r=0.05}) {heightRatio=0.5, referencePlane=1}) {samplerSource=SSM_Wrap_WorldGroupSettings, samplerType=SAMPLERTYPE_LinearColor, texture=&Glitter_NM}
  MaterialExpressionAppend3Vector_0 = MaterialXAppend3Vector!(
   A=ConstantBiasScale(None=LinearInterpolate(A=@n3.R, B=@n5.R, Alpha=@n5.R)) {bias=-0.5, scale=2}
   B=ConstantBiasScale(None=LinearInterpolate(A=@n3.G, B=@n5.G, Alpha=@n5.G)) {bias=-0.5, scale=2}
   C=Constant() {r=1}
  )
  n6 = TextureObject() {samplerType=SAMPLERTYPE_LinearColor, texture=&Glitter_NM}
  n7 = Constant() {r=0.75}
  n8 = Multiply(A=CollectionParameter() {collection=/AtomFramework/Materials/Tools/MPC_BrickData, parameterName=Brick Scale}, B=@n7)
  n9 = Subtract(A=ObjectPositionWS(), B=WorldPosition().XYZ)
  n10 = MaterialFunctionCall(Float3=MaterialFunctionCall(TextureObject=@n6, TextureSize=@n8, WorldPosition=@n9) {materialFunction=&WorldAlignedTexture}.XYZ Texture) {materialFunction=&BreakOutFloat3Components}
  n11 = MaterialFunctionCall(Float3=MaterialFunctionCall(TextureObject=@n6, TextureSize=Multiply(A=@n8, B=@n7), WorldPosition=@n9) {materialFunction=&WorldAlignedTexture}.XYZ Texture) {materialFunction=&BreakOutFloat3Components}
  MaterialExpressionAppend3Vector_1 = MaterialXAppend3Vector(
   A=ConstantBiasScale(None=LinearInterpolate(A=@n10.R, B=@n11.R)) {bias=-0.5, scale=2}
   B=ConstantBiasScale(None=LinearInterpolate(A=@n10.G, B=@n11.G)) {bias=-0.5, scale=2}
   C=Constant() {r=1}
  )
  FunctionOutput(None=MaterialFunctionCall(Normal=@MaterialExpressionAppend3Vector_1, Flatness=ScalarParameter() {parameterName=Fleck Normal}) {materialFunction=&FlattenNormal}.Result) {outputName=Glitter Normal}
  FunctionOutput(None=LinearInterpolate(A=@n10.B, B=@n11.B)) {outputName=Glitter Mask, sortPriority=1}
  LinearInterpolate(A=@n5.B, B=@n3.B)
 ]}
