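# Material function that adds a glitter layer and dithered see-through masking to incoming material attributes, toggled by the Dither Glitter static switch.
# MF_GlitterLayer supplies a normal and mask that blend in a metallic glitter surface whose color follows a two-sided facing factor, invertible via Flip Facing.
# That factor also selects, per side, between full glitter with solid opacity mask and reduced glitter with DitherTemporalAA added, faking translucency.
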MaterialFunction /AtomFramework/Materials/MatFunctions/MF_DitherGlitter
graph {nodes [
  n1 = FunctionInput(Preview=GetMaterialAttributes().MaterialAttributes) {bUsePreviewValueAsDefault=true, inputType=FunctionInput_MaterialAttributes}
  n2 = Saturate(None=TwoSidedSign())
  n3 = LinearInterpolate(A=OneMinus(None=@n2), B=@n2, Alpha=ScalarParameter() {defaultValue=1, group=Dither Translucency, parameterName=Flip Facing})
  n4 = LinearInterpolate(Alpha=@n3)
  n5 = Constant() {r=1}
  n6 = MaterialFunctionCall() {materialFunction=&/MF_GlitterLayer}
  n7 = SetMaterialAttributes!(
   MaterialAttributes
   BaseColor=@n4
   SubsurfaceColor=@n4
   Metallic=@n5
   Specular=@n5
   Roughness=Constant() {r=0.3}
   Opacity=Constant()
   Normal=@n6.Glitter Normal
  ) {attributeSetTypes [
    #1
    #2
    #3
    #4
    #5
    #6
    #7
   ]}
  n8 = LinearInterpolate(A=@n6.Glitter Mask, B=Saturate(None=Subtract(A=@n6.Glitter Mask) {constB=0.2}), Alpha=@n3)
  n9 = BlendMaterialAttributes(A=@n1, B=@n7, Alpha=@n8) {vertexAttributeBlendType=UseA}
  n10 = MaterialFunctionCall() {materialFunction=&DitherTemporalAA}
  n11 = LinearInterpolate(A=Constant() {r=1}, B=Saturate(None=Add(A=@n8, B=@n10.Result)), Alpha=@n3)
  FunctionOutput(None=StaticSwitchParameter(True=SetMaterialAttributes!(MaterialAttributes=@n9, OpacityMask=@n11) {attributeSetTypes [#8]}, False=@n1) {group=Dither Translucency, parameterName=Dither Glitter})
 ]}
