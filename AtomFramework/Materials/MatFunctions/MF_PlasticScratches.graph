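# Material function that outputs a scratched-plastic roughness value from a surface mask texture.
# WorldAlignedTexture projects the mask in XYZ using position and normal inputs that default to instance space, so by default the pattern follows each instance.
# A Seed input, scaled up, offsets the position to vary the pattern.
# The mask's red channel blends between the low and high roughness parameters, with texture size also exposed.
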MaterialFunction /AtomFramework/Materials/MatFunctions/MF_PlasticScratches
graph {nodes [
  n1 = TransformPosition(None=WorldPosition().XYZ) {transformSourceType=TRANSFORMPOSSOURCE_World, transformType=TRANSFORMPOSSOURCE_Instance}
  n2 = FunctionInput(Preview=Constant()) {bUsePreviewValueAsDefault=true, inputName=Seed}
  n3 = VertexNormalWS()
  n4 = MaterialFunctionCall(
   TextureObject=TextureObject() {samplerType=SAMPLERTYPE_Masks, texture=/AtomFramework/Materials/Textures/Primitive_Surface_R}
   TextureSize=ScalarParameter() {defaultValue=500, group=Surface, parameterName=ScratchTextureSize}
   WorldPosition=Add(A=FunctionInput(Preview=@n1) {bUsePreviewValueAsDefault=true, inputName=Position, sortPriority=-7}, B=Multiply(A=@n2) {constB=1000})
   World Space Normal=FunctionInput(Preview=Transform(None=@n3) {transformSourceType=TRANSFORMSOURCE_World, transformType=TRANSFORM_Instance}) {bUsePreviewValueAsDefault=true, inputName=NormalWS, sortPriority=-7}
  ) {materialFunction=&WorldAlignedTexture}
  FunctionOutput(None=LinearInterpolate(
    A=ScalarParameter() {defaultValue=0.2, group=Surface, parameterName=RoughnessLow, sliderMax=1}
    B=ScalarParameter() {defaultValue=0.3, group=Surface, parameterName=RoughnessHigh, sliderMax=1}
    Alpha=ComponentMask(None=@n4.XYZ Texture) {r=true}
   ))
 ]}
