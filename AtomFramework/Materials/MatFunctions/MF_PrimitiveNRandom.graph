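# MF_PrimitiveNRandom is a material function that outputs a seed-randomized tangent-space normal for primitives or instances.
# It rotates the tangent up vector in instance space about a seed-derived axis, dividing the angle by local bounds size (via a Custom node) so larger primitives tilt less.
# The result returns to tangent space and gains a seed-driven XY jitter, scaled by the NormalRandom parameter, with unit Z added.
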MaterialFunction /AtomFramework/Materials/MatFunctions/MF_PrimitiveNRandom
graph {nodes [
  n1 = Transform(None=Constant3Vector() {constant=<rgba 0 0 1 0>}) {transformType=TRANSFORM_Instance}
  n2 = FunctionInput(Preview=Constant()) {bUsePreviewValueAsDefault=true, inputName=Seed}
  n3 = Subtract(A=@n2, B=Constant3Vector() {constant=<rgba 0.5 0.5 0.5 0>})
  n4 = Divide(
   A=FunctionInput(Preview=Constant() {r=0.05}) {bUsePreviewValueAsDefault=true, inputName=RotationAngle, inputType=FunctionInput_Scalar}
   B=Multiply(A=Custom!(None) {code="GetPrimitiveData(Parameters).InstanceLocalBoundsExtent.xyz", description=BoundsExtent}) {constB=2}
  )
  n5 = Subtract(A=ComponentMask(None=@n2) {b=true}) {constB=-0.5}
  n6 = Add(
   A=@n1
   B=RotateAboutAxis(NormalizedRotationAxis=Normalize(VectorInput=@n3), RotationAngle=Multiply(A=@n4, B=@n5), PivotPoint=Constant3Vector(), Position=@n1)
  )
  n7 = Multiply(
   A=Subtract(A=ComponentMask(None=@n2) {g=true, r=true}, B=Constant2Vector() {g=0.5, r=0.5})
   B=ScalarParameter() {defaultValue=0.08, group=Normal, parameterName=NormalRandom}
  ) {constB=0.05}
  n8 = Constant() {r=1}
  FunctionOutput(None=Add(A=Transform(None=@n6) {transformSourceType=TRANSFORMSOURCE_Instance, transformType=TRANSFORM_Tangent}, B=AppendVector(A=@n7, B=@n8)))
 ]}
