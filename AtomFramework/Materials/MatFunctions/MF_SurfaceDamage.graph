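# Material function that outputs a surface-damage normal combining large dents with a subtle warble.
# It projects a dents normal map and tiling noise with WorldAlignedTexture, using instance-space position and normal defaults, with seed-scaled offsets to vary placement.
# Dents convert from instance to tangent space and scale by NormalStrength.
# Centred noise red/green, scaled by a seed-interpolated warble intensity, blends in via BlendAngleCorrectedNormals.
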MaterialFunction /AtomFramework/Materials/MatFunctions/MF_SurfaceDamage
graph {nodes [
  n1 = FunctionInput(Preview=Constant() {r=0.15}) {bUsePreviewValueAsDefault=true, inputName=NormalStrength, inputType=FunctionInput_Scalar}
  n2 = MaterialFunctionCall(X=@n1, Y=@n1, Z=Constant() {r=1}) {materialFunction=&MakeFloat3}
  n3 = NamedRerouteUsage()
  n4 = Multiply(A=NamedRerouteUsage()) {constB=1000}
  n5 = Transform(None=MaterialFunctionCall(
    TextureObject=TextureObjectParameter() {parameterName=DentsLarge, samplerType=SAMPLERTYPE_Normal, texture=/AtomFramework/Materials/Textures/T_dents_N}
    TextureSize=ScalarParameter() {defaultValue=150, parameterName=DentsTextureSize}
    WorldPosition=Add(A=@n3, B=@n4)
    World Space Normal=NamedRerouteUsage()
   ) {materialFunction=&WorldAlignedTexture}.XYZ Texture) {transformSourceType=TRANSFORMSOURCE_Instance, transformType=TRANSFORM_Tangent}
  n6 = NamedRerouteUsage()
  n7 = MaterialFunctionCall(
   TextureObject=TextureObject() {texture=/Engine/MapTemplates/TilingNoise05}
   TextureSize=ScalarParameter() {defaultValue=150, parameterName=WarbleTextureSize}
   WorldPosition=Add(A=NamedRerouteUsage(), B=Multiply(A=@n6) {constB=1000})
   World Space Normal=NamedRerouteUsage()
  ) {materialFunction=&WorldAlignedTexture}
  n8 = Multiply(
   A=Subtract(A=ComponentMask(None=@n7.XYZ Texture) {g=true, r=true}) {constB=0.5}
   B=LinearInterpolate(
    A=ScalarParameter() {defaultValue=0.02, parameterName=WarbleIntensity0}
    B=ScalarParameter() {defaultValue=0.04, parameterName=WarbleIntensity1}
    Alpha=ComponentMask(None=@n6) {b=true}
   )
  )
  n9 = Constant() {r=1}
  FunctionOutput(None=MaterialFunctionCall(BaseNormal=Multiply(A=@n2.Result, B=@n5), AdditionalNormal=AppendVector(A=@n8, B=@n9)) {materialFunction=&BlendAngleCorrectedNormals}.Result)
  MaterialFunctionCall()
  n10 = VertexNormalWS()
  NamedRerouteDeclaration(None=FunctionInput(Preview=Transform(None=@n10) {transformSourceType=TRANSFORMSOURCE_World, transformType=TRANSFORM_Instance}) {bUsePreviewValueAsDefault=true, inputName=NormalWS, sortPriority=-7}) {name=PixelN, nodeColor=<rgba 0.3828 1 0 1>}
  n11 = WorldPosition()
  NamedRerouteDeclaration(None=FunctionInput(Preview=TransformPosition(None=@n11.XYZ) {transformSourceType=TRANSFORMPOSSOURCE_World, transformType=TRANSFORMPOSSOURCE_Instance}) {bUsePreviewValueAsDefault=true, inputName=PositionWS, sortPriority=-7}) {name=InstPos, nodeColor=<rgba 0.9688 1 0 1>}
  NamedRerouteDeclaration(None=FunctionInput(Preview=Constant()) {bUsePreviewValueAsDefault=true, inputName=Seed}) {name=Seed, nodeColor=<rgba 0 0.5547 1 1>}
 ]}
