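# Material function that projects a texture triplanar in world space with randomized per-tile variation to break up repetition.
# World position divided by a scale input yields three planar UV sets, blended by sharpened world-normal masks.
# Per projection, a sine-dot hash of UV cells drives rotation and slight scaling of a second sample.
# An edge gradient fades that sample back to the unmodified one near tile borders, hiding seams.
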MaterialFunction /AtomFramework/Materials/MatFunctions/MF_WorldAlignedVariation
graph {nodes [
  n1 = NamedRerouteUsage()
  n2 = NamedRerouteUsage()
  n3 = TextureSample(UVs=@n1, Tex=@n2) {samplerSource=SSM_Wrap_WorldGroupSettings, samplerType=SAMPLERTYPE_LinearColor, texture=/AtomFramework/Materials/Textures/OpalTest_N}
  n4 = Frac(None=@n1)
  n5 = Floor(None=ScalarParameter() {defaultValue=1, parameterName=Bands})
  n6 = Multiply(A=@n1, B=@n5) {constB=1024}
  n7 = DotProduct(A=Divide(A=Floor(None=@n6), B=@n5) {constB=1024}, B=Constant2Vector() {g=78.233, r=12.9898})
  n8 = Frac(None=Multiply(A=Sine(None=@n7), B=Add() {constA=43758.5469}) {constB=43758.5469})
  n9 = Multiply(A=@n8) {constB=0.2}
  n10 = TextureSample(
   UVs=Multiply(
    A=MaterialFunctionCall(UVs=@n4, "Rotation Angle (0-1)"=@n8) {materialFunction=&CustomRotator}.Rotated Values
    B=Add(A=@n9) {constB=1.1}
   )
   Tex=@n2
  ) {samplerSource=SSM_Wrap_WorldGroupSettings, samplerType=SAMPLERTYPE_LinearColor, texture=&OpalTest_N}
  n11 = MaterialFunctionCall(Float2=Abs(None=ConstantBiasScale(None=@n4) {bias=-0.5, scale=2})) {materialFunction=&BreakOutFloat2Components}
  n12 = Constant() {r=0.744}
  n13 = Constant() {r=1}
  n14 = Constant()
  n15 = AppendVector(
   A=Saturate(None=MaterialFunctionCall(Input=@n11.R, Input Low=@n12, Input High=@n13, Target Low=@n14, Target High=@n13) {materialFunction=&RemapValueRange}.Result)
   B=Saturate(None=MaterialFunctionCall(Input=@n11.G, Input Low=@n12, Input High=@n13, Target Low=@n14, Target High=@n13) {materialFunction=&RemapValueRange}.Result)
  )
  n16 = Constant()
  n17 = Power(Base=OneMinus(None=Distance(A=@n15, B=@n16))) {constExponent=6}
  n18 = LinearInterpolate(A=@n3.RGB, B=@n10.RGB, Alpha=NamedRerouteDeclaration(None=Saturate(None=@n17)) {name=Gradient, nodeColor=<rgba 1 0.3516 0 1>})
  n19 = NamedRerouteUsage()
  n20 = NamedRerouteUsage()
  n21 = TextureSample(UVs=@n19, Tex=@n20) {samplerSource=SSM_Wrap_WorldGroupSettings, samplerType=SAMPLERTYPE_LinearColor, texture=&OpalTest_N}
  n22 = Frac(None=@n19)
  n23 = Floor(None=ScalarParameter() {defaultValue=1, parameterName=Bands})
  n24 = Multiply(A=@n19, B=@n23) {constB=1024}
  n25 = DotProduct(A=Divide(A=Floor(None=@n24), B=@n23) {constB=1024}, B=Constant2Vector() {g=78.233, r=12.9898})
  n26 = Frac(None=Multiply(A=Sine(None=@n25), B=Add() {constA=43758.5469}) {constB=43758.5469})
  n27 = Multiply(A=@n26) {constB=0.2}
  n28 = TextureSample(
   UVs=Multiply(
    A=MaterialFunctionCall(UVs=@n22, "Rotation Angle (0-1)"=@n26) {materialFunction=&CustomRotator}.Rotated Values
    B=Add(A=@n27) {constB=1.1}
   )
   Tex=@n20
  ) {samplerSource=SSM_Wrap_WorldGroupSettings, samplerType=SAMPLERTYPE_LinearColor, texture=&OpalTest_N}
  n29 = MaterialFunctionCall(Float2=Abs(None=ConstantBiasScale(None=@n22) {bias=-0.5, scale=2})) {materialFunction=&BreakOutFloat2Components}
  n30 = Constant() {r=0.744}
  n31 = Constant() {r=1}
  n32 = Constant()
  n33 = AppendVector(
   A=Saturate(None=MaterialFunctionCall(Input=@n29.R, Input Low=@n30, Input High=@n31, Target Low=@n32, Target High=@n31) {materialFunction=&RemapValueRange}.Result)
   B=Saturate(None=MaterialFunctionCall(Input=@n29.G, Input Low=@n30, Input High=@n31, Target Low=@n32, Target High=@n31) {materialFunction=&RemapValueRange}.Result)
  )
  n34 = Constant()
  n35 = Power(Base=OneMinus(None=Distance(A=@n33, B=@n34))) {constExponent=6}
  n36 = LinearInterpolate(A=@n21.RGB, B=@n28.RGB, Alpha=NamedRerouteDeclaration(None=Saturate(None=@n35)) {name=Gradient_1, nodeColor=<rgba 1 0.3516 0 1>})
  n37 = Abs(None=ComponentMask(None=VertexNormalWS()) {r=true})
  n38 = Saturate(None=Power(Base=Multiply(A=@n37) {constB=1.5}) {constExponent=20})
  n39 = NamedRerouteUsage()
  n40 = NamedRerouteUsage()
  n41 = TextureSample(UVs=@n39, Tex=@n40) {samplerSource=SSM_Wrap_WorldGroupSettings, samplerType=SAMPLERTYPE_LinearColor, texture=&OpalTest_N}
  n42 = Frac(None=@n39)
  n43 = Floor(None=ScalarParameter() {defaultValue=1, parameterName=Bands})
  n44 = Multiply(A=@n39, B=@n43) {constB=1024}
  n45 = DotProduct(A=Divide(A=Floor(None=@n44), B=@n43) {constB=1024}, B=Constant2Vector() {g=78.233, r=12.9898})
  n46 = Frac(None=Multiply(A=Sine(None=@n45), B=Add() {constA=43758.5469}) {constB=43758.5469})
  n47 = Multiply(A=@n46) {constB=0.2}
  n48 = TextureSample(
   UVs=Multiply(
    A=MaterialFunctionCall(UVs=@n42, "Rotation Angle (0-1)"=@n46) {materialFunction=&CustomRotator}.Rotated Values
    B=Add(A=@n47) {constB=1.1}
   )
   Tex=@n40
  ) {samplerSource=SSM_Wrap_WorldGroupSettings, samplerType=SAMPLERTYPE_LinearColor, texture=&OpalTest_N}
  n49 = MaterialFunctionCall(Float2=Abs(None=ConstantBiasScale(None=@n42) {bias=-0.5, scale=2})) {materialFunction=&BreakOutFloat2Components}
  n50 = Constant() {r=0.744}
  n51 = Constant() {r=1}
  n52 = Constant()
  n53 = Saturate(None=MaterialFunctionCall(Input=@n49.R, Input Low=@n50, Input High=@n51, Target Low=@n52, Target High=@n51) {materialFunction=&RemapValueRange}.Result)
  n54 = Saturate(None=MaterialFunctionCall(Input=@n49.G, Input Low=@n50, Input High=@n51, Target Low=@n52, Target High=@n51) {materialFunction=&RemapValueRange}.Result)
  n55 = OneMinus(None=Distance(A=AppendVector(A=@n53, B=@n54), B=Constant()))
  n56 = NamedRerouteDeclaration(None=Saturate(None=Power(Base=@n55) {constExponent=6})) {name=Gradient_2, nodeColor=<rgba 1 0.3516 0 1>}
  n57 = ComponentMask(None=VertexNormalWS()) {b=true}
  n58 = Power(Base=Multiply(A=Abs(None=@n57)) {constB=1.5}) {constExponent=20}
  FunctionOutput(None=LinearInterpolate(A=LinearInterpolate(A=@n18, B=@n36, Alpha=@n38), B=LinearInterpolate(A=@n41.RGB, B=@n48.RGB, Alpha=@n56), Alpha=Saturate(None=@n58)))
  n59 = Divide(
   A=WorldPosition().XYZ
   B=FunctionInput(Preview=Constant() {r=128}) {bUsePreviewValueAsDefault=true, inputName=Scale_1, inputType=FunctionInput_Scalar}
  ) {constB=128}
  NamedRerouteDeclaration(None=ComponentMask(None=@n59) {b=true, r=true}) {name=XZ, nodeColor=<rgba 1 0 0 1>}
  NamedRerouteDeclaration(None=ComponentMask(None=@n59) {b=true, g=true}) {name=YZ, nodeColor=<rgba 0 1 0 1>}
  NamedRerouteDeclaration(None=ComponentMask(None=@n59) {g=true, r=true}) {name=RG, nodeColor=<rgba 0 0 1 1>}
  NamedRerouteDeclaration(None=FunctionInput(Preview=TextureObject() {samplerType=SAMPLERTYPE_LinearColor, texture=&OpalTest_N}) {bUsePreviewValueAsDefault=true, inputName=Texture, inputType=FunctionInput_Texture2D}) {name=Texture, nodeColor=<rgba 0.2656 0 1 1>}
 ]}
