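# MF_WorldDither is a material function that outputs a noise value, named for use as a world-space dither.
# A single-level, tiling Value ALU noise node with turbulence off feeds the output, using camera-relative world position as its origin type and no wired inputs.
# A scaled world-position branch with an RG mask and multiply, plus a vertex normal node, does not reach the output.
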MaterialFunction /AtomFramework/Materials/MatFunctions/MF_WorldDither
graph {nodes [
  n1 = Divide(A=WorldPosition().XYZ) {constB=256}
  ComponentMask(None=@n1) {g=true, r=true}
  n2 = Noise!(Translated World Position, FilterWidth) {
   bTiling=true
   bTurbulence=false
   levelScale=1
   levels=1
   noiseFunction=NOISEFUNCTION_ValueALU
   outputMin=0
   worldPositionOriginType=CameraRelative
  }
  FunctionOutput(None=@n2)
  Multiply(A=@n1)
  VertexNormalWS()
 ]}
