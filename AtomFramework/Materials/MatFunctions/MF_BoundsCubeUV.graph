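# Material function that builds box-projected (cube) UVs from the channels of MF_InstanceBounds.
# The vertex normal, transformed to instance space, supplies weights for each axis direction.
# Chained lerps blend per-face channel pairs, flipped as needed, using clamped positive components for +X/+Y and remapped negative ones for -X/-Y/-Z.
# The +Z face is the unblended base and takes R for both U and V.
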MaterialFunction /AtomFramework/Materials/MatFunctions/MF_BoundsCubeUV
graph {nodes [
  n1 = MaterialFunctionCall() {materialFunction=&/MF_InstanceBounds}
  n2 = AppendVector(A=@n1.R, B=@n1.R)
  n3 = OneMinus(None=@n1.B)
  n4 = AppendVector(A=OneMinus(None=@n1.G), B=@n3)
  n5 = Transform(None=VertexNormalWS()) {transformSourceType=TRANSFORMSOURCE_World, transformType=TRANSFORM_Instance}
  n6 = ComponentMask(None=@n5) {r=true}
  n7 = Clamp(None=@n6)
  n8 = OneMinus(None=@n1.B)
  n9 = ComponentMask(None=@n5) {g=true}
  n10 = Constant()
  n11 = Constant() {r=-1}
  n12 = LinearInterpolate(
   A=LinearInterpolate(A=LinearInterpolate(A=@n2, B=@n4, Alpha=@n7), B=AppendVector(A=@n1.R, B=@n8), Alpha=Clamp(None=@n9))
   B=AppendVector(A=@n1.G, B=@n3)
   Alpha=MaterialFunctionCall!(Input=@n6, Input Low=@n10, Input High=@n11) {materialFunction=/Engine/Functions/Engine_MaterialFunctions03/Math/RemapValueRangeNormalized}.Result
  ) {constB=0}
  n13 = AppendVector(A=OneMinus(None=@n1.R), B=@n8)
  n14 = MaterialFunctionCall(Input=@n9, Input Low=@n10, Input High=@n11) {materialFunction=&RemapValueRangeNormalized}
  n15 = OneMinus(None=@n1.R)
  n16 = ComponentMask(None=@n5) {b=true}
  FunctionOutput(None=LinearInterpolate(
    A=LinearInterpolate(A=@n12, B=@n13, Alpha=@n14.Result)
    B=AppendVector(A=@n15, B=@n1.G)
    Alpha=MaterialFunctionCall(Input=@n16, Input Low=@n10, Input High=@n11) {materialFunction=&RemapValueRangeNormalized}.Result
   ))
 ]}
