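# Material function that, behind a static switch, renders common parts such as knobs and tubes as ray-traced analytic cylinders on proxy geometry.
# Vertex alpha encodes part type, setting height, radii and corner cut.
# UV-driven WPO orients proxies to the view; a custom ray intersection yields opacity mask, normal, specular and pixel depth offset.
# Knob caps receive a logo normal, with parallax stepping, like depth offset, reserved for Epic quality.
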MaterialFunction /AtomFramework/Materials/MatFunctions/MF_CommonPart_Analytic
graph {
 properties {description=Analytic Common Part}
 nodes [
  n1 = FunctionInput() {bUsePreviewValueAsDefault=true, inputType=FunctionInput_MaterialAttributes}
  n2 = SetMaterialAttributes!(
   MaterialAttributes=@n1
   OpacityMask=NamedRerouteUsage()
   WorldPositionOffset=NamedRerouteUsage()
   PixelDepthOffset=NamedRerouteUsage()
   Normal=NamedRerouteUsage()
   Specular
  ) {attributeSetTypes [
    #1
    #2
    #3
    #4
    #5
   ]}
  FunctionOutput(None=StaticSwitch(True=@n2, False=@n1, Value=StaticBoolParameter() {group=Common Parts, parameterName=CommonPartsAnalytic, sortPriority=2}))
  n3 = NamedRerouteUsage()
  n4 = NamedRerouteUsage()
  n5 = MaterialFunctionCall(Float2=TextureCoordinate()) {materialFunction=&BreakOutFloat2Components}
  n6 = Subtract(B=@n5.G) {constA=2.5, constB=0.5}
  n7 = Subtract(A=@n5.R) {constB=3}
  n8 = Subtract(A=@n5.R) {constB=0.5}
  n9 = Subtract(A=@n5.G) {constB=0.5}
  n10 = If(
   A=@n5.R
   "A > B"=MaterialFunctionCall(X=Constant() {r=1}, Y=Multiply(A=@n6) {constB=-2}, Z=Multiply(A=@n7) {constB=-1}) {materialFunction=&MakeFloat3}.Result
   "A < B"=MaterialFunctionCall(X=Multiply(A=@n8) {constB=2}, Y=Multiply(A=@n9) {constB=2}, Z=Constant() {r=1}) {materialFunction=&MakeFloat3}.Result
  ) {constB=1.5}
  n11 = MaterialFunctionCall(Float3=@n10) {materialFunction=&BreakOutFloat3Components}
  n12 = Multiply(A=@n11.G) {constB=-1}
  n13 = OneMinus(None=@n11.B)
  n14 = StaticSwitch(
   True=If(
    A=DotProduct(A=@n3, B=@n4)
    "A > B"=MaterialFunctionCall(X=@n11.R, Y=@n12, Z=@n13) {materialFunction=&MakeFloat3}.Result
    "A < B"=@n10
   )
   False=@n10
   Value=StaticBoolParameter() {defaultValue=true, dynamicBranch=true, group=Common Parts, parameterName=CommonPartsWPOFlip, sortPriority=8}
  )
  n15 = LinearInterpolate(
   A=NamedRerouteUsage()
   B=Transform(None=Constant3Vector() {constant=<rgba 0 0 1 0>}) {transformSourceType=TRANSFORMSOURCE_View}
   Alpha=ShadowReplace(Default=IsOrthographic(), Shadow=Constant() {r=1})
  )
  n16 = MaterialFunctionCall(X=DotProduct(A=NamedRerouteUsage(), B=@n15), Y=DotProduct(A=@n15, B=NamedRerouteUsage())) {materialFunction=&MakeFloat2}
  n17 = MaterialFunctionCall(Float2=Normalize(VectorInput=Multiply(A=@n16.Result) {constB=-1})) {materialFunction=&BreakOutFloat2Components}
  n18 = ComponentMask(None=@n14) {g=true, r=true}
  n19 = DotProduct(A=@n18, B=AppendVector(A=@n17.G, B=@n17.G))
  n20 = DotProduct(A=@n18, B=AppendVector(A=@n17.R, B=Multiply(A=@n17.G) {constB=-1}))
  n21 = MaterialFunctionCall(X=@n20, Y=@n19, Z=Multiply(A=ComponentMask(None=@n14) {b=true}, B=NamedRerouteUsage())) {materialFunction=&MakeFloat3}
  n22 = NamedRerouteUsage()
  n23 = NamedRerouteUsage()
  n24 = NamedRerouteUsage()
  n25 = NamedRerouteUsage()
  NamedRerouteDeclaration(None=Subtract(
    A=MaterialFunctionCall(VectorToTransform=@n21.Result, BasisX=@n22, BasisY=@n23, BasisZ=@n24, BasisPosition=@n25) {materialFunction=&Transform3x3Matrix}.Transformed Vector
    B=WorldPosition() {worldPositionShaderOffset=WPT_CameraRelativeNoOffsets}.XYZ
   )) {name=WPO, nodeColor=<rgba 0 1 0.8359 1>}
  n26 = Constant()
  n27 = ShadowReplace(Default=IsOrthographic(), Shadow=Constant() {r=1})
  n28 = Transform(None=Constant3Vector() {constant=<rgba 0 0 -1 0>}) {transformSourceType=TRANSFORMSOURCE_View}
  MaterialExpressionReroute_35 = Reroute(None=LinearInterpolate(A=CameraVectorWS(), B=@n28, Alpha=@n27))
  MaterialExpressionLinearInterpolate_8 = LinearInterpolate(A=Constant3Vector(), B=WorldPosition() {worldPositionShaderOffset=WPT_CameraRelative}.XYZ, Alpha=@n27)
  MaterialExpressionWorldPosition_0 = WorldPosition() {worldPositionShaderOffset=WPT_CameraRelative}
  MaterialExpressionReroute_25 = Reroute(None=VertexInterpolator(VS=NamedRerouteUsage()).PS)
  MaterialExpressionAdd_2 = Add(A=@MaterialExpressionReroute_25, B=Multiply(A=NamedRerouteUsage(), B=NamedRerouteUsage()))
  MaterialExpressionReroute_26 = Reroute(None=NamedRerouteUsage())
  MaterialExpressionReroute_7 = Reroute(None=NamedRerouteUsage())
  n29 = NamedRerouteUsage()
  MaterialExpressionSubtract_0 = Subtract(A=Saturate(None=Sign(None=@n29)), B=Saturate(None=Sign(None=@MaterialExpressionReroute_7)))
  MaterialExpressionNamedRerouteUsage_41 = NamedRerouteUsage()
  MaterialExpressionNamedRerouteUsage_5 = NamedRerouteUsage()
  MaterialExpressionNamedRerouteUsage_6 = NamedRerouteUsage()
  MaterialExpressionConstant_0 = Constant() {r=0.56}
  n30 = Custom!(
   QuadrantIndex=@MaterialExpressionNamedRerouteUsage_41
   X=@MaterialExpressionNamedRerouteUsage_5
   Y=@MaterialExpressionNamedRerouteUsage_6
   Scale=@MaterialExpressionConstant_0
  ) {
   code="float3 Val = 0;\r\nQuadrantIndex = floor(QuadrantIndex + .5);\r\nVal += (QuadrantIndex == 2 || QuadrantIndex == 3) ? Y : -Y;\r\nVal += (QuadrantIndex == 1 || QuadrantIndex == 2) ? X : -X;\r\nreturn Scale * Val * saturate(QuadrantIndex);"
   description=QuadrantToOffset
   inputs [
    {
     inputName=QuadrantIndex
     input {expression=&self:MaterialExpressionNamedRerouteUsage_41}
    }
    {inputName=X, input {expression=&self:MaterialExpressionNamedRerouteUsage_5}}
    {inputName=Y, input {expression=&self:MaterialExpressionNamedRerouteUsage_6}}
    {inputName=Scale, input {expression=&self:MaterialExpressionConstant_0}}
   ]
  }
  MaterialExpressionReroute_8 = Reroute(None=@n30)
  MaterialExpressionReroute_9 = Reroute(None=Constant())
  n31 = Custom!(
   RayOrigin=@MaterialExpressionLinearInterpolate_8
   RayDirection=@MaterialExpressionReroute_35
   WorldPosition=@MaterialExpressionWorldPosition_0.XYZ
   Origin=@MaterialExpressionReroute_25
   Top=@MaterialExpressionAdd_2
   Radius=@MaterialExpressionReroute_26
   InnerRadius=@MaterialExpressionReroute_7
   ShowLogo=@MaterialExpressionSubtract_0
   PlaneOffset=@MaterialExpressionReroute_8
   TwoSided=@MaterialExpressionReroute_9
  ) {
   code="float3 ro = RayOrigin;\r\nfloat3 rd = RayDirection;\r\nfloat3 a = Origin;\r\nfloat3 b = Top;\r\nfloat r0 = Radius;\r\nfloat r1 = InnerRadius;\r\n\r\nfloat3  ba = b - a;\r\nfloat3  oc = ro - a;\r\n\r\nfloat baba = dot(ba,ba);\r\nfloat bard = dot(ba,rd);\r\nfloat baoc = dot(ba,oc);\r\n\r\nfloat k2 = baba - bard * bard;\r\nfloat k1 = baba * dot(oc, rd) - baoc * bard;\r\n\r\nfloat tMin = 0;\r\nint hit = -1; // OuterRing=0, InnerRing=1, Cap=2 or CornerPlane=3\r\n\r\n// Outer ring\r\nfloat hOuter = 0;\r\n{\r\n    float k0 = baba * dot(oc, oc) - baoc * baoc - r0 * r0 * baba;\r\n    float h = k1 * k1 - k2 * k0;\r\n    if (h >= 0)\r\n    {\r\n        h = sqrt(h);\r\n        hOuter = h;\r\n\r\n        if (1)\r\n        {\r\n            float t = (-k1 + h) / k2;\r\n            float y = baoc + t * bard;\r\n            if (y > 0 &&&& y < baba)\r\n            {\r\n                if (hit == -1 || t > tMin)\r\n                {\r\n                    tMin = t;\r\n                    hit = 0;\r\n                }\r\n            }\r\n        }\r\n\r\n        if (TwoSided != 0)\r\n        {\r\n            float t = (-k1 - h) / k2;\r\n\r\n            float y = baoc + t * bard;\r\n            if (y > 0 &&&& y < baba)\r\n            {\r\n                if (hit == -1 || t > tMin)\r\n                {\r\n                    tMin = t;\r\n                    hit = 0;\r\n                }\r\n            }  \r\n        }\r\n    }\r\n}\r\n\r\n// Inner ring\r\nfloat hInner = 0;\r\nif (r1 > 0)\r\n{\r\n    float k0 = baba * dot(oc, oc) - baoc * baoc - r1 * r1 * baba;\r\n    float h = k1 * k1 - k2 * k0;\r\n    if (h >= 0)\r\n    {\r\n        h = sqrt(h);\r\n        hInner = h;\r\n\r\n        if (1)\r\n        {\r\n            float t = (-k1 - h) / k2;\r\n            float y = baoc + t * bard;\r\n            if (y > 0 &&&& y < baba)\r\n            {\r\n                if (hit == -1 || t > tMin)\r\n                {\r\n                    tMin = t;\r\n                    hit = 1;\r\n                }\r\n            }\r\n        }\r\n\r\n        if (TwoSided != 0)\r\n        {\r\n            float t = (-k1 + h) / k2;\r\n            float y = baoc + t * bard;\r\n            if (y > 0 &&&& y < baba)\r\n            {\r\n                if (hit == -1 || t > tMin)\r\n                {\r\n                    tMin = t;\r\n                    hit = 1;\r\n                }\r\n            }\r\n        }\r\n    }\r\n}\r\n\r\n// Cap\r\nif (1)\r\n{\r\n    float t = (baba - baoc) / bard;\r\n    float h = abs(k1 + k2 * t);\r\n    if (h >= hInner &&&& h <= hOuter)\r\n    {\r\n        if (TwoSided != 0 || bard > 0)\r\n        {\r\n            if (hit == -1 || t > tMin)\r\n            {\r\n                tMin = t;\r\n                hit = 2;\r\n            }\r\n        }\r\n    }\r\n}\r\n\r\n// CornerPlane\r\nif (any(PlaneOffset) != 0)\r\n{\r\n    float3 p0 = a + PlaneOffset;\r\n    float3 n = PlaneOffset;\r\n\r\n    if (dot(ro + tMin * rd - p0, n) > 0)\r\n    {\r\n        hit = -1;\r\n        \r\n        float3 p = ro - p0;\r\n        float t = -dot(p, n) / dot(rd, n);\r\n        if (t < 0)\r\n        {\r\n            float y = baoc + t * bard;\r\n            float h = abs(k1 + k2 * t);\r\n            if (y > 0 &&&& y < baba &&&& h <= hOuter)\r\n            {\r\n                hit = 3;\r\n                tMin = t;\r\n            }\r\n        }\r\n    }\r\n}\r\n\r\nif (hit == -1)\r\n{\r\n    // No intersection\r\n    Normal = float3(0, 0, 1);\r\n    Opacity = 0; \r\n    LogoOpacity = 0;\r\n    return WorldPosition;    \r\n}\r\n\r\nif (hit == 0 || hit == 1)\r\n{\r\n    // Inner or Outer Ring\r\n    float y = baoc + tMin * bard;\r\n    float ynorm = y / baba;\r\n    float r = hit == 0 ? r0 : -r1;\r\n    float3 Axis = ba / baba;\r\n    Normal = (oc + tMin * rd - ba * ynorm) / r;\r\n    Normal = lerp(Axis, Normal, smoothstep(0.0, 0.2, ynorm));\r\n    Normal = lerp(Normal, Axis, smoothstep(0.8, 1.0, ynorm));\r\n    LogoOpacity = 0;\r\n}\r\nif (hit == 2)\r\n{\r\n    // Cap\r\n    Normal = ba / baba;\r\n    LogoOpacity = ShowLogo;\r\n}\r\nif (hit == 3)\r\n{\r\n    // Plane\r\n    float h = abs(k1 + k2 * tMin);\r\n    float y = baoc + tMin * bard;\r\n    float3 Axis = ba / baba;\r\n    float3 CylinderNormal = (oc + tMin * rd - ba * y / baba) / r0;\r\n    Normal = normalize(PlaneOffset);\r\n    Normal = lerp(Normal, CylinderNormal, smoothstep(0.9, 1.0, h / hOuter));\r\n    Normal = lerp(Axis, Normal, smoothstep(0.0, 0.2, y / baba));\r\n    Normal = lerp(Normal, Axis, smoothstep(0.8, 1.0, y / baba));\r\n    LogoOpacity = 0;\r\n}\r\n\r\nOpacity = 1;\r\nreturn ro + tMin * rd;\r\n"
   description=Ray Intersection
   additionalOutputs [{outputName=Normal, outputType=CMOT_Float3}, {outputName=Opacity, outputType=CMOT_Float1}, {outputName=LogoOpacity, outputType=CMOT_Float1}]
   inputs [
    {inputName=RayOrigin, input {expression=&self:MaterialExpressionLinearInterpolate_8}}
    {inputName=RayDirection, input {expression=&self:MaterialExpressionReroute_35}}
    {
     inputName=WorldPosition
     input {expression=&self:MaterialExpressionWorldPosition_0, mask=1, maskB=1, maskG=1, maskR=1}
    }
    {inputName=Origin, input {expression=&self:MaterialExpressionReroute_25}}
    {inputName=Top, input {expression=&self:MaterialExpressionAdd_2}}
    {inputName=Radius, input {expression=&self:MaterialExpressionReroute_26}}
    {inputName=InnerRadius, input {expression=&self:MaterialExpressionReroute_7}}
    {inputName=ShowLogo, input {expression=&self:MaterialExpressionSubtract_0}}
    {inputName=PlaneOffset, input {expression=&self:MaterialExpressionReroute_8}}
    {inputName=TwoSided, input {expression=&self:MaterialExpressionReroute_9}}
   ]
  }
  n32 = Subtract(
   A=Multiply(A=DotProduct(A=@n28, B=@MaterialExpressionReroute_35), B=Length(None=@n31.return) {is=class MaterialExpressionLength})
   B=PixelDepth().R
  ) {desc=PDO}
  n33 = Saturate(None=@n31.Opacity)
  NamedRerouteDeclaration(None=QualitySwitch(Default=@n26, Low=@n26, High=@n26, Medium=@n26, Epic=Multiply(A=@n32, B=@n33))) {name=PDO, nodeColor=<rgba 1 0.2812 0 1>}
  n34 = TextureCoordinate()
  n35 = Transform(None=If(
    A=MaterialFunctionCall(Float2=@n34) {materialFunction=&BreakOutFloat2Components}.R
    "A > B"=Constant3Vector() {constant=<rgba -1 0 0 0>}
    "A < B"=Constant3Vector() {constant=<rgba 0 0 1 0>}
   ) {constB=1.5})
  n36 = NamedRerouteUsage()
  NamedRerouteDeclaration(None=Multiply(A=@n35, B=Abs(None=@n36))) {name=CPToWorld2, nodeColor=<rgba 0.8984 0 1 1>}
  n37 = TextureCoordinate()
  n38 = Transform(None=If(
    A=MaterialFunctionCall(Float2=@n37) {materialFunction=&BreakOutFloat2Components}.R
    "A > B"=Constant3Vector() {constant=<rgba 0 0 1 0>}
    "A < B"=Constant3Vector() {constant=<rgba 1 0 0 0>}
   ) {constB=1.5})
  n39 = MaterialFunctionCall(Float3=@n38) {materialFunction=&BreakOutFloat3Components}
  n40 = Transform(None=Constant3Vector() {constant=<rgba 0 1 0 0>})
  n41 = MaterialFunctionCall(Float3=@n40) {materialFunction=&BreakOutFloat3Components}
  n42 = MaterialFunctionCall(Float3=@n35) {materialFunction=&BreakOutFloat3Components}
  NamedRerouteDeclaration(None=Divide(
    A=MaterialFunctionCall(X=@n39.R, Y=@n41.R, Z=@n42.R) {materialFunction=&MakeFloat3}.Result
    B=NamedRerouteUsage()
   )) {name=WorldToCP0, nodeColor=<rgba 0 0.4375 1 1>}
  NamedRerouteDeclaration(None=Divide(
    A=MaterialFunctionCall(X=@n39.G, Y=@n41.G, Z=@n42.G) {materialFunction=&MakeFloat3}.Result
    B=NamedRerouteUsage()
   )) {name=WorldToCP1, nodeColor=<rgba 0 0.4375 1 1>}
  n43 = NamedRerouteUsage()
  NamedRerouteDeclaration(None=Divide(
    A=MaterialFunctionCall(X=@n39.B, Y=@n41.B, Z=@n42.B) {materialFunction=&MakeFloat3}.Result
    B=Abs(None=@n43)
   )) {name=WorldToCP2, nodeColor=<rgba 0 0.4375 1 1>}
  n44 = MaterialFunctionCall(Float2=TextureCoordinate()) {materialFunction=&BreakOutFloat2Components}
  n45 = Multiply(A=Subtract(A=@n44.R) {constB=3})
  n46 = Abs(None=NamedRerouteUsage())
  n47 = Multiply(A=Subtract(B=@n44.G) {constA=2.5, constB=0.5}) {constB=-2}
  n48 = NamedRerouteUsage()
  n49 = Multiply(A=Subtract(A=@n44.R) {constB=0.5}) {constB=2}
  n50 = Multiply(A=Subtract(A=@n44.G) {constB=0.5}) {constB=2}
  n51 = If(
   A=@n44.R
   "A > B"=MaterialFunctionCall(X=Multiply(A=@n45, B=@n46), Y=Multiply(A=@n47, B=@n48), Z=@n48) {materialFunction=&MakeFloat3}.Result
   "A < B"=MaterialFunctionCall(X=Multiply(A=@n49, B=@n48), Y=Multiply(A=@n50, B=@n48), Z=@n46) {materialFunction=&MakeFloat3}.Result
  ) {constB=1.5}
  n52 = NamedRerouteDeclaration(None=Subtract(A=WorldPosition() {worldPositionShaderOffset=WPT_CameraRelativeNoOffsets}.XYZ, B=Transform(None=@n51))) {name=CPToWorld3, nodeColor=<rgba 0.8984 0 1 1>}
  NamedRerouteDeclaration(None=MaterialFunctionCall(VectorToTransform=Multiply(A=@n52) {constB=-1}, BasisX=NamedRerouteUsage(), BasisY=NamedRerouteUsage(), BasisZ=NamedRerouteUsage()) {materialFunction=&Transform3x3Matrix}.Transformed Vector) {name=WorldToCP3, nodeColor=<rgba 0 0.4375 1 1>}
  n53 = Add(A=Multiply(A=NamedRerouteUsage()) {constB=0.57}) {constB=0.5}
  n54 = ComponentMask(None=@n53) {r=true}
  n55 = OneMinus(None=@n54)
  MaterialExpressionSign_1 = Sign(None=NamedRerouteUsage())
  n56 = Custom!(In=@MaterialExpressionSign_1) {
   code="return In * GetPrimitive_DeterminantSign_FromFlags(GetPrimitiveData(Parameters).Flags);"
   description=Multiply by primitive determinant sign
   inputs [{inputName=In, input {expression=&self:MaterialExpressionSign_1}}]
  }
  n57 = Multiply(A=@n56) {constB=-1}
  n58 = Add(A=Multiply(A=@n57) {constB=0.5}) {constB=0.5}
  n59 = ComponentMask(None=@n53) {g=true}
  NamedRerouteDeclaration(None=AppendVector(A=LinearInterpolate(A=@n55, B=@n54, Alpha=@n58), B=OneMinus(None=@n59))) {name=LogoUV, nodeColor=<rgba 0 0.3672 1 1>}
  n60 = DotProduct(
   A=WorldPosition() {worldPositionShaderOffset=WPT_CameraRelative}.XYZ
   B=Transform(None=Constant3Vector() {constant=<rgba 0 0 1 0>}) {transformSourceType=TRANSFORMSOURCE_View}
  )
  NamedRerouteDeclaration(None=Multiply(A=Divide(B=@n60), B=Constant() {r=1.75})) {name=Ruler, nodeColor=<rgba 1 0 0.9609 1>}
  MaterialExpressionMultiply_7 = Multiply(A=VertexColor().A) {constB=255}
  n61 = Custom(In=@MaterialExpressionMultiply_7) {
   code="// Default to simple knob height/radius.\r\nfloat Height = 0.18;\r\nRadius = 0.25;\r\nInnerRadius = 0;\r\nQuadrantId = 0;\r\n\r\n// CPIndex is from EAtomCommonPartType.\r\n// Only supported subset of common parts are considered here.\r\n// The output is expected to match the dimensions of the common part mesh.\r\n// Ugly branchy logic is faster than any other LUT style approaches I tried (both texture and shader array data).\r\nfloat CPIndex = floor(In+0.1);\r\nif (CPIndex >= 8 &&&& CPIndex <= 17)\r\n{\r\n\t if (CPIndex == 8) \r\n\t {\r\n\t \tInnerRadius = 0.1575;\r\n\t }\r\n\t else if (CPIndex > 9)\r\n\t {\r\n\t\tInnerRadius = frac(CPIndex / 2) != 0 ? 0.1575 : 0;\r\n\t\tQuadrantId = floor((CPIndex - 8) * 0.5);\r\n\t }\r\n}\r\nelse if (CPIndex < 22)\r\n{\r\n\tRadius = 0.1575;\r\n\tHeight = (CPIndex < 20) ? -0.21 : -0.85;\r\n\tInnerRadius = frac(CPIndex / 2) == 0 ? 0.075 : 0;\r\n}\r\nelse\r\n{\r\n\tRadius = 0.34;\r\n\tInnerRadius = 0.22;\r\n\tHeight = (CPIndex == 22) ? -0.21 : (CPIndex == 23) ? -0.85 : -2.735;\r\n}\r\n\r\nreturn Height;"
   description=CommonPartsIdToDescription
   additionalOutputs [
    {outputName=Radius, outputType=CMOT_Float1}
    {outputName=InnerRadius, outputType=CMOT_Float1}
    {outputName=QuadrantId, outputType=CMOT_Float1}
   ]
   inputs [{inputName=In, input {expression=&self:MaterialExpressionMultiply_7}}]
  }
  NamedRerouteDeclaration(None=@n61.QuadrantId) {name=QuadrantId, nodeColor=<rgba 1 0 0.2812 1>}
  NamedRerouteDeclaration(None=@n31.Opacity) {name=Opacity, nodeColor=<rgba 0 1 0.7188 1>}
  n62 = WorldPosition() {worldPositionShaderOffset=WPT_CameraRelativeNoOffsets}
  n63 = Add(
   A=Multiply(
    A=DotProduct(A=@n62.XYZ, B=@n62.XYZ)
    B=ScalarParameter() {bUseCustomPrimitiveData=true, parameterName=CommonPartsFadeScale, primitiveDataIndex=1}
   )
   B=ScalarParameter() {bUseCustomPrimitiveData=true, defaultValue=1, parameterName=CommonPartsFadeBias, primitiveDataIndex=2}
  )
  NamedRerouteDeclaration(None=Saturate(None=OneMinus(None=@n63))) {name=HeightFade, nodeColor=<rgba 0 0.2031 1 1>}
  NamedRerouteDeclaration(None=MaterialFunctionCall(X=@n20, Y=@n19) {materialFunction=&MakeFloat2}.Result) {name=CPSpaceForUV, nodeColor=<rgba 1 0.7639 0.1963 1>}
  n64 = DotProduct(A=@n31.Normal, B=Normalize(VectorInput=NamedRerouteUsage()))
  NamedRerouteDeclaration(None=Abs(None=Multiply(A=@n64) {constB=0.25})) {name=Specular, nodeColor=<rgba 1 0 0.4922 1>}
  n65 = MaterialFunctionCall() {materialFunction=&ObjectScale}
  n66 = Multiply(
   A=ScalarParameter() {defaultValue=1, group=Common Parts, parameterName=CommonPartScale}
   B=ShaderStageSwitch(PixelShader=VertexInterpolator(VS=@n65.Scale X).PS, VertexShader=@n65.Scale X)
  )
  NamedRerouteDeclaration(None=Multiply(A=@n61.Radius, B=@n66)) {name=Radius, nodeColor=<rgba 1 0 0.2109 1>}
  NamedRerouteDeclaration(None=Multiply(A=@n61.InnerRadius, B=@n66)) {name=InnerRadius, nodeColor=<rgba 1 0 0.2812 1>}
  n67 = NamedRerouteDeclaration(None=Multiply(A=@n31.Opacity, B=@n31.LogoOpacity)) {name=LogoOpacity, nodeColor=<rgba 0 1 0.3672 1>}
  n68 = TextureSample(UVs=NamedRerouteUsage()) {samplerSource=SSM_Clamp_WorldGroupSettings, samplerType=SAMPLERTYPE_Normal, texture=/AtomFramework/Materials/Textures/AtomLogo_N}
  n69 = MaterialFunctionCall(X=@n68.G, Y=@n68.G, Z=@n68.G) {materialFunction=&MakeFloat3}
  MaterialExpressionMultiply_40 = Multiply(A=NamedRerouteUsage()) {constB=-1}
  MaterialExpressionNamedRerouteUsage_38 = NamedRerouteUsage()
  MaterialExpressionDDX_1 = DDX(Value=@MaterialExpressionNamedRerouteUsage_38)
  MaterialExpressionDDY_1 = DDY(Value=@MaterialExpressionNamedRerouteUsage_38)
  MaterialExpressionNamedRerouteUsage_39 = NamedRerouteUsage()
  n70 = Saturate(None=NamedRerouteUsage())
  MaterialExpressionLinearInterpolate_5 = LinearInterpolate(A=Constant() {r=1}, B=Constant() {r=8}, Alpha=@n70) {constA=1, constB=8}
  MaterialExpressionLinearInterpolate_4 = LinearInterpolate(A=Constant(), B=Constant() {r=0.025}, Alpha=@n70)
  MaterialExpressionReroute_69 = Reroute(None=TextureObject() {samplerType=SAMPLERTYPE_Normal, texture=&AtomLogo_N})
  MaterialExpressionReroute_70 = Reroute(None=TextureObject() {samplerType=SAMPLERTYPE_Masks, texture=/AtomFramework/Materials/Textures/AtomLogo_M})
  n71 = Custom!(
   v=@MaterialExpressionMultiply_40
   uv=@MaterialExpressionNamedRerouteUsage_38
   InDDX=@MaterialExpressionDDX_1
   InDDY=@MaterialExpressionDDY_1
   Opacity=@MaterialExpressionNamedRerouteUsage_39
   Steps=@MaterialExpressionLinearInterpolate_5
   Depth=@MaterialExpressionLinearInterpolate_4
   Normal=@MaterialExpressionReroute_69
   Bump=@MaterialExpressionReroute_70
  ) {
   code="if (Steps == 0 || Opacity == 0)\r\n{\r\n\treturn float3(0, 0, 1);\r\n}\r\n\r\nfloat3 OutNormal = float3(0, 0, 1);\r\nfloat2 uvOut = float2(0, 0);\r\nfloat specOut = 0;\r\nfloat alpha = 0;\r\nfloat intersect = 0;\r\nfloat2 NormalXY = float2(0,0); \r\n\r\nint i = 0;\r\nwhile (i < Steps)\r\n{\r\n    float h = Depth * i / Steps;\r\n\tfloat2 uvCurrent = uv + v.xy * (h - Depth);\r\n\tintersect = ((Bump.SampleGrad(GetMaterialSharedSampler(BumpSampler, View.MaterialTextureBilinearWrapedSampler), uvCurrent, InDDX, InDDY).x - 0.5) * 2) * Depth;\r\n\tif (h < intersect.x || i == 0)\r\n\t{\t\r\n        uvOut = uvCurrent;\r\n\t\talpha = 1;\r\n\t}\r\n\ti++;\r\n}\r\n\r\nNormalXY = Normal.SampleGrad(GetMaterialSharedSampler(NormalSampler, View.MaterialTextureBilinearWrapedSampler), uvOut, InDDX, InDDY).yx;\r\nNormalXY = NormalXY * 2 - 1;\r\nfloat NormalZ = sqrt(saturate( 1.0f - dot( NormalXY, NormalXY ) ));\r\nOutNormal = float3(NormalXY, NormalZ);\r\n\r\nspecOut = Bump.SampleGrad(GetMaterialSharedSampler(BumpSampler, View.MaterialTextureBilinearWrapedSampler), uvOut, InDDX, InDDY);\r\n\r\nreturn OutNormal;"
   description=MultiBumpLogo
   inputs [
    {inputName=v, input {expression=&self:MaterialExpressionMultiply_40}}
    {inputName=uv, input {expression=&self:MaterialExpressionNamedRerouteUsage_38}}
    {inputName=InDDX, input {expression=&self:MaterialExpressionDDX_1}}
    {inputName=InDDY, input {expression=&self:MaterialExpressionDDY_1}}
    {inputName=Opacity, input {expression=&self:MaterialExpressionNamedRerouteUsage_39}}
    {inputName=Steps, input {expression=&self:MaterialExpressionLinearInterpolate_5}}
    {inputName=Depth, input {expression=&self:MaterialExpressionLinearInterpolate_4}}
    {inputName=Normal, input {expression=&self:MaterialExpressionReroute_69}}
    {inputName=Bump, input {expression=&self:MaterialExpressionReroute_70}}
   ]
  }
  n72 = MaterialFunctionCall(Float3=ShadingPathSwitch(Default=QualitySwitch(Default=@n69.Result, Low=@n69.Result, High=@n69.Result, Medium=@n69.Result, Epic=@n71))) {materialFunction=&BreakOutFloat3Components}
  n73 = MaterialFunctionCall(
   VectorToTransform=MaterialFunctionCall(X=Multiply(A=@n72.G, B=@n57) {constB=-1}, Y=Multiply(A=@n72.R) {constB=-1}, Z=@n72.B) {materialFunction=&MakeFloat3}.Result
   BasisX=NamedRerouteUsage()
   BasisY=NamedRerouteUsage()
   BasisZ=NamedRerouteUsage()
  ) {materialFunction=&Transform3x3Matrix}
  NamedRerouteDeclaration(None=Transform(None=If(A=@n67, "A > B"=@n73.Transformed Vector, "A == B"=@n31.Normal, "A < B"=@n31.Normal) {equalsThreshold=0}) {transformSourceType=TRANSFORMSOURCE_World, transformType=TRANSFORM_Tangent}) {name=Normal, nodeColor=<rgba 1 0.7031 0 1>}
  NamedRerouteDeclaration(None=Multiply(A=@n61.return, B=@n66)) {name=Height, nodeColor=<rgba 1 0 0.2812 1>}
  NamedRerouteDeclaration(None=Multiply(A=@n38, B=NamedRerouteUsage())) {name=CPToWorld0, nodeColor=<rgba 0.8984 0 1 1>}
  NamedRerouteDeclaration(None=Multiply(A=@n40, B=NamedRerouteUsage())) {name=CPToWorld1, nodeColor=<rgba 0.8984 0 1 1>}
  NamedRerouteUsage()
 ]
}
