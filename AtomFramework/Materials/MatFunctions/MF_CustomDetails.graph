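# Material function that adds custom normal and opacity mask detail to incoming material attributes when Use Custom Details is enabled.
# Detail sources are UV-tile-gated textures or two HLSL ray-intersection fakes, an analytic tube and a rotated lattice in a box, with static switches for selection.
# Detail normals blend angle-corrected with the input, masks multiply, and NaniteReplace restricts both to non-Nanite LODs.
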MaterialFunction /AtomFramework/Materials/MatFunctions/MF_CustomDetails
graph {nodes [
  n1 = GetMaterialAttributes(None=FunctionInput() {inputType=FunctionInput_MaterialAttributes}) {attributeGetTypes [#1, #2]}
  n2 = NamedRerouteUsage()
  n3 = NamedRerouteUsage()
  FunctionOutput(None=StaticSwitchParameter(
    True=SetMaterialAttributes!(MaterialAttributes=@n1.MaterialAttributes, Normal=@n2, OpacityMask=@n3) {attributeSetTypes [#1, #2]}
    False=@n1.MaterialAttributes
   ) {group=Custom Details, parameterName=Use Custom Details, sortPriority=0})
  NamedRerouteDeclaration(None=@n1.Normal) {name=Normal In, nodeColor=<rgba 0 1 0.1562 1>}
  NamedRerouteDeclaration(None=@n1.OpacityMask) {name=Opacity Mask In, nodeColor=<rgba 0.9453 0 1 1>}
  NamedRerouteDeclaration(None=Transform(None=CameraVectorWS()) {transformSourceType=TRANSFORMSOURCE_World, transformType=TRANSFORM_Tangent}) {name=Camera Vector TS, nodeColor=<rgba 0.4062 0 1 1>}
  NamedRerouteDeclaration(None=TextureCoordinate()) {name=UV_Normal, nodeColor=<rgba 1 0.7969 0 1>}
  NamedRerouteDeclaration(None=TextureCoordinate() {coordinateIndex=1}) {name=UV_Alpha, nodeColor=<rgba 1 0.7969 0 1>}
  n4 = NamedRerouteUsage()
  n5 = NamedRerouteUsage()
  NamedRerouteDeclaration(None=NaniteReplace!(
    Default=MaterialFunctionCall(BaseNormal=@n4, AdditionalNormal=@n5) {materialFunction=&BlendAngleCorrectedNormals}.Result
    Nanite=NamedRerouteUsage()
   ) {desc=Custom details on LODs only}) {name=Normal Out, nodeColor=<rgba 0.2891 0 1 1>}
  n6 = Multiply(A=NamedRerouteUsage(), B=NamedRerouteUsage())
  NamedRerouteDeclaration(None=NaniteReplace(Default=Saturate(None=@n6), Nanite=NamedRerouteUsage()) {desc=Custom details on LODs only}) {name=Opacity Mask Out, nodeColor=<rgba 1 0.9375 0 1>}
  n7 = NamedRerouteUsage()
  n8 = ComponentMask(None=@n7) {g=true}
  n9 = NamedRerouteUsage()
  n10 = TextureSampleParameter2D(UVs=@n7) {
   group=Custom Details
   parameterName=Custom Details Normal
   samplerType=SAMPLERTYPE_Normal
   sortPriority=1
   texture=/AtomFramework/Materials/Textures/T_Flat_N
  }
  NamedRerouteDeclaration(None=If(A=Floor(None=@n8), B=Constant() {r=1}, "A > B"=@n9, "A == B"=@n10.RGB, "A < B"=@n9)) {name=Default Custom Normal, nodeColor=<rgba 1 0 0.6562 1>}
  n11 = NamedRerouteUsage()
  n12 = ComponentMask(None=@n11) {r=true}
  n13 = Constant() {r=1}
  n14 = TextureSampleParameter2D(UVs=@n11) {
   group=Custom Details
   parameterName=Custom Details Opacity
   samplerType=SAMPLERTYPE_LinearGrayscale
   sortPriority=2
   texture=/AtomFramework/Materials/Textures/T_White_A
  }
  NamedRerouteDeclaration(None=If(A=Floor(None=@n12), B=Constant() {r=-1}, "A > B"=@n13, "A == B"=@n14.R, "A < B"=@n13)) {name=Default Custom Mask, nodeColor=<rgba 0.5703 1 0 1>}
  MaterialExpressionNamedRerouteUsage_15 = NamedRerouteUsage()
  MaterialExpressionNamedRerouteUsage_16 = NamedRerouteUsage()
  n15 = VectorParameter() {
   group=Custom Details
   parameterName=Tube Controls
   channelNames {a=Bottom Depth, b=Ring Thickness, g=Ring Depth, r=Radius}
   defaultValue=<rgba 0.8 0.1 0.3 1.05>
  }
  MaterialExpressionAppendVector_1 = AppendVector(A=@n15, B=@n15)
  n16 = VectorParameter() {
   group=Custom Details
   parameterName=Scale Controls
   channelNames {a=Bottom Scale Y, b=Bottom Scale X, g=Ring Scale Y, r=Ring Scale X}
   defaultValue=<rgba 1 1 1 1>
  }
  MaterialExpressionAppendVector_0 = AppendVector(A=@n16, B=@n16)
  n17 = VectorParameter() {group=Custom Details, parameterName=Surface Scale, channelNames {a="", b="", g=Scale V, r=Scale U}, defaultValue=<rgba 1 1 1 1>}
  MaterialExpressionAppendVector_2 = AppendVector(A=@n17.Scale U, B=@n17.Scale U)
  n18 = Custom!(
   UV=@MaterialExpressionNamedRerouteUsage_15
   RayDirection=@MaterialExpressionNamedRerouteUsage_16
   TubeControls=@MaterialExpressionAppendVector_1
   ScaleControls=@MaterialExpressionAppendVector_0
   SurfaceScale=@MaterialExpressionAppendVector_2
  ) {
   code="/* Input --------------------------------------------------------------------------------------------*/\r\n\r\nfloat3  ro = float3(UV - 0.5, 0.0);                             // Ray origin\r\nfloat3  rd = -normalize(RayDirection);\r\n\r\nfloat   tr = TubeControls.x * 0.5; tr *= tr;                    // Tube radius squared\r\nfloat2  ss = SurfaceScale;                                      // UV scale\r\n\r\nfloat   od = TubeControls.y;                                    // Ring depth\r\nfloat   ot = saturate(1 - TubeControls.z);                      // Ring thickness\r\nfloat2  os = ScaleControls.xy;                                  // Ring scale\r\n\r\nfloat   bd = TubeControls.w;                                    // Bottom depth\r\nfloat2  bs = ScaleControls.zw;                                  // Bottom scale\r\n\r\n\r\n// Functions\r\nstruct Functions\r\n{\r\n    float2 ParallaxOffset(float2 uv, float3 rd, float d)\r\n    {\r\n        return uv - rd.xy / rd.z * d;\r\n    }\r\n\r\n    float Tube(float2 ro, float2 rd, float rr)\r\n    {\r\n        float a = 2 * dot(rd, rd);\r\n        float b = 2 * dot(rd, ro);\r\n        float c = dot(ro, ro) - rr;\r\n        float d = b * b - 2.0 * a * c;\r\n\r\n        if(d < 0.0) return -1.0; // No intersection\r\n\r\n        return (-b + sqrt(d)) / a;\r\n    }\r\n};\r\n\r\nFunctions f;\r\n//-------------------------//\r\n\r\n// Tube\r\nro.xy *= ss;\r\ntubeT = f.Tube(ro.xy, rd.xy, tr);\r\n\r\n// Parallax offset UVs\r\nfloat2 ruv = f.ParallaxOffset(ro.xy * os, rd, od); // Ring UVs\r\nfloat2 buv = f.ParallaxOffset(ro.xy * bs, rd, bd); // Bottom UVs\r\n\r\n// Masks\r\nfloat ms = step(tr, dot(ro.xy, ro.xy)); // Surface mask\r\nfloat mo = step(tr * ot, dot(ruv, ruv)) - step(tr, dot(ruv, ruv)); // Ring mask\r\nfloat mb = step(dot(buv, buv), tr); // Bottom mask\r\n\r\n// If intersecting\r\nif (tubeT > 0)\r\n{\r\n    normal = -normalize(float3(ro.xy + rd * tubeT, 0)); // Line equation to find position\r\n}\r\n\r\nif (saturate(ms + mo + mb) == 1.0)\r\n{\r\n    normal = float3(0.0, 0.0, 1.0);\r\n}\r\n\r\nreturn saturate(ms + mo + (1.0 - mb)); // Opacity"
   description=Analytic Tube
   outputType=CMOT_Float1
   additionalOutputs [{outputName=normal, outputType=CMOT_Float3}, {outputName=debug, outputType=CMOT_Float3}, {outputName=tubeT, outputType=CMOT_Float1}]
   inputs [
    {inputName=UV, input {expression=&self:MaterialExpressionNamedRerouteUsage_15}}
    {inputName=RayDirection, input {expression=&self:MaterialExpressionNamedRerouteUsage_16}}
    {inputName=TubeControls, input {expression=&self:MaterialExpressionAppendVector_1}}
    {inputName=ScaleControls, input {expression=&self:MaterialExpressionAppendVector_0}}
    {inputName=SurfaceScale, input {expression=&self:MaterialExpressionAppendVector_2}}
   ]
  }
  NamedRerouteDeclaration(None=@n18.return) {name=Analytic Tube Mask, nodeColor=<rgba 0.3125 0 1 1>}
  NamedRerouteDeclaration(None=@n18.normal) {name=Analytic Tube Normal, nodeColor=<rgba 0 1 0.9297 1>}
  MaterialExpressionNamedRerouteUsage_12 = NamedRerouteUsage()
  MaterialExpressionNamedRerouteUsage_13 = NamedRerouteUsage()
  n19 = ScalarParameter() {defaultValue=0.125, group=Custom Details, parameterName=Lattice Angle}
  MaterialExpressionSine_0 = Sine(None=@n19)
  MaterialExpressionCosine_0 = Cosine(None=@n19)
  MaterialExpressionScalarParameter_0 = ScalarParameter() {defaultValue=5, group=Custom Details, parameterName=UV Frac, sortPriority=-1}
  n20 = VectorParameter() {
   group=Custom Details
   parameterName=IntersectionControls
   sortPriority=0
   channelNames {a=Lattice Depth, b=Lattice Size, g=Box Height, r=Box Width}
   defaultValue=<rgba 0.5 0.0865 0.425 0.0875>
  }
  MaterialExpressionAppendVector_6 = AppendVector(A=@n20, B=@n20)
  n21 = VectorParameter() {
   group=Custom Details
   parameterName=UVControls
   sortPriority=0
   channelNames {a=Offset Y, b=Offset X, g=Scale Y, r=Scale X}
   defaultValue=<rgba 1.01 1 0.132 0>
  }
  MaterialExpressionAppendVector_7 = AppendVector(A=@n21, B=@n21)
  n22 = Custom!(
   uv=@MaterialExpressionNamedRerouteUsage_12
   cameraVector=@MaterialExpressionNamedRerouteUsage_13
   sine=@MaterialExpressionSine_0
   cosine=@MaterialExpressionCosine_0
   uvFrac=@MaterialExpressionScalarParameter_0
   intersectionControls=@MaterialExpressionAppendVector_6
   uvControls=@MaterialExpressionAppendVector_7
  ) {
   code="struct Functions\r\n{\r\n    float4\tGetBoxInterior(float3 ro, float3 rd, float3 box, float3 m) \r\n    {\r\n\t\tfloat3 n = m * ro;\r\n        float3 k = abs(m) * box;\r\n        float3 t = -n + k;\r\n    \r\n        float tF = min(min(t.x, t.y), t.z);\r\n\r\n        if(tF < 0.0) return -1.0; // No intersection\r\n        \r\n        return float4(tF, t);\r\n    }\r\n\r\n\tfloat3\tGetBoxInteriorNormal(float3 rd, float4 box)\r\n\t{\r\n\t\treturn step(box.yzw, float3(box.xxx)) * -sign(rd);\r\n\t}\r\n\r\n\tfloat\tGetBoxBorderMask(float3 ro, float3 box)\r\n\t{\r\n\t\tfloat2 cross = step(abs(ro.xy), abs(box.xy));\r\n\t\treturn saturate((cross * cross.yx).x);\r\n\t}\r\n\r\n\tfloat2\tRotateVec2(float2 vec, float s, float c) // s: sin || c: cos\r\n\t{\r\n\t\tvec -= 0.5;\r\n\t\treturn float2(dot(vec, float2(c, -s)), dot(vec, float2(s, c))) + 0.5;\r\n\t}\r\n\r\n\tfloat3\tTransformMatrixZ(float3 vec, float s, float c) // s: sin || c: cos\r\n\t{\r\n\t\treturn mul(vec, float3x3(float3(c, -s, 0.0), float3(s, c, 0.0), float3(0.0, 0.0, 1.0)));\r\n\t}\r\n};\r\n\r\nFunctions f;\r\n\r\n/* Initialize ---------------------------------------------------------------------------------------*/\r\nmask = 1.0;\r\nfloat3\tfn\t= float3(0.0, 0.0, 1.0); \t\t\t\t\t\t\t\t\t\t// Flat normal\r\nfloat3\tn\t= fn;\t\t\t\t\t\t\t\t\t\t\t\t\t\t\t// Output normal\r\n\r\nfloat2\tuvScale = uvControls.xy;\r\nfloat2\tuvOffset = uvControls.zw;\r\n\r\nfloat3\tlatticeSize\t= float3(intersectionControls.zzz);\r\nfloat3\tboxSize\t= float3(intersectionControls.xy, intersectionControls.w);\r\n\r\n/* Output -------------------------------------------------------------------------------------------*/\r\n\r\n// Rotate lattice UVs, transform view vector to lattice space\r\nfloat2\tluv\t= f.RotateVec2((uv + uvOffset) * uvScale, sine, cosine); \t\t// Lattice space UVs\r\nfloat3\tlro\t= float3(frac(luv * uvFrac) - 0.5, 0.0);\t\t\t\t\t\t// Lattice space ray origin\r\nfloat3\tlrd\t= normalize(-f.TransformMatrixZ(cameraVector, -sine, cosine)); \t// Lattice space ray direction\r\n\r\n\r\n// Box parameters\r\nfloat3\tbro\t= float3(uv - 0.5, 0.0);\t\t\t\t\t\t\t\t\t\t// Box ray origin\r\nfloat3\tbrd\t= normalize(-cameraVector);\t\t\t\t\t\t\t\t\t\t// Box ray direction\r\n\r\n// Early exit if out of box or lattice bounds\r\nif (f.GetBoxBorderMask(bro, boxSize) == 0.0 || f.GetBoxBorderMask(lro, latticeSize) == 0.0) return fn;\r\n\r\n// Intersect\r\nfloat4\tbox = f.GetBoxInterior(bro, brd, boxSize, 1 / brd);\t\t\t\t\t// Box\r\nfloat4\tltc = f.GetBoxInterior(lro, lrd, latticeSize, 1 / lrd);\t\t\t\t// Lattice\r\n\r\nn = f.GetBoxInteriorNormal(brd, box);\t\t\t\t\t\t\t\t\t\t// Box normals\r\n\r\nif (ltc.x / latticeSize.z <= box.x / boxSize.z)\r\n{\t\r\n\tn = f.TransformMatrixZ(f.GetBoxInteriorNormal(lrd, ltc), sine, cosine);\t// Lattice normals\r\n}\r\n\r\nmask = floor(lerp(1.0 - dot(n, fn), mask, dot(ltc.x, latticeSize.z))); \t\t// Opacity mask bottom plane\r\n\r\nreturn n;"
   description=Lattice
   additionalOutputs [{outputName=mask, outputType=CMOT_Float1}]
   inputs [
    {inputName=uv, input {expression=&self:MaterialExpressionNamedRerouteUsage_12}}
    {inputName=cameraVector, input {expression=&self:MaterialExpressionNamedRerouteUsage_13}}
    {inputName=sine, input {expression=&self:MaterialExpressionSine_0}}
    {inputName=cosine, input {expression=&self:MaterialExpressionCosine_0}}
    {inputName=uvFrac, input {expression=&self:MaterialExpressionScalarParameter_0}}
    {inputName=intersectionControls, input {expression=&self:MaterialExpressionAppendVector_6}}
    {inputName=uvControls, input {expression=&self:MaterialExpressionAppendVector_7}}
   ]
  }
  NamedRerouteDeclaration(None=@n22.mask) {name=Lattice Mask, nodeColor=<rgba 1 0 0.2344 1>}
  NamedRerouteDeclaration(None=@n22.return) {name=Lattice Normal, nodeColor=<rgba 0.0781 1 0 1>}
  NamedRerouteUsage()
  StaticSwitchParameter(
   True=NamedRerouteUsage()
   False=StaticSwitchParameter(True=NamedRerouteUsage(), False=NamedRerouteUsage()) {group=Custom Details, parameterName=UseAnalyticTube}
  ) {group=Custom Details, parameterName=UseLattice}
 ]}
