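# Material function that overlays up to four optional decorator textures, plus an optional "3DP" one, onto incoming material attributes.
# Each numbered decorator samples UV1 and is masked to a parameterized UDIM tile by floored U, while the 3DP one samples UV0.
# The summed layers are limited to UV1's first V row, then their alpha blends base color, subsurface color and opacity.
# Static switches toggle each layer and the whole effect, which also outputs the mask.
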MaterialFunction /AtomFramework/Materials/MatFunctions/MF_Decorator
graph {nodes [
  n1 = GetMaterialAttributes(None=FunctionInput() {bUsePreviewValueAsDefault=true, inputType=FunctionInput_MaterialAttributes}) {attributeGetTypes [#1, #2, #3]}
  n2 = TextureCoordinate() {coordinateIndex=1}
  n3 = TextureSampleParameter2D(UVs=@n2) {
   group=Decorator
   parameterName=Decorator 1
   samplerSource=SSM_Wrap_WorldGroupSettings
   sortPriority=34
   texture=&Black
  }
  n4 = ComponentMask(None=@n2) {r=true}
  n5 = ScalarParameter() {group=Decorator, parameterName=Decorator 1 UDIM, sortPriority=35}
  n6 = Constant()
  n7 = Multiply(A=@n3.RGBA, B=If(A=Floor(None=@n4), B=Floor(None=@n5), "A > B"=@n6, "A == B"=Constant() {r=1}, "A < B"=@n6))
  n8 = Constant()
  n9 = TextureCoordinate() {coordinateIndex=1}
  n10 = TextureSampleParameter2D(UVs=@n9) {
   group=Decorator
   parameterName=Decorator 2
   samplerSource=SSM_Wrap_WorldGroupSettings
   sortPriority=37
   texture=&Black
  }
  n11 = ComponentMask(None=@n9) {r=true}
  n12 = ScalarParameter() {defaultValue=1, group=Decorator, parameterName=Decorator 2 UDIM, sortPriority=38}
  n13 = Constant()
  n14 = Multiply(A=@n10.RGBA, B=If(A=Floor(None=@n11), B=Floor(None=@n12), "A > B"=@n13, "A == B"=Constant() {r=1}, "A < B"=@n13))
  n15 = TextureCoordinate() {coordinateIndex=1}
  n16 = TextureSampleParameter2D(UVs=@n15) {
   group=Decorator
   parameterName=Decorator 3
   samplerSource=SSM_Wrap_WorldGroupSettings
   sortPriority=40
   texture=&Black
  }
  n17 = Constant()
  n18 = If(
   A=Floor(None=ComponentMask(None=@n15) {r=true})
   B=Floor(None=ScalarParameter() {defaultValue=2, group=Decorator, parameterName=Decorator 3 UDIM, sortPriority=41})
   "A > B"=@n17
   "A == B"=Constant() {r=1}
   "A < B"=@n17
  )
  n19 = Add(
   A=Add(
    A=StaticSwitchParameter(True=@n7, False=@n8) {group=Decorator, parameterName=Decorator ID 1, sortPriority=33}
    B=StaticSwitchParameter(True=@n14, False=@n8) {group=Decorator, parameterName=Decorator ID 2, sortPriority=36}
   )
   B=StaticSwitchParameter(True=Multiply(A=@n16.RGBA, B=@n18), False=@n8) {group=Decorator, parameterName=Decorator ID 3, sortPriority=39}
  )
  n20 = TextureCoordinate() {coordinateIndex=1}
  n21 = TextureSampleParameter2D(UVs=@n20) {
   group=Decorator
   parameterName=Decorator 4
   samplerSource=SSM_Wrap_WorldGroupSettings
   sortPriority=43
   texture=&Black
  }
  n22 = Floor(None=ComponentMask(None=@n20) {r=true})
  n23 = Floor(None=ScalarParameter() {defaultValue=3, group=Decorator, parameterName=Decorator 4 UDIM, sortPriority=44})
  n24 = Constant()
  n25 = Constant() {r=1}
  n26 = StaticSwitchParameter(True=Multiply(A=@n21.RGBA, B=If(A=@n22, B=@n23, "A > B"=@n24, "A == B"=@n25, "A < B"=@n24)), False=@n8) {group=Decorator, parameterName=Decorator ID 4, sortPriority=42}
  n27 = TextureSampleParameter2D(UVs=TextureCoordinate()) {
   group=Decorator
   parameterName=Decorator 3DP
   samplerSource=SSM_Wrap_WorldGroupSettings
   sortPriority=43
   texture=&Black
  }
  n28 = Constant()
  n29 = ComponentMask(None=TextureCoordinate() {coordinateIndex=1}) {g=true}
  n30 = Constant()
  n31 = Multiply(
   A=Add(
    A=Add(A=@n19, B=@n26)
    B=StaticSwitchParameter(True=@n27.RGBA, False=@n28) {group=Decorator, parameterName=Has 3DP Decorator, sortPriority=42}
   )
   B=If(A=Floor(None=@n29), "A > B"=@n30, "A == B"=Constant() {r=1}, "A < B"=@n30)
  )
  n32 = ComponentMask(None=@n31) {b=true, g=true, r=true}
  n33 = ComponentMask(None=@n31) {a=true}
  n34 = LinearInterpolate(A=@n1.BaseColor, B=@n32, Alpha=@n33)
  n35 = LinearInterpolate(A=@n1.SubsurfaceColor, B=@n32, Alpha=@n33)
  n36 = LinearInterpolate(A=@n1.Opacity, Alpha=@n33)
  FunctionOutput(None=StaticSwitchParameter(
    True=SetMaterialAttributes!(MaterialAttributes=@n1.MaterialAttributes, BaseColor=@n34, SubsurfaceColor=@n35, Opacity=@n36) {attributeSetTypes [#1, #2, #3]}
    False=@n1.MaterialAttributes
   ) {group=Decorator, parameterName=Uses Decorators}) {outputName=Texture}
  FunctionOutput(None=StaticSwitchParameter(True=@n33, False=Constant()) {group=Decorator, parameterName=Uses Decorators}) {outputName=Mask, sortPriority=1}
 ]}
