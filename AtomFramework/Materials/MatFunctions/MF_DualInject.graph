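# Material function that adds an optional dual-injection look, blending a second color into material attributes and a separate Thin color output.
# The mask combines a local-Z gradient scaled by object bounds with a world-aligned texture fed local position, limiting front faces more than back faces.
# It drives base color, roughness and opacity, with a different opacity target for path tracing.
# The Dual Injection static switch passes both inputs through when off.
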MaterialFunction /AtomFramework/Materials/MatFunctions/MF_DualInject
graph {nodes [
  n1 = GetMaterialAttributes(None=FunctionInput() {bUsePreviewValueAsDefault=true, inputType=FunctionInput_MaterialAttributes}) {attributeGetTypes [#1, #2, #3]}
  n2 = NamedRerouteUsage()
  n3 = LinearInterpolate(A=@n1.Opacity, B=PathTracingQualitySwitch!(Normal=Constant() {r=0.25}, PathTraced=Constant() {r=0.5}), Alpha=@n2) {constA=1, constB=0.25}
  n4 = LinearInterpolate(A=@n1.Roughness, Alpha=@n2) {constA=1, constB=0.75}
  n5 = VectorParameter() {parameterName=Dual Col, defaultValue=<rgba 0.8333 0.3928 0.0761 1>}
  n6 = LinearInterpolate(A=@n1.BaseColor, B=@n5.RGB, Alpha=@n2)
  FunctionOutput(None=StaticSwitchParameter(
    True=SetMaterialAttributes!(MaterialAttributes=@n1.MaterialAttributes, Opacity=@n3, Roughness=@n4, BaseColor=@n6) {attributeSetTypes [#2, #3, #1]}
    False=@n1.MaterialAttributes
   ) {parameterName=Dual Injection})
  n7 = FunctionInput() {inputName=In_1}
  FunctionOutput(None=StaticSwitchParameter(True=LinearInterpolate(A=@n7, B=@n5.RGB, Alpha=@n2), False=@n7) {parameterName=Dual Injection}) {outputName=Thin}
  n8 = Saturate(None=TwoSidedSign())
  n9 = TransformPosition(None=WorldPosition().XYZ) {transformSourceType=TRANSFORMPOSSOURCE_World}
  n10 = ComponentMask(None=@n9) {b=true}
  n11 = ComponentMask(None=Multiply(A=ObjectBounds()) {constB=2}) {b=true}
  n12 = OneMinus(None=Saturate(None=Divide(A=@n10, B=@n11)))
  n13 = Subtract(A=Power(Base=@n12) {constExponent=0.5}) {constB=0.75}
  n14 = Saturate(None=Add(A=OneMinus(None=@n8), B=Saturate(None=@n13)))
  n15 = Multiply(
   A=MaterialFunctionCall(
    TextureObject=TextureObject() {samplerType=SAMPLERTYPE_Masks, texture=/AtomFramework/Materials/Textures/DualInjection_M}
    TextureSize=ScalarParameter() {defaultValue=147.3903, parameterName=Size}
    WorldPosition=@n9
   ) {materialFunction=&WorldAlignedTexture}.XYZ Texture
   B=@n12
  )
  n16 = Power(Base=Multiply(A=Add(A=@n15, B=@n12))) {constExponent=3}
  NamedRerouteDeclaration(None=Saturate(None=Multiply(A=@n14, B=@n16))) {name=Mask, nodeColor=<rgba 0 1 0.4141 1>}
 ]}
