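# MF_EdgeWarpMask is a material function that builds a mask rising toward the edges of each bounding-box face, named for edge warping.
# It centers 0-1 bounds UVW as Base Bounds, then remaps its absolute value to zero at center, one at sides.
# Each axis's absolute normal weights the larger of the other two distances; the maximum becomes Edge Mask.
# Inputs default to MF_InstanceBounds and the vertex normal in instance space.
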MaterialFunction /AtomFramework/Materials/MatFunctions/MF_EdgeWarpMask
graph {nodes [
  FunctionOutput(None=NamedRerouteUsage()) {outputName=Base Bounds}
  n1 = Subtract(
   A=FunctionInput(Preview=MaterialFunctionCall() {materialFunction=&/MF_InstanceBounds}.RGB) {bUsePreviewValueAsDefault=true, inputName=BoundingBoxBased_0-1_UVW, sortPriority=-7}
   B=Constant3Vector() {constant=<rgba 0.5 0.5 0.5 0>}
  )
  n2 = Multiply(A=Abs(None=NamedRerouteDeclaration(None=@n1) {name=Bounds, nodeColor=<rgba 0 1 0.4375 1>})) {constB=2}
  n3 = ComponentMask(None=@n2) {r=true}
  n4 = ComponentMask(None=@n2) {g=true}
  n5 = VertexNormalWS()
  n6 = Abs(None=FunctionInput(Preview=Transform(None=@n5) {transformSourceType=TRANSFORMSOURCE_World, transformType=TRANSFORM_Instance}) {bUsePreviewValueAsDefault=true, inputName=NormalWS, sortPriority=-7})
  n7 = Multiply(A=Max(A=@n3, B=@n4), B=ComponentMask(None=@n6) {b=true})
  n8 = ComponentMask(None=@n2) {b=true}
  n9 = Multiply(A=ComponentMask(None=@n6) {g=true}, B=Max(A=@n3, B=@n8))
  n10 = ComponentMask(None=@n6) {r=true}
  n11 = Max(A=@n4, B=@n8)
  FunctionOutput(None=Max(A=Max(A=@n7, B=@n9), B=Multiply(A=@n10, B=@n11))) {outputName=Edge Mask, sortPriority=1}
 ]}
