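# Material function that outputs a tangent-space normal warped at edges, with strength varied by a Seed input.
# MF_EdgeWarpMask supplies an edge mask, used as the U coordinate into the CA_EdgeBrick texture, whose red channel is centred around zero.
# That signed profile, scaled by a seed-blended strength between the WarpStrengthMix and WarpStrengthMax parameters, lerps a flat normal toward the Base Bounds vector transformed from instance to tangent space.
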MaterialFunction /AtomFramework/Materials/MatFunctions/MF_EdgeWarp_N
graph {nodes [
  n1 = MaterialFunctionCall!(BoundingBoxBased_0-1_UVW, NormalWS) {materialFunction=&/MF_EdgeWarpMask}
  n2 = Constant()
  n3 = Subtract(A=TextureSample(UVs=AppendVector(A=@n1.Edge Mask, B=@n2)) {samplerType=SAMPLERTYPE_LinearColor, texture=/AtomFramework/Materials/Textures/CA_EdgeBrick}.R) {constB=0.5}
  n4 = LinearInterpolate(
   A=ScalarParameter() {defaultValue=-0.216, group=Normal, parameterName=WarpStrengthMix}
   B=ScalarParameter() {defaultValue=0.1578, group=Normal, parameterName=WarpStrengthMax}
   Alpha=ComponentMask(None=FunctionInput() {bUsePreviewValueAsDefault=true, inputName=Seed}) {r=true}
  )
  FunctionOutput(None=LinearInterpolate(
    A=Constant3Vector() {constant=<rgba 0 0 1 0>}
    B=Transform(None=@n1.Base Bounds) {transformSourceType=TRANSFORMSOURCE_Instance, transformType=TRANSFORM_Tangent}
    Alpha=Multiply(A=@n3, B=@n4)
   )) {outputName=Normal}
 ]}
