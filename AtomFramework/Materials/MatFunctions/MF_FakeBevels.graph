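# Material function that fakes beveled edges by altering Normal and BaseColor of incoming material attributes, toggled by the UseFakeBevel static switch.
# Texture coordinates encode data: the fraction gives position, the integer part gives size.
# Remapped edge gradients, scaled by named reroute values, become derived-Z normals, plus a base-color highlight from the sky atmosphere light direction in tangent space.
# Lerps blend the results, with NaniteReplace providing separate Nanite outputs.
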MaterialFunction /AtomFramework/Materials/MatFunctions/MF_FakeBevels
graph {nodes [
  n1 = NamedRerouteUsage()
  n2 = NamedRerouteUsage()
  n3 = GetMaterialAttributes(None=FunctionInput() {inputType=FunctionInput_MaterialAttributes}) {attributeGetTypes [#1, #2, #3]}
  FunctionOutput(None=StaticSwitchParameter(
    True=SetMaterialAttributes!(MaterialAttributes=@n3.MaterialAttributes, Normal=@n1, BaseColor=@n2, Specular=@n3.Specular) {attributeSetTypes [#1, #2, #3]}
    False=@n3.MaterialAttributes
   ) {defaultValue=true, group=Fake Bevels, parameterName=UseFakeBevel, sortPriority=-1})
  n4 = Add(A=NamedRerouteDeclaration(None=TextureCoordinate()) {name=Tex Coord, nodeColor=<rgba 0 1 0.2266 1>}, B=NamedRerouteUsage())
  n5 = AppendVector(A=ComponentMask(None=@n4) {r=true}, B=OneMinus(None=ComponentMask(None=@n4) {g=true})) {desc=flip y}
  NamedRerouteDeclaration(None=Multiply(A=Frac(None=@n5)) {constB=2.56}) {name=UV, desc=pivot is Bottom left, nodeColor=<rgba 1 0 0.1641 1>}
  n6 = Constant2Vector() {g=-1, r=1}
  n7 = Add(A=Floor(None=@n4), B=@n6)
  NamedRerouteDeclaration(None=Multiply(A=Multiply(A=@n7) {constB=0.08, desc=old:0.16}, B=@n6)) {name=Size, nodeColor=<rgba 0.6875 1 0 1>}
  n8 = Multiply(A=ComponentMask(None=NamedRerouteUsage()) {g=true}) {constB=-1, desc=Flip Y}
  NamedRerouteDeclaration(None=Saturate(None=Sign(None=@n8))) {name=Is In Y Positive, nodeColor=<rgba 0.1719 1 0 1>}
  n9 = NamedRerouteUsage()
  n10 = SkyAtmosphereLightDirection()
  n11 = DotProduct(
   A=ComponentMask(None=Transform(None=@n10) {transformSourceType=TRANSFORMSOURCE_World, transformType=TRANSFORM_Tangent}) {g=true, r=true}
   B=ComponentMask(None=NamedRerouteUsage()) {g=true, r=true}
  )
  n12 = MaterialFunctionCall!(Base=NamedRerouteUsage(), Add=Multiply(A=Saturate(None=@n11)) {desc=Strenght}) {materialFunction=/Engine/Functions/Engine_MaterialFunctions02/Math/MultiplyAdd}
  n13 = NamedRerouteUsage()
  NamedRerouteDeclaration(None=NaniteReplace!(Default=LinearInterpolate(A=@n9, B=@n12.Result, Alpha=@n13), Nanite=NamedRerouteUsage())) {name=Base Color Out, nodeColor=<rgba 0 0.3906 1 1>}
  n14 = NamedRerouteUsage()
  n15 = NamedRerouteUsage()
  n16 = NamedRerouteUsage()
  NamedRerouteDeclaration(None=NaniteReplace(Default=LinearInterpolate(A=@n14, B=@n15, Alpha=@n16), Nanite=NamedRerouteUsage())) {name=Normal Out, nodeColor=<rgba 1 0 0.3281 1>}
  n17 = NamedRerouteUsage()
  n18 = NamedRerouteUsage()
  n19 = ComponentMask(None=NamedRerouteUsage()) {r=true}
  n20 = MaterialFunctionCall!(
   Input=Subtract(A=ComponentMask(None=@n17) {r=true}, B=ComponentMask(None=@n18) {r=true}) {constA=2.56, constB=2.56}
   Input Low=@n19
   Input High=Subtract(A=@n19, B=NamedRerouteUsage())
  ) {materialFunction=/Engine/Functions/Engine_MaterialFunctions03/Math/RemapValueRangeNormalized}
  n21 = Multiply(A=OneMinus(None=Saturate(None=@n20.Result)), B=NamedRerouteUsage()) {constB=0.5}
  n22 = ComponentMask(None=NamedRerouteUsage()) {r=true}
  n23 = MaterialFunctionCall(Input=ComponentMask(None=NamedRerouteUsage()) {r=true}, Input Low=@n22, Input High=Subtract(A=@n22, B=NamedRerouteUsage())) {materialFunction=&RemapValueRangeNormalized}
  n24 = Multiply(A=OneMinus(None=Saturate(None=@n23.Result)), B=NamedRerouteUsage()) {constB=0.5}
  n25 = AppendVector(A=Add(A=Multiply(A=@n21) {constB=-1}, B=@n24), B=Constant())
  n26 = Constant() {r=1}
  n27 = NamedRerouteUsage()
  n28 = NamedRerouteUsage()
  n29 = ComponentMask(None=NamedRerouteUsage()) {g=true}
  n30 = MaterialFunctionCall(
   Input=Subtract(A=ComponentMask(None=@n27) {g=true}, B=ComponentMask(None=@n28) {g=true}) {constA=2.56, constB=2.56}
   Input Low=@n29
   Input High=Subtract(A=@n29, B=NamedRerouteUsage())
  ) {materialFunction=&RemapValueRangeNormalized}
  n31 = Multiply(A=OneMinus(None=Saturate(None=@n30.Result)), B=NamedRerouteUsage()) {constB=0.5}
  n32 = ComponentMask(None=NamedRerouteUsage() {desc="0.32, 1.28"}) {g=true}
  n33 = MaterialFunctionCall(Input=ComponentMask(None=NamedRerouteUsage()) {g=true}, Input Low=@n32, Input High=Subtract(A=@n32, B=NamedRerouteUsage())) {materialFunction=&RemapValueRangeNormalized}
  n34 = Multiply(A=OneMinus(None=Saturate(None=@n33.Result)), B=NamedRerouteUsage()) {constB=0.5}
  n35 = AppendVector(A=Constant(), B=Add(A=@n31, B=Multiply(A=@n34) {constB=-1}))
  n36 = Constant() {r=1}
  NamedRerouteDeclaration(None=Add(
    A=MaterialFunctionCall!(XY vector=@n25, Z sign=@n26, Default Normalization vector) {materialFunction=/Engine/Functions/Engine_MaterialFunctions02/Math/DeriveNormalZ_Function}.Result
    B=MaterialFunctionCall(XY vector=@n35, Z sign=@n36) {materialFunction=&DeriveNormalZ_Function}.Result
   ) {desc=Normals}) {name=Normals, nodeColor=<rgba 0 1 0.8359 1>}
  n37 = NamedRerouteUsage()
  n38 = Constant4Vector() {constant=<rgba 8 1.5 16 32>}
  n39 = AppendVector(A=@n38.B, B=@n38.B)
  n40 = Frac(None=Multiply(A=Multiply(A=@n38.G, B=@n39), B=ScreenPosition().ViewportUV) {constB=24})
  n41 = Constant4Vector() {constant=<rgba 0 0.5 1.5 1>}
  n42 = Constant3Vector() {constant=<rgba 1 0 0 0>}
  n43 = Constant3Vector() {constant=<rgba 0 1 0 0>}
  n44 = NamedRerouteUsage()
  NamedRerouteDeclaration(None=Multiply(
    A=MaterialFunctionCall(Number=@n37, MaximumNumberOfDigits=@n38.R, UVs=@n40, DebugTextLocation RG_UpperRight BA_LowerLeft=@n41.RGBA) {materialFunction=&DebugScalarValues}.Result
    B=LinearInterpolate(A=@n42, B=@n43, Alpha=@n44)
   )) {name=Debug, nodeColor=<rgba 1 0.4219 0 1>}
  NamedRerouteDeclaration(None=LinearInterpolate(A=Constant2Vector() {g=-1}, B=Constant2Vector() {g=2}, Alpha=NamedRerouteUsage())) {name=Y Offset, nodeColor=<rgba 1 0 0.1641 1>}
  n45 = Constant() {r=0.015}
  n46 = Constant() {r=0.015}
  NamedRerouteDeclaration(None=QualitySwitch(Default=@n45, Low=@n46, High=@n45, Medium=@n46, Epic=@n45)) {name=Bevel Size, nodeColor=<rgba 1 0 0.2578 1>}
  NamedRerouteDeclaration(None=ScalarParameter() {defaultValue=0.5, group=Fake Bevels, parameterName=Bevel Strength, sliderMax=1, sortPriority=1}) {name=Bevel Strength, nodeColor=<rgba 0.4297 0 1 1>}
  n47 = Add(A=ComponentMask(None=TextureCoordinate()) {r=true}) {constB=-1}
  NamedRerouteDeclaration(None=Saturate(None=Sign(None=@n47))) {name=Exclusion Mask, nodeColor=<rgba 1 0.6563 0 1>}
  n48 = Saturate(None=@n21)
  n49 = Saturate(None=@n24)
  n50 = Saturate(None=@n31)
  n51 = Saturate(None=@n34)
  NamedRerouteDeclaration(None=Max(A=Max(A=@n48, B=@n49), B=Max(A=@n50, B=@n51)) {desc=Gradients}) {name=Gradients, nodeColor=<rgba 0 0.9531 1 1>}
  NamedRerouteDeclaration(None=@n3.Normal) {name=Normal In, nodeColor=<rgba 1 0 0.3516 1>}
  NamedRerouteDeclaration(None=@n3.BaseColor) {name=Base Color In, nodeColor=<rgba 0 1 0.8359 1>}
 ]}
