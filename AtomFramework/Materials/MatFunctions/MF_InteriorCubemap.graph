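# Material function for interior mapping, turning view direction and tiled UVs into a cubemap lookup vector for fake rooms behind a surface.
# It intersects the tangent-space view ray with a box in each UV tile, with depth adjusted by RoomZscale.
# A static switch parameter can replace mesh tangents with a box basis chosen from local axes by vertex-normal masks.
# Optional "Randomize Rotation" flips and swizzles the result by tile parity, without true randomness.
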MaterialFunction /AtomFramework/Materials/MatFunctions/MF_InteriorCubemap
graph {nodes [
  n1 = FunctionInput(Preview=CameraVectorWS()) {bUsePreviewValueAsDefault=true, inputName=View}
  n2 = NamedRerouteUsage()
  n3 = NamedRerouteUsage()
  n4 = NamedRerouteUsage()
  n5 = NamedRerouteUsage()
  n6 = NamedRerouteUsage()
  n7 = NamedRerouteUsage()
  n8 = LinearInterpolate(A=LinearInterpolate(A=LinearInterpolate(A=@n2, B=@n3, Alpha=@n4), B=@n2, Alpha=@n5), B=@n6, Alpha=@n7) {constB=0}
  n9 = NamedRerouteUsage()
  n10 = NamedRerouteUsage()
  n11 = NamedRerouteUsage()
  n12 = NamedRerouteUsage()
  n13 = LinearInterpolate(A=LinearInterpolate(A=LinearInterpolate(A=@n6, B=@n12, Alpha=@n4), B=@n12, Alpha=@n5), B=@n12, Alpha=@n7) {constB=0}
  n14 = NamedRerouteUsage()
  n15 = LinearInterpolate(A=LinearInterpolate(A=LinearInterpolate(A=@n14, B=@n2, Alpha=@n4), B=@n6, Alpha=@n5), B=@n9, Alpha=@n7) {constB=0}
  n16 = MaterialFunctionCall!(
   VectorToTransform=@n1
   BasisX=LinearInterpolate(A=LinearInterpolate(A=@n8, B=@n9, Alpha=@n10), B=@n9, Alpha=@n11)
   BasisY=LinearInterpolate(A=LinearInterpolate(A=@n13, B=@n12, Alpha=@n10), B=@n6, Alpha=@n11)
   BasisZ=LinearInterpolate(A=LinearInterpolate(A=@n15, B=@n3, Alpha=@n10), B=@n12, Alpha=@n11)
  ) {materialFunction=/Engine/Functions/Engine_MaterialFunctions02/Math/InverseTransformMatrix}
  n17 = Transform(None=@n1) {transformSourceType=TRANSFORMSOURCE_World, transformType=TRANSFORM_Tangent}
  n18 = Constant() {r=1}
  n19 = Constant() {r=1}
  n20 = Divide(B=FunctionInput(Preview=Constant() {r=1}) {bUsePreviewValueAsDefault=true, inputName=RoomZscale, inputType=FunctionInput_Scalar})
  n21 = Multiply(
   A=Multiply(
    A=StaticSwitchParameter(True=@n16.Transformed Vector 3, False=@n17) {parameterName=UseBoxTangentTransforms}
    B=MaterialFunctionCall(X=@n18, Y=@n19, Z=@n20) {materialFunction=&MakeFloat3}.Result
   )
   B=Constant3Vector() {constant=<rgba -1 1 1 0>}
  )
  n22 = Divide(B=@n21) {desc=id}
  n23 = NamedRerouteUsage()
  n24 = MaterialFunctionCall(Float3=Subtract(A=Abs(None=@n22), B=Multiply(A=@n22, B=@n23)) {desc=k}) {materialFunction=&BreakOutFloat3Components}
  n25 = Min(A=Min(A=@n24.R, B=@n24.R), B=@n24.B)
  n26 = MaterialFunctionCall(Float3=Add(A=Multiply(A=@n25, B=@n21), B=NamedRerouteUsage()) {desc="pos.xyz += kMin * In.dir;"}) {materialFunction=&BreakOutFloat3Components}
  n27 = MaterialFunctionCall(X=@n26.B, Y=@n26.B, Z=@n26.B) {materialFunction=&MakeFloat3}
  n28 = Constant3Vector() {constant=<rgba -1 1 1 0>}
  n29 = Constant3Vector() {constant=<rgba 1 -1 1 0>}
  n30 = FunctionInput() {bUsePreviewValueAsDefault=true, inputName=Tiling, inputType=FunctionInput_Vector2, sortPriority=1, previewValue=<4 4 0 1>}
  n31 = Multiply(A=FunctionInput(Preview=TextureCoordinate()) {bUsePreviewValueAsDefault=true, inputName=UVs, inputType=FunctionInput_Vector2}, B=@n30)
  n32 = MaterialFunctionCall(Float2=Ceil(None=@n31)) {materialFunction=&BreakOutFloat2Components}
  n33 = Constant() {r=2}
  n34 = Fmod(A=@n32.R, B=@n33)
  n35 = Constant3Vector() {constant=<rgba 1 1 1 0>}
  n36 = Constant3Vector() {constant=<rgba -1 1 1 0>}
  n37 = Fmod(A=@n32.G, B=@n33)
  n38 = Multiply(A=@n27.Result, B=Multiply(A=LinearInterpolate(A=@n28, B=@n29, Alpha=@n34), B=LinearInterpolate(A=@n35, B=@n36, Alpha=@n37)))
  n39 = MaterialFunctionCall(XYZ=@n38) {materialFunction=&Swizzle}
  n40 = FunctionInput(Preview=StaticBool()) {bUsePreviewValueAsDefault=true, inputName=Randomize Rotation, inputType=FunctionInput_StaticBool, sortPriority=2, previewValue=<4 4 0 1>}
  FunctionOutput(None=StaticSwitch(True=LinearInterpolate(A=@n38, B=@n39.YXZ, Alpha=@n37), False=@n27.Result, Value=@n40))
  n41 = Transform(None=Constant3Vector() {constant=<rgba 1 0 0 0>}) {transformSourceType=TRANSFORMSOURCE_Local}
  NamedRerouteDeclaration(None=@n41) {name=R+, nodeColor=<rgba 0.7344 1 0 1>}
  NamedRerouteDeclaration(None=Multiply(A=@n41) {constB=-1}) {name=R-, nodeColor=<rgba 0.9453 0 1 1>}
  n42 = Transform(None=VertexNormalWS()) {transformSourceType=TRANSFORMSOURCE_World, transformType=TRANSFORM_Local}
  n43 = ComponentMask(None=@n42) {g=true}
  n44 = ComponentMask(None=@n42) {r=true}
  n45 = Clamp(None=@n44)
  NamedRerouteDeclaration(None=Clamp(None=@n43, Max=@n45)) {name=G+ Mask, nodeColor=<rgba 0 1 0 1>}
  NamedRerouteDeclaration(None=@n45) {name=R+ Mask, nodeColor=<rgba 1 0.3047 0 1>}
  n46 = Transform(None=Constant3Vector() {constant=<rgba 0 1 0 0>}) {transformSourceType=TRANSFORMSOURCE_Local}
  NamedRerouteDeclaration(None=Multiply(A=NamedRerouteDeclaration(None=@n46) {name=G+, nodeColor=<rgba 0 1 0 1>}) {constB=-1}) {name=G-, nodeColor=<rgba 0 1 0 1>}
  n47 = Transform(None=Constant3Vector() {constant=<rgba 0 0 1 0>}) {transformSourceType=TRANSFORMSOURCE_Local}
  NamedRerouteDeclaration(None=@n47) {name=B+, nodeColor=<rgba 0 0 1 1>}
  NamedRerouteDeclaration(None=Multiply(A=@n47) {constB=-1}) {name=B-, nodeColor=<rgba 0 0 1 1>}
  n48 = Multiply(A=Frac(None=@n31) {desc=f}, B=Constant2Vector() {g=-2, r=2})
  n49 = Constant2Vector() {g=-1, r=1}
  NamedRerouteDeclaration(None=AppendVector(A=Subtract(A=@n48, B=@n49), B=Constant() {r=-1})) {name=Room UV, nodeColor=<rgba 0 0.5078 1 1>}
  n50 = Constant()
  n51 = Constant() {r=-1}
  NamedRerouteDeclaration(None=MaterialFunctionCall!(Input=ComponentMask(None=@n42) {b=true}, Input Low=@n50, Input High=@n51) {materialFunction=/Engine/Functions/Engine_MaterialFunctions03/Math/RemapValueRangeNormalized}.Result) {name=B- Mask, nodeColor=<rgba 0 0 1 1>}
  NamedRerouteDeclaration(None=MaterialFunctionCall(Input=@n43, Input Low=@n50, Input High=@n51) {materialFunction=&RemapValueRangeNormalized}.Result) {name=G- Mask, nodeColor=<rgba 0 1 0 1>}
  NamedRerouteDeclaration(None=MaterialFunctionCall(Input=@n44, Input Low=@n50, Input High=@n51) {materialFunction=&RemapValueRangeNormalized}.Result) {name=R- Mask, nodeColor=<rgba 1 0 0.2578 1>}
 ]}
