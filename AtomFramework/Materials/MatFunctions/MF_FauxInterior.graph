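# Material function that fakes the refractive interior of a translucent surface, overriding incoming attributes with black base color and emissive lighting.
# A refracted view ray, scaled for tangent or instance space, drives MF_InteriorCubemap mapping and an optional interior-normal cubemap.
# Tinted sky reflection and refraction, total-internal-reflection blending, Schlick-phase sun scatter and custom GGX internal highlights build the lighting.
# Mobile swaps in cheaper highlights plus sky ambient; original color is a second output.
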MaterialFunction /AtomFramework/Materials/MatFunctions/MF_FauxInterior
graph {nodes [
  n1 = NamedRerouteUsage()
  n2 = Constant()
  n3 = Multiply(A=NamedRerouteUsage(), B=NamedRerouteUsage())
  n4 = NamedRerouteUsage()
  n5 = DotProduct(A=NamedRerouteUsage(), B=SkyAtmosphereLightDirection())
  n6 = Multiply(
   A=Multiply(A=Multiply(A=@n3) {constB=0.5}, B=LinearInterpolate(Alpha=@n4) {constA=1, constB=0.5})
   B=LinearInterpolate(Alpha=Abs(None=@n5)) {constA=0.5}
  )
  n7 = Multiply(A=@n6, B=SkyLightEnvMapSample(Direction=VertexNormalWS(), Roughness=Constant() {r=1}))
  n8 = Add(A=NamedRerouteUsage(), B=NamedRerouteUsage()) {constB=0}
  n9 = NamedRerouteUsage()
  n10 = SetMaterialAttributes!(
   MaterialAttributes=@n1
   BaseColor=Constant()
   Metallic=Constant()
   Roughness=Constant() {r=0.15}
   EmissiveColor=Add(A=ShadingPathSwitch(Default=@n2, Mobile=@n7), B=Add(A=@n8, B=@n9)) {constB=0}
   Opacity=Constant() {r=0.7}
   SubsurfaceColor=@n6
  ) {attributeSetTypes [
    #1
    #2
    #3
    #4
    #5
    #6
   ]}
  FunctionOutput(None=@n10)
  FunctionOutput(None=GetMaterialAttributes(None=@n1) {attributeGetTypes [#1]}.BaseColor) {outputName=OrigColor, sortPriority=1}
  n11 = ScalarParameter() {group=Translucency - Faux Interior, parameterName=RoomUOffset}
  n12 = ScalarParameter() {group=Translucency - Faux Interior, parameterName=RoomVOffset}
  n13 = StaticSwitchParameter(
   True=MaterialFunctionCall() {materialFunction=&/MF_BoundsCubeUV}.Result
   False=Add(A=TextureCoordinate(), B=AppendVector(A=@n11, B=@n12))
  ) {parameterName=UseBoxTangentTransforms}
  n14 = ScalarParameter() {defaultValue=1, group=Translucency - Faux Interior, parameterName=RoomUVScale}
  n15 = StaticSwitchParameter(True=@n14, False=Multiply(A=@n14, B=NamedRerouteUsage())) {group=Translucency - Faux Interior, parameterName=ISM}
  NamedRerouteDeclaration(None=MaterialFunctionCall!(
    UVs=MaterialFunctionCall(UVs=@n13, Texture Scale=@n15) {materialFunction=&ScaleUVsByCenter}.UVs
    View=NamedRerouteUsage()
    RoomZscale=ScalarParameter() {defaultValue=1, group=Translucency - Faux Interior, parameterName=RoomDepth}
    Tiling=ScalarParameter() {defaultValue=1, group=Translucency - Faux Interior, parameterName=RoomTile}
    Randomize Rotation
   ) {materialFunction=&/MF_InteriorCubemap}.Result) {name=UVW, nodeColor=<rgba 0 0.4844 1 1>}
  n16 = TextureSampleParameterCube!(UVs=NamedRerouteUsage(), Apply View MipBias) {
   group=Translucency - Faux Interior
   parameterName=InteriorNormals
   samplerType=SAMPLERTYPE_LinearColor
   texture=/AtomFramework/Materials/Textures/RT_InteriorNormals_Tex
  }
  n17 = Normalize(VectorInput=Subtract(A=Multiply(A=@n16.RGB) {constB=2}))
  NamedRerouteDeclaration(None=StaticSwitchParameter(True=Transform(None=@n17), False=VertexNormalWS()) {group=Translucency - Faux Interior, parameterName=UseInteriorNormal}) {name=InteriorNormal, nodeColor=<rgba 1 0.3984 0 1>}
  MaterialExpressionNormalize_2 = Normalize(VectorInput=SkyAtmosphereLightDirection())
  MaterialExpressionCameraVectorWS_1 = CameraVectorWS()
  MaterialExpressionNamedRerouteUsage_1 = NamedRerouteUsage()
  MaterialExpressionScalarParameter_9 = ScalarParameter() {defaultValue=0.3, group=Translucency - Faux Interior, parameterName=FakeTransSunRoughness}
  MaterialExpressionConstant3Vector_3 = Constant3Vector() {constant=<rgba 1 1 1 0>}
  n18 = Custom!(
   L=@MaterialExpressionNormalize_2
   V=@MaterialExpressionCameraVectorWS_1
   N=@MaterialExpressionNamedRerouteUsage_1
   roughness=@MaterialExpressionScalarParameter_9
   F0=@MaterialExpressionConstant3Vector_3
  ) {
   code="    float3 H = normalize(L + V);\r\n    float NdotH = saturate(dot(N, H));\r\n    float NdotL = saturate(dot(N, L));\r\n    float NdotV = saturate(dot(N, V));\r\n    float3 F = F0 + (1.0 - F0) * pow(1.0 - NdotV, 5.0);\r\n    float epsilon = .0001;\r\n\r\n\tfloat alpha = roughness * roughness;\r\n    float alphaSqr = alpha * alpha;\r\n\r\n    float denominator = NdotH * NdotH * (alphaSqr - 1.0) + 1.0;\r\n    float D = alphaSqr / (PI * denominator * denominator + epsilon);\r\n\t\r\n\t\r\n\t\r\n    float k = alpha / 2;    \r\n    float g_v = NdotV / ( NdotV*(1 - k ) + k);\r\n    float g_l = NdotL / (NdotL*(1 - k ) + k);\r\n    float G = g_v * g_l;\r\n    return NdotL*(D * G * F) / (4.0 * NdotL * NdotV + epsilon );"
   description=GGX
   inputs [
    {inputName=L, input {expression=&self:MaterialExpressionNormalize_2}}
    {inputName=V, input {expression=&self:MaterialExpressionCameraVectorWS_1}}
    {inputName=N, input {expression=&self:MaterialExpressionNamedRerouteUsage_1}}
    {inputName=roughness, input {expression=&self:MaterialExpressionScalarParameter_9}}
    {
     inputName=F0
     input {expression=&self:MaterialExpressionConstant3Vector_3, mask=1, maskB=1, maskG=1, maskR=1}
    }
   ]
  }
  n19 = Multiply(A=@n18) {constA=1}
  n20 = DotProduct(A=ReflectionVectorWS(CustomWorldNormal=@MaterialExpressionNamedRerouteUsage_1), B=@MaterialExpressionNormalize_2)
  n21 = Multiply(A=Power(Base=Saturate(None=@n20)) {constExponent=24}) {constB=5}
  n22 = NamedRerouteUsage()
  n23 = Multiply(A=Clamp(None=ShadingPathSwitch(Default=@n19, Mobile=@n21)) {maxDefault=40}, B=Multiply(A=@n22, B=@n22))
  n24 = NamedRerouteUsage()
  n25 = NamedRerouteUsage()
  NamedRerouteDeclaration(None=Multiply(A=Multiply(A=@n23, B=@n24), B=LinearInterpolate(Alpha=@n25) {constA=0.5, constB=0.05})) {name=SunInternalReflections, nodeColor=<rgba 0 0.8828 1 1>}
  n26 = GetMaterialAttributes(None=FunctionInput() {inputName=InMaterial, inputType=FunctionInput_MaterialAttributes}) {attributeGetTypes [#7, #1]}
  NamedRerouteDeclaration(None=@n26.Normal) {name=Normal, nodeColor=<rgba 1 0.5859 0 1>}
  n27 = Multiply(A=SkyLightEnvMapSample(
    Direction=NamedRerouteUsage()
    Roughness=ScalarParameter() {defaultValue=0.2, group=Translucency - Faux Interior, parameterName=ReflectRoughness}
   )) {constB=0.75}
  n28 = NamedRerouteUsage()
  n29 = ScalarParameter() {defaultValue=0.2, group=Translucency - Faux Interior, parameterName=RefractRoughness}
  n30 = Multiply(A=Multiply(A=SkyLightEnvMapSample(Direction=@n28, Roughness=@n29)) {constB=1.2}, B=NamedRerouteUsage())
  n31 = NamedRerouteUsage()
  NamedRerouteDeclaration(None=Multiply(A=NamedRerouteUsage(), B=LinearInterpolate(A=@n27, B=@n30, Alpha=@n31))) {name=Sky Refractions, nodeColor=<rgba 0.6875 1 0 1>}
  n32 = Constant() {r=1}
  n33 = ComponentMask(None=NamedRerouteUsage()) {b=true}
  n34 = Saturate(None=OneMinus(None=Multiply(A=@n33) {constB=-1}))
  NamedRerouteDeclaration(None=MaterialFunctionCall!(1=@n32, 2=Constant() {r=0.5}, 3=@n32, 4, A=Multiply(A=@n34, B=@n34)) {materialFunction=/Engine/Functions/Engine_MaterialFunctions02/Lerp_Multiple_Float}.Lerp 3 Inputs) {name=GroundGradient, nodeColor=<rgba 0.4062 0 1 1>}
  n35 = Abs(None=ComponentMask(None=TextureCoordinate() {coordinateIndex=1}) {g=true})
  NamedRerouteDeclaration(None=Multiply(A=Floor(None=@n35)) {constB=0.1562}) {name=Auto Depth, nodeColor=<rgba 0.0547 1 0 1>}
  NamedRerouteDeclaration(None=@n26.MaterialAttributes) {name=InMaterial, nodeColor=<rgba 1 0.5859 0 1>}
  NamedRerouteDeclaration(None=@n26.BaseColor) {name=InColor, nodeColor=<rgba 0 0.2266 1 1>}
  n36 = Abs(None=TextureCoordinate())
  NamedRerouteDeclaration(None=Multiply(A=Floor(None=@n36)) {constB=0.0312}) {name=Auto UV, nodeColor=<rgba 0 0.3621 1 1>}
  n37 = CameraVectorWS()
  n38 = MaterialFunctionCall!(
   Ray Direction=Normalize(VectorInput=Multiply(A=@n37) {constB=-1})
   Surface Normal=Transform(None=NamedRerouteUsage())
   Refractive Index Origin=Constant() {r=1}
   Refractive Index Target=Constant() {r=1.5}
  ) {materialFunction=/Engine/Functions/Engine_MaterialFunctions01/Vectors/Refract}
  NamedRerouteDeclaration(None=Multiply(A=NamedRerouteDeclaration(None=@n38.Vector) {name=RefractVectorSurface, nodeColor=<rgba 0.2188 0 1 1>}) {constB=-1}) {name=Ray Direction, nodeColor=<rgba 1 0 0.9375 1>}
  n39 = Custom!(None) {code="GetPrimitiveData(Parameters).InstanceLocalBoundsExtent.xyz", description=BoundsExtent}
  n40 = MaterialFunctionCall() {materialFunction=&ObjectScale}
  n41 = Divide(
   A=Transform(None=NamedRerouteUsage()) {transformSourceType=TRANSFORMSOURCE_World, transformType=TRANSFORM_Instance}
   B=Normalize(VectorInput=Multiply(A=@n39, B=@n40.Scale XYZ))
  )
  n42 = Multiply(A=NamedRerouteUsage()) {constB=3.2}
  n43 = Divide(
   A=Transform(None=NamedRerouteUsage()) {transformSourceType=TRANSFORMSOURCE_World, transformType=TRANSFORM_Tangent}
   B=MaterialFunctionCall(X=ComponentMask(None=@n42) {r=true}, Y=ComponentMask(None=@n42) {g=true}, Z=NamedRerouteUsage()) {materialFunction=&MakeFloat3}.Result
  )
  NamedRerouteDeclaration(None=StaticSwitchParameter(True=Transform(None=@n41) {transformSourceType=TRANSFORMSOURCE_Instance}, False=Transform(None=@n43)) {group=Translucency - Faux Interior, parameterName=ISM}) {name=Ray Direction Scaled, nodeColor=<rgba 0.5938 0 1 1>}
  n44 = StaticSwitchParameter(True=NamedRerouteUsage(), False=CameraVectorWS()) {group=Translucency - Faux Interior, parameterName=UseInteriorNormal}
  n45 = Multiply(A=DotProduct(A=Multiply(A=@n44) {constB=-1}, B=NamedRerouteUsage())) {constB=-1}
  n46 = Constant() {r=0.73}
  n47 = Constant() {r=0.78}
  NamedRerouteDeclaration(None=LinearInterpolate(Alpha=MaterialFunctionCall!(Input=@n45, Input Low=@n46, Input High=@n47) {materialFunction=/Engine/Functions/Engine_MaterialFunctions03/Math/RemapValueRangeNormalized}.Result) {constA=0.1, constB=0.9}) {name=TIR, nodeColor=<rgba 0 0.7422 1 1>}
  NamedRerouteDeclaration(None=MaterialFunctionCall(
    Ray Direction=NamedRerouteUsage()
    Surface Normal=NamedRerouteUsage()
    Refractive Index Origin=Constant() {r=1}
    Refractive Index Target=Constant() {r=1.5}
   ) {materialFunction=&Refract}.Vector) {name=RefractVectorInterior, nodeColor=<rgba 1 0 0.9609 1>}
  n48 = ScalarParameter() {defaultValue=0.5, group=Translucency - Faux Interior, parameterName=GradientLow}
  n49 = ScalarParameter() {defaultValue=1, group=Translucency - Faux Interior, parameterName=GradientHi}
  n50 = MaterialFunctionCall(Input=NamedRerouteUsage(), Input Low=Constant() {r=-1}, Input High=Constant() {r=1}) {materialFunction=&RemapValueRangeNormalized}
  NamedRerouteDeclaration(None=StaticSwitchParameter(True=LinearInterpolate(A=@n48, B=@n49, Alpha=@n50.Result) {constA=0.45}, False=Constant() {r=1}) {group=Translucency - Faux Interior, parameterName=UseInteriorNormal}) {name=Depth Gradient, nodeColor=<rgba 1 0 0.2812 1>}
  n51 = NamedRerouteUsage()
  NamedRerouteDeclaration(None=Desaturation(None=FunctionInput(Preview=@n51) {bUsePreviewValueAsDefault=true, inputName=InColor}, Fraction=Constant() {r=-0.2})) {name=Tint, nodeColor=<rgba 0.3125 1 0 1>}
  NamedRerouteDeclaration(None=ReflectionVectorWS(CustomWorldNormal=NamedRerouteUsage())) {name=ReflectionVector, nodeColor=<rgba 1 0 0.3281 1>}
  n52 = Constant3Vector() {constant=<rgba 0 0 100 0>}
  NamedRerouteDeclaration(None=Multiply(A=SkyAtmosphereLightIlluminance!(World Position=@n52)) {constB=0.5}) {name=DirectionalLightColor, nodeColor=<rgba 0 0.4844 1 1>}
  n53 = DotProduct(A=SkyAtmosphereLightDirection(), B=NamedRerouteUsage())
  n54 = ScalarParameter() {defaultValue=-0.985, group=Translucency - Faux Interior, parameterName=Phase}
  n55 = Clamp(None=Multiply(
    A=NamedRerouteUsage()
    B=MaterialFunctionCall!(Cosine=@n53, Anisotropy=@n54) {materialFunction=/Engine/Functions/Engine_MaterialFunctions01/Volumetrics/SchlickPhase}.Result
   )) {maxDefault=20}
  n56 = NamedRerouteUsage()
  n57 = NamedRerouteUsage()
  NamedRerouteDeclaration(None=Multiply(A=Multiply(A=@n55, B=@n56) {constB=2}, B=LinearInterpolate(Alpha=@n57) {constA=0.2})) {name=SunNscatter, nodeColor=<rgba 1 0.7031 0 1>}
  NamedRerouteUsage()
 ]}
