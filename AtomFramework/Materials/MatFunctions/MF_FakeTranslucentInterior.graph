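# Material function that gives incoming material attributes a fake translucent interior look, with optional glitter or opal effects.
# The input feeds MF_FauxInterior, whose result branches into MF_Glitter_NT and MF_Primitive_OpalNT.
# Translucency-group static switches choose the output: IsTranslucent off passes the input through unchanged.
# When on, Glitter takes precedence over Opal, and with both off the plain faux interior is used.
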MaterialFunction /AtomFramework/Materials/MatFunctions/MF_FakeTranslucentInterior
graph {nodes [
  n1 = FunctionInput() {bUsePreviewValueAsDefault=true, inputType=FunctionInput_MaterialAttributes}
  n2 = MaterialFunctionCall!(InMaterial=@n1, InColor) {materialFunction=&/MF_FauxInterior}
  n3 = MaterialFunctionCall!(In=@n2.Result) {materialFunction=&/MF_Glitter_NT}
  n4 = StaticSwitchParameter(
   True=MaterialFunctionCall!(In=@n2.Result) {materialFunction=&/MF_Primitive_OpalNT}.Result
   False=@n2.Result
  ) {group=Translucency, parameterName=Opal, sortPriority=2}
  FunctionOutput(None=StaticSwitchParameter(True=StaticSwitchParameter(True=@n3.Result, False=@n4) {group=Translucency, parameterName=Glitter, sortPriority=1}, False=@n1) {group=Translucency, parameterName=IsTranslucent})
 ]}
