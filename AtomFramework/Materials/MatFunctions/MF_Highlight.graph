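# MF_Highlight is a material function that overlays an optional blue (by default) highlight on incoming Material Attributes, gated by a static switch.
# A screen-space grid with 32-pixel cells plus the tint forms the emissive, faded by a bottom-to-top gradient from instance position and bounds.
# Sine of time pulses the gradient height, which also masks the attribute blend alongside a strength parameter.
# A Fresnel tint branch is left unconnected.
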MaterialFunction /AtomFramework/Materials/MatFunctions/MF_Highlight
graph {nodes [
  n1 = FunctionInput(Preview=SetMaterialAttributes!(MaterialAttributes)) {bUsePreviewValueAsDefault=true, inputName=Mat In, inputType=FunctionInput_MaterialAttributes}
  n2 = VectorParameter() {defaultValue=<rgba 0.0093 0.335 0.5938 1>}
  n3 = ScreenPosition()
  n4 = MaterialFunctionCall(Float2=Frac(None=Divide(A=@n3.PixelPosition) {constB=32})) {materialFunction=&BreakOutFloat2Components}
  n5 = Abs(None=ConstantBiasScale(None=@n4.R) {bias=-0.5, scale=2})
  n6 = Abs(None=ConstantBiasScale(None=@n4.G) {bias=-0.5, scale=2})
  n7 = Saturate(None=MaterialFunctionCall(Input=Max(A=@n5, B=@n6), Input Low=Constant() {r=0.9}) {materialFunction=&RemapValueRange}.Result)
  n8 = LinearInterpolate(A=@n2.RGB, B=@n2.RGB) {constAlpha=0.01}
  n9 = WorldPosition()
  n10 = ComponentMask(None=TransformPosition(None=Constant3Vector()) {transformSourceType=TRANSFORMPOSSOURCE_Instance, transformType=TRANSFORMPOSSOURCE_World}) {b=true}
  n11 = ComponentMask(None=Bounds().Half Extents) {b=true}
  n12 = Multiply(A=@n11) {constB=2}
  n13 = Sine(None=Time())
  n14 = Divide(A=Add(A=Subtract(A=@n9.Z, B=@n10), B=@n11), B=Add(A=@n12, B=Multiply(A=@n13, B=@n12))) {constB=32}
  n15 = Power(Base=OneMinus(None=Saturate(None=@n14)))
  n16 = SetMaterialAttributes!(MaterialAttributes, BaseColor=@n2.RGB, EmissiveColor=Multiply(A=Add(A=@n7, B=@n8), B=@n15)) {attributeSetTypes [#1, #2]}
  n17 = LinearInterpolate(B=Multiply(A=Saturate(None=@n15)) {constB=0.5}, Alpha=ScalarParameter() {parameterName=Highlight Layer Strength, sliderMax=1}) {constB=0}
  FunctionOutput(None=StaticSwitchParameter(True=BlendMaterialAttributes(A=@n1, B=@n16, Alpha=@n17), False=@n1) {parameterName=Use Highlighter})
  Multiply(A=Multiply(A=Fresnel() {baseReflectFraction=0, exponent=2}) {constB=0.2}, B=@n2.RGB)
 ]}
