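# Material function that applies an optional opacity mask with fake depth to incoming material attributes, toggled by the Mask static switch.
# It takes the maximum of the mask texture and two parallax-shifted BumpOffset samples at full and half Depth Distance, extruding the cutout.
# Where the unshifted mask is dark, it darkens base and subsurface color and bends normals toward the camera vector to shade the faked thickness.
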MaterialFunction /AtomFramework/Materials/MatFunctions/MF_MaskFDepth
graph {nodes [
  n1 = FunctionInput() {inputType=FunctionInput_MaterialAttributes}
  n2 = GetMaterialAttributes(None=@n1) {attributeGetTypes [#1, #2]}
  n3 = TextureCoordinate()
  n4 = TextureSampleParameter2D(UVs=@n3) {group=Mask, parameterName=Mask Texture, samplerType=SAMPLERTYPE_Masks, texture=/AtomFramework/Materials/Textures/FlatUndersides_Shape_Mask}
  n5 = OneMinus(None=@n4.RGB)
  n6 = ScalarParameter() {group=Mask, parameterName=Darken Fake Depth, sliderMax=1}
  n7 = Multiply(A=@n2.BaseColor, B=OneMinus(None=Multiply(A=@n5, B=@n6) {constB=0.8}))
  n8 = Constant()
  n9 = ScalarParameter() {defaultValue=0.09, group=Mask, parameterName=Depth Distance}
  n10 = TextureSampleParameter2D(UVs=BumpOffset(Coordinate=@n3, Height=@n8, HeightRatioInput=@n9) {heightRatio=0.09, referencePlane=1}) {group=Mask, parameterName=Mask Texture, samplerType=SAMPLERTYPE_Masks, texture=&FlatUndersides_Shape_Mask}
  n11 = TextureSampleParameter2D(UVs=BumpOffset(Coordinate=@n3, Height=@n8, HeightRatioInput=Divide(A=@n9)) {heightRatio=0.045, referencePlane=1}) {group=Mask, parameterName=Mask Texture, samplerType=SAMPLERTYPE_Masks, texture=&FlatUndersides_Shape_Mask}
  n12 = MaterialFunctionCall() {materialFunction=&CameraDirectionVector}
  n13 = SetMaterialAttributes!(
   MaterialAttributes=@n1
   BaseColor=@n7
   SubsurfaceColor=@n7
   OpacityMask=Max(A=@n4.RGB, B=Max(A=@n10.R, B=@n11.R))
   Normal=LinearInterpolate(
    A=@n2.Normal
    B=Transform(None=@n12.Result) {transformSourceType=TRANSFORMSOURCE_World, transformType=TRANSFORM_Tangent}
    Alpha=Multiply(A=@n5) {constB=0.5}
   )
  ) {attributeSetTypes [
    #1
    #3
    #4
    #2
   ]}
  FunctionOutput(None=StaticSwitchParameter(True=@n13, False=@n1) {group=Mask, parameterName=Mask})
  Multiply(A=@n12.Result)
 ]}
