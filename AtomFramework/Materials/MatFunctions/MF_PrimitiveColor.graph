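# Material function that outputs a primitive's base color, selected by static switches.
# Color comes from a parameter, vertex colors, per-instance packed 8-bit RGB converted to linear, or a LUT atlas lookup.
# The LUT column index comes from an override, vertex red, dynamic parameter or packed bits; the row comes from instance or primitive data.
# The result is range-remapped, varied in saturation and brightness by a Random Seed input, then Fresnel-darkened.
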MaterialFunction /AtomFramework/Materials/MatFunctions/MF_PrimitiveColor
graph {nodes [
  n1 = VertexColor()
  n2 = Multiply(A=Add(A=@n1.R, B=@n1.R)) {constB=255}
  MaterialExpressionPerInstanceCustomData_3 = PerInstanceCustomData!(DefaultValue=ScalarParameter() {bUseCustomPrimitiveData=true, defaultValue=16053492, parameterName=Color}) {constDefaultValue=16053492}
  MaterialExpressionConstant_2 = Constant() {r=9}
  MaterialExpressionConstant_3 = Constant()
  n3 = Custom!(Scalar=@MaterialExpressionPerInstanceCustomData_3, BitCount=@MaterialExpressionConstant_2, BitIndex=@MaterialExpressionConstant_3) {
   code="return BitFieldExtractU32(asuint(Scalar), BitCount, BitIndex);"
   desc="Index (0-8) - 9 bits"
   description=UnpackBit
   outputType=CMOT_Float1
   inputs [
    {inputName=Scalar, input {expression=&self:MaterialExpressionPerInstanceCustomData_3}}
    {inputName=BitCount, input {expression=&self:MaterialExpressionConstant_2}}
    {inputName=BitIndex, input {expression=&self:MaterialExpressionConstant_3}}
   ]
  }
  n4 = StaticSwitchParameter(True=DynamicParameter() {paramNames [ColorIndex,Param2,Param3,Param4]}.ColorIndex, False=@n3) {desc=Use a dynamic material parameter for the index value for texture lookup., group=1 - Col, parameterName=Use Dynamic Parameter}
  n5 = StaticSwitchParameter(True=NaniteReplace!(Default=VertexInterpolator(VS=@n2).PS, Nanite=@n2), False=@n4) {
   desc="Whether or not to source the color information from a packed float per-instance (false) or standard vertex colors (true). Defaults to false. Only available if Override Color is false."
   group=1 - Col
   parameterName=Use Vertex Colors
  }
  n6 = StaticSwitchParameter(True=ScalarParameter() {desc=The manual index value for the color lookup table, group=1 - Col, parameterName=Override Index}, False=@n5) {
   desc=Whether or not the user has manually specified the color index as a material parameter
   group=1 - Col
   parameterName=Use Override Index Parameter
  }
  n7 = Add(A=Divide(A=Multiply(A=@n6)) {constB=512}, B=Divide(A=Constant() {r=0.5}) {constB=512})
  n8 = ScalarParameter() {bUseCustomPrimitiveData=true, group=1 - Col, parameterName=LUTIndex, primitiveDataIndex=3}
  n9 = TextureObjectParameter() {
   desc=The texture to use as the look-up texture atlas
   group=1 - Col
   parameterName=Look Up Table Atlas Texture
   texture=/AtomFramework/Materials/Textures/T_LUT_DefautAtlas
  }
  n10 = Multiply(
   A=Add(A=PerInstanceCustomData(DefaultValue=@n8) {dataIndex=4}) {constB=0.5}
   B=ComponentMask(None=TextureProperty(Tex=@n9) {property=TMTM_TexelSize}) {g=true}
  )
  MaterialExpressionConstant_8 = Constant() {r=8}
  MaterialExpressionConstant_11 = Constant() {r=16}
  n11 = Custom(Scalar=@MaterialExpressionPerInstanceCustomData_3, BitCount=@MaterialExpressionConstant_8, BitIndex=@MaterialExpressionConstant_11) {
   code="return BitFieldExtractU32(Scalar, BitCount, BitIndex);"
   desc="R (16-23)"
   description=UnpackBit
   outputType=CMOT_Float1
   inputs [
    {inputName=Scalar, input {expression=&self:MaterialExpressionPerInstanceCustomData_3}}
    {inputName=BitCount, input {expression=&self:MaterialExpressionConstant_8}}
    {inputName=BitIndex, input {expression=&self:MaterialExpressionConstant_11}}
   ]
  }
  MaterialExpressionConstant_12 = Constant() {r=8}
  MaterialExpressionConstant_13 = Constant() {r=8}
  n12 = Custom(Scalar=@MaterialExpressionPerInstanceCustomData_3, BitCount=@MaterialExpressionConstant_12, BitIndex=@MaterialExpressionConstant_13) {
   code="return BitFieldExtractU32(Scalar, BitCount, BitIndex);"
   desc="G (8-15)"
   description=UnpackBit
   outputType=CMOT_Float1
   inputs [
    {inputName=Scalar, input {expression=&self:MaterialExpressionPerInstanceCustomData_3}}
    {inputName=BitCount, input {expression=&self:MaterialExpressionConstant_12}}
    {inputName=BitIndex, input {expression=&self:MaterialExpressionConstant_13}}
   ]
  }
  n13 = AppendVector(
   A=MaterialFunctionCall!(sRGB=Divide(A=@n11) {constB=255}) {materialFunction=/Engine/Functions/Engine_MaterialFunctions01/ImageAdjustment/sRGBToLinear}.Linear
   B=MaterialFunctionCall(sRGB=Divide(A=@n12) {constB=255}) {materialFunction=&sRGBToLinear}.Linear
  )
  MaterialExpressionConstant_14 = Constant() {r=8}
  MaterialExpressionConstant_15 = Constant()
  n14 = Custom(Scalar=@MaterialExpressionPerInstanceCustomData_3, BitCount=@MaterialExpressionConstant_14, BitIndex=@MaterialExpressionConstant_15) {
   code="return BitFieldExtractU32(Scalar, BitCount, BitIndex);"
   desc="B (0-7)"
   description=UnpackBit
   outputType=CMOT_Float1
   inputs [
    {inputName=Scalar, input {expression=&self:MaterialExpressionPerInstanceCustomData_3}}
    {inputName=BitCount, input {expression=&self:MaterialExpressionConstant_14}}
    {inputName=BitIndex, input {expression=&self:MaterialExpressionConstant_15}}
   ]
  }
  n15 = MaterialFunctionCall(sRGB=Divide(A=@n14) {constB=255}) {materialFunction=&sRGBToLinear}
  n16 = StaticSwitchParameter(True=VertexColor(), False=VertexInterpolator(VS=AppendVector(A=@n13, B=@n15.Linear)).PS) {
   desc="Whether or not to source the color information from a packed float per-instance (false) or standard vertex colors (true). Defaults to false. Only available if Override Color is false."
   group=1 - Col
   parameterName=Use Vertex Colors
  }
  n17 = StaticSwitchParameter(True=VectorParameter() {group=1 - Col, parameterName=Color, defaultValue=<rgba 1 1 1 1>}.RGB, False=@n16) {
   desc="Whether or not to source the color from the Color parameter (if true) or from a PerVertex/PerInstance location (if false). Defaults to false."
   group=1 - Col
   parameterName=Override Color
  }
  n18 = StaticSwitchParameter(True=TextureSample(UVs=AppendVector(A=@n7, B=@n10), Tex=@n9) {texture=&T_LUT_DefautAtlas}.RGB, False=@n17) {
   desc=Whether or not to source the color from the Index Color Texture. This requries a reference to a LUT of textures and an index texture to define which color to use.
   group=1 - Col
   parameterName=Index Color
  }
  n19 = ComponentMask(None=@n18) {r=true}
  n20 = Constant() {r=0.02}
  n21 = Constant() {r=0.95}
  n22 = ComponentMask(None=@n18) {g=true}
  n23 = Constant() {r=0.02}
  n24 = Constant() {r=0.95}
  n25 = ComponentMask(None=@n18) {b=true}
  n26 = Constant() {r=0.02}
  n27 = Constant() {r=0.95}
  n28 = FunctionInput(Preview=Constant()) {bUsePreviewValueAsDefault=true, inputName=Random Seed, inputType=FunctionInput_Vector2}
  n29 = Desaturation(
   None=MaterialFunctionCall(
    X=MaterialFunctionCall(Input=@n19, Target Low=@n20, Target High=@n21) {materialFunction=&RemapValueRange}.Result
    Y=MaterialFunctionCall(Input=@n22, Target Low=@n23, Target High=@n24) {materialFunction=&RemapValueRange}.Result
    Z=MaterialFunctionCall(Input=@n25, Target Low=@n26, Target High=@n27) {materialFunction=&RemapValueRange}.Result
   ) {materialFunction=&MakeFloat3}.Result
   Fraction=Multiply(A=ComponentMask(None=@n28) {r=true}, B=ScalarParameter() {group=1 - Col, parameterName=VarySat})
  ) {luminanceFactors=<rgba 1 1 1 0>}
  n30 = Constant() {r=1}
  n31 = ScalarParameter() {defaultValue=0.05, group=1 - Col, parameterName=VaryValue}
  n32 = LinearInterpolate(A=Subtract(A=@n30, B=@n31), B=Add(A=@n30, B=@n31), Alpha=ComponentMask(None=@n28) {g=true}) {constA=0.75, constB=1.25}
  n33 = Fresnel() {baseReflectFraction=0}
  FunctionOutput(None=Multiply(A=Multiply(A=@n29, B=@n32), B=LinearInterpolate(Alpha=@n33) {constA=1, constB=0.6}))
 ]}
