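# Material function that renders the primitive logo embossed on knob tops, outputting Normal, Spec and a KnobsTop mask.
# A custom HLSL loop parallax-marches the height mask along tangent-space view, with the Ruler input scaling steps and depth.
# It samples with precomputed gradients and exits early outside the knob mask.
# Instance-space radial distance drives an injection-mold switch, whose two texture sets are currently identical, and a central spot that dims Spec.
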MaterialFunction /AtomFramework/Materials/MatFunctions/MF_PrimitiveLogo
graph {nodes [
  MaterialExpressionMaterialFunctionCall_3 = MaterialFunctionCall(
   UVs=Add(A=TextureCoordinate(), B=Constant2Vector() {g=1})
   Texture Scale=ScalarParameter() {defaultValue=0.48, group=Logo, parameterName=LogoUV}
  ) {materialFunction=&ScaleUVsByCenter}
  MaterialExpressionTransform_0 = Transform(None=CameraVectorWS()) {transformSourceType=TRANSFORMSOURCE_World, transformType=TRANSFORM_Tangent}
  n1 = Saturate(None=FunctionInput(Preview=Constant() {r=1}) {bUsePreviewValueAsDefault=true, inputName=Ruler, inputType=FunctionInput_Scalar})
  MaterialExpressionLinearInterpolate_3 = LinearInterpolate(
   A=ScalarParameter() {defaultValue=1, group=Logo, parameterName=MinSteps}
   B=ScalarParameter() {defaultValue=8, group=Logo, parameterName=MaxSteps}
   Alpha=@n1
  ) {constA=1, constB=8}
  MaterialExpressionDDX_0 = DDX(Value=@MaterialExpressionMaterialFunctionCall_3.UVs)
  MaterialExpressionDDY_0 = DDY(Value=@MaterialExpressionMaterialFunctionCall_3.UVs)
  MaterialExpressionLinearInterpolate_4 = LinearInterpolate(
   A=ScalarParameter() {group=Logo, parameterName=DepthMin}
   B=ScalarParameter() {defaultValue=0.025, group=Logo, parameterName=DepthMax}
   Alpha=@n1
  )
  MaterialExpressionTextureObject_1 = TextureObject() {samplerType=SAMPLERTYPE_Masks, texture=/AtomFramework/Materials/Textures/PrimitiveLogo_M}
  MaterialExpressionTextureObject_0 = TextureObject() {samplerType=SAMPLERTYPE_Masks, texture=&PrimitiveLogo_M}
  MaterialExpressionTextureObject_3 = TextureObject() {samplerType=SAMPLERTYPE_Normal, texture=/AtomFramework/Materials/Textures/PrimitiveLogo_N}
  MaterialExpressionTextureObject_2 = TextureObject() {samplerType=SAMPLERTYPE_Normal, texture=&PrimitiveLogo_N}
  MaterialExpressionNamedRerouteUsage_1 = NamedRerouteUsage()
  MaterialExpressionNamedRerouteUsage_4 = NamedRerouteUsage()
  n2 = Custom!(
   uv=@MaterialExpressionMaterialFunctionCall_3.UVs
   v=@MaterialExpressionTransform_0
   Steps=@MaterialExpressionLinearInterpolate_3
   InDDX=@MaterialExpressionDDX_0
   InDDY=@MaterialExpressionDDY_0
   Depth=@MaterialExpressionLinearInterpolate_4
   Bump=@MaterialExpressionTextureObject_1
   BumpInjection=@MaterialExpressionTextureObject_0
   Normal=@MaterialExpressionTextureObject_3
   NormalInjection=@MaterialExpressionTextureObject_2
   Injection=@MaterialExpressionNamedRerouteUsage_1
   KnobTop=@MaterialExpressionNamedRerouteUsage_4
  ) {
   code="if(Steps==0 || KnobTop == 0){\r\nreturn float3(0,0,1);\r\n}\r\nfloat2 uvOut = float2(0,0);\r\nfloat alpha = 0;\r\nfloat i=0;\r\nfloat intersect=0;\r\nfloat2 NormalXY = float2(0,0); \r\nfloat3 OutNormal = float3(0,0,1);\r\nfloat specOut = 0;\r\nwhile (i<Steps){\r\n    float h = Depth*i/Steps;\r\n\tfloat2 uvCurrent = uv+v.xy*(-Depth+h);\r\n\tif( Injection == 0 )\r\n\t{\r\n\t\tintersect = ((Bump.SampleGrad(GetMaterialSharedSampler(BumpSampler,View.MaterialTextureBilinearWrapedSampler),uvCurrent,InDDX, InDDY).x-.5)*2)*Depth;\r\n\t} else {\r\n\t\tintersect = ((BumpInjection.SampleGrad(GetMaterialSharedSampler(BumpInjectionSampler,View.MaterialTextureBilinearWrapedSampler),uvCurrent,InDDX, InDDY).x-.5)*2)*Depth;\r\n\t}\r\n\tif( h < intersect.x || i == 0 )\r\n\t{\t\r\n         uvOut = uvCurrent;\r\n\t\t\talpha = 1;\r\n\t}\r\n\ti++;\r\n}\r\nif( Injection == 0 ){\r\n\tNormalXY = Normal.SampleGrad(GetMaterialSharedSampler(NormalSampler,View.MaterialTextureBilinearWrapedSampler),uvOut,InDDX,InDDY).xy;\r\n\tNormalXY = NormalXY * MaterialFloat2(2.0f,2.0f) - MaterialFloat2(1.0f,1.0f);\r\n\tfloat  NormalZ = sqrt( saturate( 1.0f - dot( NormalXY, NormalXY ) ) );\r\n\tOutNormal = float3(NormalXY,NormalZ);\r\n\tspecOut = Bump.SampleGrad(GetMaterialSharedSampler(BumpSampler,View.MaterialTextureBilinearWrapedSampler),uvOut,InDDX, InDDY);\r\n}\r\nelse {\r\n\tNormalXY = NormalInjection.SampleGrad(GetMaterialSharedSampler(NormalInjectionSampler,View.MaterialTextureBilinearWrapedSampler),uvOut,InDDX,InDDY).xy;\r\n\tNormalXY = NormalXY * MaterialFloat2(2.0f,2.0f) - MaterialFloat2(1.0f,1.0f);\r\n\tfloat  NormalZ = sqrt( saturate( 1.0f - dot( NormalXY, NormalXY ) ) );\r\n\tOutNormal = float3(NormalXY,NormalZ);\r\n\tspecOut = BumpInjection.SampleGrad(GetMaterialSharedSampler(BumpInjectionSampler,View.MaterialTextureBilinearWrapedSampler),uvOut,InDDX, InDDY);\r\n}\r\nOutUV = uvOut;\r\nOutSpec = specOut;\r\nreturn OutNormal;"
   description=MultiBumpLogo
   additionalOutputs [{outputName=OutUV, outputType=CMOT_Float2}, {outputName=OutSpec, outputType=CMOT_Float1}]
   inputs [
    {inputName=uv, input {expression=&self:MaterialExpressionMaterialFunctionCall_3}}
    {inputName=v, input {expression=&self:MaterialExpressionTransform_0}}
    {inputName=Steps, input {expression=&self:MaterialExpressionLinearInterpolate_3}}
    {inputName=InDDX, input {expression=&self:MaterialExpressionDDX_0}}
    {inputName=InDDY, input {expression=&self:MaterialExpressionDDY_0}}
    {inputName=Depth, input {expression=&self:MaterialExpressionLinearInterpolate_4}}
    {inputName=Bump, input {expression=&self:MaterialExpressionTextureObject_1}}
    {inputName=BumpInjection, input {expression=&self:MaterialExpressionTextureObject_0}}
    {inputName=Normal, input {expression=&self:MaterialExpressionTextureObject_3}}
    {inputName=NormalInjection, input {expression=&self:MaterialExpressionTextureObject_2}}
    {inputName=Injection, input {expression=&self:MaterialExpressionNamedRerouteUsage_1}}
    {inputName=KnobTop, input {expression=&self:MaterialExpressionNamedRerouteUsage_4}}
   ]
  }
  n3 = MaterialFunctionCall(Input=@n2.OutSpec, Input Low=Constant() {r=0.5}, Input High=Constant() {r=1}, Target Low=Constant() {r=1}, Target High=Constant() {r=0.5}) {materialFunction=&RemapValueRange}
  n4 = FunctionInput(Preview=TransformPosition(None=WorldPosition().XYZ) {transformSourceType=TRANSFORMPOSSOURCE_World, transformType=TRANSFORMPOSSOURCE_Instance}) {bUsePreviewValueAsDefault=true, inputName=PositionWS, sortPriority=-7}
  n5 = MaterialFunctionCall(Vector 2=Divide(A=ComponentMask(None=@n4) {g=true, r=true}, B=ScalarParameter() {defaultValue=16, group=Logo, parameterName=WorldSize})) {materialFunction=/Engine/Functions/Engine_MaterialFunctions02/Utility/VectorLength}
  n6 = ScalarParameter() {defaultValue=0.04, group=Logo, parameterName=InjectionCenter}
  n7 = MaterialFunctionCall!(Input=@n5.V2 Length, Input Low=Add(A=@n6) {constB=0.02}, Input High=@n6) {materialFunction=/Engine/Functions/Engine_MaterialFunctions03/Math/RemapValueRangeNormalized}
  n8 = NamedRerouteDeclaration(None=SphereMask(
    A=@MaterialExpressionMaterialFunctionCall_3.UVs
    B=Constant2Vector() {g=0.5, r=0.5}
    Radius=ScalarParameter() {defaultValue=0.5597, group=Logo, parameterName=LogoMask}
    Hardness=Constant() {r=1}
   )) {name=KnobTop, nodeColor=<rgba 0.1719 1 0 1>}
  FunctionOutput(None=LinearInterpolate(A=@n3.Result, B=Multiply(A=@n3.Result) {constB=0.5}, Alpha=Multiply(A=@n7.Result, B=@n8))) {outputName=Spec}
  n9 = ScalarParameter() {defaultValue=2, group=Logo, parameterName=LogoNormalStrength}
  n10 = Constant() {r=1}
  FunctionOutput(None=Multiply(
    A=@n2.return
    B=MaterialFunctionCall(X=@n9, Y=@n9, Z=@n10) {materialFunction=&MakeFloat3}.Result
   )) {outputName=Normal, sortPriority=1}
  FunctionOutput(None=@n8) {outputName=KnobsTop, sortPriority=2}
  n11 = ScalarParameter() {defaultValue=0.33, group=Logo, parameterName=PivotKnob}
  NamedRerouteDeclaration(None=MaterialFunctionCall(Input=@n5.V2 Length, Input Low=Add(A=@n11) {constB=0.001}, Input High=@n11) {materialFunction=&RemapValueRangeNormalized}.Result) {name=InjectionMold, nodeColor=<rgba 0.7578 1 0 1>}
 ]}
