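# Material function that builds the layered surface normal for brick primitives, imitating moulded-plastic imperfections.
# It blends edge warp (flat on Low and Medium quality) with injection knob warp projected in instance space, using angle-corrected blending.
# An optional VDI grain averages two Brick Scale-driven, object-relative world-aligned samples, masked by texture, UV1 or strength.
# MF_PrimitiveLogo's normal is lerped on top, also serving as its own alpha.
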MaterialFunction /AtomFramework/Materials/MatFunctions/MF_PrimitiveNormal
graph {nodes [
  n1 = MaterialFunctionCall!(Seed=FunctionInput(Preview=Constant()) {bUsePreviewValueAsDefault=true, inputName=Random Seed, inputType=FunctionInput_Scalar}) {materialFunction=&/MF_EdgeWarp_N}
  n2 = NamedRerouteDeclaration(None=FunctionInput(Preview=Constant()) {bUsePreviewValueAsDefault=true, inputName=Flatten Edge warp, inputType=FunctionInput_Scalar}) {name=Random Seed, nodeColor=<rgba 1 0.5859 0 1>}
  n3 = Constant3Vector() {constant=<rgba 0 0 1 0>}
  n4 = TransformPosition(None=WorldPosition().XYZ) {transformSourceType=TRANSFORMPOSSOURCE_World, transformType=TRANSFORMPOSSOURCE_Instance}
  n5 = Divide(A=ComponentMask(None=Add(A=@n4) {constB=-6.5}) {g=true, r=true}) {constB=30.5}
  n6 = ScalarParameter() {defaultValue=0.9, group=Normal, parameterName=Injection Warp Strength, sliderMax=1}
  n7 = MaterialFunctionCall(
   BaseNormal=QualitySwitch(
    Default=MaterialFunctionCall(Normal=@n1.Normal, Flatness=@n2) {materialFunction=&FlattenNormal}.Result
    Low=@n3
    Medium=@n3
   )
   AdditionalNormal=MaterialFunctionCall(
    Normal=TextureSample(UVs=@n5) {samplerType=SAMPLERTYPE_Normal, texture=/AtomFramework/Materials/Textures/KnobWarp_N}.RGB
    Flatness=FunctionInput(Preview=@n6) {bUsePreviewValueAsDefault=true, inputName=Flatten Injection Warp N, inputType=FunctionInput_Scalar}
   ) {materialFunction=&FlattenNormal}.Result
  ) {materialFunction=&BlendAngleCorrectedNormals}
  n8 = TextureObject() {samplerType=SAMPLERTYPE_Normal, texture=/AtomFramework/Materials/Textures/T_VDI_N}
  n9 = Multiply(A=CollectionParameter() {collection=/AtomFramework/Materials/Tools/MPC_BrickData, parameterName=Brick Scale}) {constB=0.75}
  n10 = ObjectPositionWS()
  n11 = Add(A=Subtract(A=WorldPosition().XYZ, B=VertexInterpolator(VS=@n10).PS), B=Multiply(A=NamedRerouteUsage()) {constB=1000})
  n12 = Constant() {r=1}
  n13 = Multiply(A=@n9) {constB=1.5}
  n14 = ComponentMask(None=LinearInterpolate(
    A=MaterialFunctionCall(TextureObject=@n8, TextureSize=@n9, WorldPosition=@n11, ProjectionTransitionContrast=@n12) {materialFunction=&WorldAlignedTexture}.XYZ Texture
    B=MaterialFunctionCall(TextureObject=@n8, TextureSize=@n13, WorldPosition=@n11, ProjectionTransitionContrast=@n12) {materialFunction=&WorldAlignedTexture}.XYZ Texture
   )) {g=true, r=true}
  n15 = Constant() {r=1}
  n16 = TextureSampleParameter2D() {group=VDI, parameterName=VDI Mask, samplerType=SAMPLERTYPE_LinearColor, texture=/AtomFramework/Materials/Textures/T_VDITemp}
  n17 = ScalarParameter() {defaultValue=1, group=VDI, parameterName=VDI Strength}
  n18 = Saturate(None=OneMinus(None=Multiply(A=@n16.R, B=@n17)))
  n19 = OneMinus(None=ComponentMask(None=TextureCoordinate() {coordinateIndex=1}) {g=true})
  n20 = StaticSwitchParameter(True=Saturate(None=Sign(None=@n19)), False=@n17) {defaultValue=true, desc=UVs in UDIM space 0 will be no VDI and full VDI in 1., group=VDI, parameterName=VDI UV Mask}
  n21 = MaterialFunctionCall(
   BaseNormal=@n7.Result
   AdditionalNormal=MaterialFunctionCall(Normal=AppendVector(A=@n14, B=@n15), Flatness=StaticSwitchParameter(True=@n18, False=@n20) {group=VDI, parameterName=VDI Texture}) {materialFunction=&FlattenNormal}.Result
  ) {materialFunction=&BlendAngleCorrectedNormals}
  n22 = MaterialFunctionCall!(PositionWS, Ruler) {materialFunction=&/MF_PrimitiveLogo}
  FunctionOutput(None=LinearInterpolate(A=StaticSwitchParameter(True=@n21.Result, False=@n7.Result) {group=VDI, parameterName=VDI}, B=@n22.Normal, Alpha=@n22.Normal))
 ]}
