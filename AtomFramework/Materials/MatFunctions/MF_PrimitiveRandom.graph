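# MF_PrimitiveRandom is a material function that outputs a random vector per primitive, varied by instance or vertex data.
# A Custom node reads the persistent primitive index and adds a scaled seed to feed VectorNoise.
# Static switches (Index Color, Override Color, Use Vertex Colors) choose vertex alpha or PerInstanceRandom as that seed.
# NaniteReplace uses the noise directly for Nanite and routes it through a VertexInterpolator otherwise, computing it per vertex.
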MaterialFunction /AtomFramework/Materials/MatFunctions/MF_PrimitiveRandom
graph {nodes [
  n1 = VertexColor()
  n2 = PerInstanceRandom()
  n3 = StaticSwitchParameter(True=@n1.A, False=@n2) {
   desc="Whether or not to source the color information from a packed float per-instance (false) or standard vertex colors (true). Defaults to false. Only available if Override Color is false."
   group=1 - Col
   parameterName=Use Vertex Colors
  }
  n4 = StaticSwitchParameter(True=@n2, False=@n3) {
   desc="Whether or not to source the color from the Color parameter (if true) or from a PerVertex/PerInstance location (if false). Defaults to false."
   group=1 - Col
   parameterName=Override Color
  }
  n5 = StaticSwitchParameter(True=@n1.A, False=@n4) {group=1 - Col, parameterName=Index Color}
  n6 = VectorNoise(World Position=Add(
    A=Custom!(None) {code="GetPrimitiveData(Parameters).PersistentPrimitiveIndex", description=PrimIndex, outputType=CMOT_Float1}
    B=Multiply(A=@n5) {constB=255}
   ))
  FunctionOutput(None=NaniteReplace!(Default=VertexInterpolator(VS=@n6).PS, Nanite=@n6))
 ]}
