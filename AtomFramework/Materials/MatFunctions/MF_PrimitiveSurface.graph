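# Material function that supplies surface values for AtomFramework primitive materials: pass-through Metal and Spec, plus a varied Roughness.
# Samples a mask texture with WorldAlignedTexture in instance space, offset by Seed and sized by Brick Scale from MPC_BrickData.
# The red channel blends between minimum roughness and maximum scaled by saturated Seed.
# A sharpened Edge Falloff input then blends toward a slightly rougher, power-adjusted value.
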MaterialFunction /AtomFramework/Materials/MatFunctions/MF_PrimitiveSurface
graph {nodes [
  FunctionOutput(None=FunctionInput(Preview=Constant()) {bUsePreviewValueAsDefault=true, inputName=Metal, inputType=FunctionInput_Scalar}) {outputName=Metal}
  FunctionOutput(None=FunctionInput(Preview=Constant() {r=0.565}) {bUsePreviewValueAsDefault=true, inputName=Spec, inputType=FunctionInput_Scalar, sortPriority=1}) {outputName=Spec, sortPriority=1}
  n1 = Constant() {r=0.15}
  n2 = Constant() {r=0.3}
  n3 = MaterialFunctionCall(Float2=FunctionInput(Preview=MaterialFunctionCall(X=@n1, Y=@n2) {materialFunction=&MakeFloat2}.Result) {bUsePreviewValueAsDefault=true, inputName=Min Max Roughness, inputType=FunctionInput_Vector2}) {materialFunction=&BreakOutFloat2Components}
  n4 = FunctionInput(Preview=Constant() {r=1}) {bUsePreviewValueAsDefault=true, inputName=Seed}
  n5 = TextureObject() {samplerType=SAMPLERTYPE_Masks, texture=/AtomFramework/Materials/Textures/Primitive_Surface_R}
  n6 = Multiply(A=CollectionParameter() {collection=/AtomFramework/Materials/Tools/MPC_BrickData, parameterName=Brick Scale}) {constB=20}
  n7 = Add(
   A=TransformPosition(None=WorldPosition().XYZ) {transformSourceType=TRANSFORMPOSSOURCE_World, transformType=TRANSFORMPOSSOURCE_Instance}
   B=Multiply(A=@n4) {constB=1000}
  )
  n8 = Transform(None=VertexNormalWS()) {transformSourceType=TRANSFORMSOURCE_World, transformType=TRANSFORM_Instance}
  n9 = LinearInterpolate(
   A=@n3.R
   B=Multiply(A=@n3.G, B=Saturate(None=@n4))
   Alpha=ComponentMask(None=MaterialFunctionCall(TextureObject=@n5, TextureSize=@n6, WorldPosition=@n7, World Space Normal=@n8) {materialFunction=&WorldAlignedTexture}.XYZ Texture) {r=true}
  )
  n10 = Power(Base=FunctionInput(Preview=Constant() {r=1}) {bUsePreviewValueAsDefault=true, inputName=Edge Falloff, inputType=FunctionInput_Scalar, sortPriority=3}) {constExponent=8}
  FunctionOutput(None=LinearInterpolate(A=@n9, B=Power(Base=@n9) {constExponent=0.85}, Alpha=Saturate(None=@n10)) {constA=0.1}) {outputName=Roughness, sortPriority=2}
 ]}
