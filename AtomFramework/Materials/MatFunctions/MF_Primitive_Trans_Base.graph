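# Material function that builds the base translucent, glass-like surface for AtomFramework primitives, outputting material attributes, Refraction and a Thin color.
# Primitive random seeds drive shared color, surface and normal functions, with the normal flattened and metalness varying by face side.
# Emissive fakes reflections by tinting a Fresnel-weighted skylight environment sample, while view angle shapes opacity and refraction, the latter also fading with camera distance.
# Path-tracing switches replace emissive, opacity and refraction with constants.
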MaterialFunction /AtomFramework/Materials/MatFunctions/MF_Primitive_Trans_Base
graph {nodes [
  n1 = VectorParameter() {parameterName=Surface, defaultValue=<rgba 0.05 0.256 0.2 0.7>}
  n2 = MaterialFunctionCall!(
   Seed=MaterialFunctionCall() {materialFunction=&/MF_PrimitiveRandom}.Result
   Min Max Roughness=MaterialFunctionCall(X=@n1.R, Y=@n1.R) {materialFunction=&MakeFloat2}.Result
   Metal=@n1.B
   Spec=@n1.B
   Edge Falloff=MaterialFunctionCall!(BoundingBoxBased_0-1_UVW, NormalWS) {materialFunction=&/MF_EdgeWarpMask}.Edge Mask
  ) {materialFunction=&/MF_PrimitiveSurface}
  n3 = TwoSidedSign()
  n4 = MaterialFunctionCall() {materialFunction=&MF_PrimitiveRandom}
  n5 = Saturate(None=Desaturation(
    None=MaterialFunctionCall!(Random Seed=@n4.Result) {materialFunction=&/MF_PrimitiveColor}.Result
    Fraction=Constant()
   ))
  n6 = Multiply(A=@n5, B=ScalarParameter() {defaultValue=1, parameterName=Inner Emissive Boost})
  n7 = ReflectionVectorWS()
  n8 = ScalarParameter() {defaultValue=1, parameterName=SkyRough0}
  n9 = Multiply(
   A=Multiply(A=SkyLightEnvMapSample(Direction=@n7, Roughness=@n8), B=ScalarParameter() {defaultValue=2, parameterName=SkyMult})
   B=Fresnel() {baseReflectFraction=0.2, exponent=1}
  )
  n10 = Constant() {r=0.5}
  n11 = MaterialFunctionCall() {materialFunction=&CameraDirectionVector}
  n12 = VertexNormalWS()
  n13 = Power(Base=Abs(None=DotProduct(A=@n11.Result, B=@n12))) {constExponent=5}
  n14 = MaterialFunctionCall() {materialFunction=&MF_PrimitiveRandom}
  n15 = Constant() {r=0.1}
  n16 = Distance(A=WorldPosition().XYZ, B=CameraPositionWS())
  n17 = Saturate(None=OneMinus(None=Divide(A=@n16) {constB=1000}))
  n18 = Add(A=Multiply(A=Abs(None=@n13), B=Multiply(A=@n15, B=@n17)))
  n19 = SetMaterialAttributes!(
   MaterialAttributes
   BaseColor
   Metallic=LinearInterpolate(B=@n2.Metal, Alpha=Saturate(None=@n3)) {constA=1, constB=0}
   Specular=@n2.Spec
   Roughness=@n2.Spec
   EmissiveColor=PathTracingQualitySwitch!(Normal=Multiply(A=@n6, B=@n9), PathTraced=Constant())
   Opacity=PathTracingQualitySwitch(Normal=LinearInterpolate(A=@n10, Alpha=@n13) {constA=0.01, constB=0.01}, PathTraced=Constant())
   Normal=MaterialFunctionCall(
    Normal=MaterialFunctionCall!(Flatten Edge warp, Flatten Injection Warp N, Random Seed=@n14.Result) {materialFunction=&/MF_PrimitiveNormal}.Result
    Flatness=Constant() {r=0.75}
   ) {materialFunction=&FlattenNormal}.Result
   Refraction=@n18
  ) {attributeSetTypes [
    #1
    #2
    #3
    #4
    #5
    #6
    #7
    #8
   ]}
  FunctionOutput(None=@n19) {sortPriority=1}
  FunctionOutput(None=PathTracingQualitySwitch(Normal=@n18, PathTraced=Constant() {r=1.5})) {outputName=Refraction, sortPriority=2}
  FunctionOutput(None=Add(A=@n5) {constB=0.1}) {outputName=Thin}
 ]}
