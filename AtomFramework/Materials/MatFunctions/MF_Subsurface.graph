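# Material function that derives subsurface scattering attributes from an incoming Material Attributes set's base color and metallic.
# Desaturated base color drives a remapped blend that splits color between Base Color and Subsurface Color, with metallic suppressing the latter.
# The same brightness lowers Opacity for lighter colors.
# Quality and shading path switches use this only at Default and Epic quality; Low, Medium, High and mobile just zero Subsurface Color.
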MaterialFunction /AtomFramework/Materials/MatFunctions/MF_Subsurface
graph {nodes [
  n1 = FunctionInput() {inputType=FunctionInput_MaterialAttributes}
  n2 = GetMaterialAttributes(None=@n1) {attributeGetTypes [#1, #2]}
  n3 = ComponentMask(None=Desaturation(None=@n2.BaseColor, Fraction=Constant() {r=1})) {r=true}
  n4 = LinearInterpolate(
   A=Constant() {desc=SSAmountDark, r=0.1}
   B=Constant() {desc=SSAmountLight, r=0.2}
   Alpha=MaterialFunctionCall!(Input=@n3, Input Low=Constant(), Input High=Constant() {r=0.7}) {materialFunction=/Engine/Functions/Engine_MaterialFunctions03/Math/RemapValueRangeNormalized}.Result
  )
  n5 = SetMaterialAttributes!(
   MaterialAttributes=@n2.MaterialAttributes
   BaseColor=Multiply(A=@n2.BaseColor, B=OneMinus(None=@n4))
   Opacity=LinearInterpolate(A=Constant() {desc=SubsurfaceTrans0, r=0.95}, B=Constant() {desc=SubsurfaceTrans1, r=0.8}, Alpha=@n3)
   SubsurfaceColor=Multiply(A=Multiply(A=@n2.BaseColor, B=@n4), B=OneMinus(None=@n2.Metallic))
  ) {attributeSetTypes [#1, #3, #4]}
  n6 = SetMaterialAttributes!(MaterialAttributes=@n1, SubsurfaceColor=Constant3Vector()) {attributeSetTypes [#4]}
  FunctionOutput(None=ShadingPathSwitch(Default=QualitySwitch(Default=@n5, Low=@n6, High=@n6, Medium=@n6, Epic=@n5), Mobile=@n6))
 ]}
