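# MF_TechPrim is a material function that layers switchable "Technic" detail onto incoming material attributes.
# An optional secondary color from a LUT atlas, indexed by UV1, blends into base and subsurface color via the red mask.
# Other mask channels drive shading and opacity mask, a normal map replaces the normal, and TwoSidedSign darkens backfaces.
# Optional parallax occlusion mapping reads mask alpha as height and switches opacity mask from green to alpha.
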MaterialFunction /AtomFramework/Materials/MatFunctions/MF_TechPrim
graph {
 properties {previewBlendMode=BLEND_Masked}
 nodes [
  n1 = GetMaterialAttributes(None=NamedRerouteDeclaration(None=FunctionInput() {bUsePreviewValueAsDefault=true, inputType=FunctionInput_MaterialAttributes}) {name=IN MAT, nodeColor=<rgba 0.8984 1 0 1>}) {attributeGetTypes [#1, #2]}
  n2 = ComponentMask(None=TextureCoordinate() {coordinateIndex=1}) {r=true}
  n3 = Divide(A=Multiply(A=Floor(None=@n2))) {constB=512}
  n4 = Divide(A=Constant() {r=0.5}) {constB=512}
  n5 = TextureObjectParameter() {
   desc=The texture to use as the look-up texture atlas
   group=1 - Col
   parameterName=Look Up Table Atlas Texture
   texture=/AtomFramework/Materials/Textures/T_LUT_DefautAtlas
  }
  n6 = TextureProperty(Tex=@n5) {property=TMTM_TexelSize}
  n7 = TextureSample(UVs=AppendVector(A=Add(A=@n3, B=@n4), B=ComponentMask(None=@n6) {g=true}), Tex=@n5) {texture=&T_LUT_DefautAtlas}
  n8 = ComponentMask(None=NamedRerouteUsage()) {r=true}
  n9 = LinearInterpolate(A=@n1.BaseColor, B=@n7.RGB, Alpha=@n8)
  n10 = LinearInterpolate(A=@n1.SubsurfaceColor, B=@n7.RGB, Alpha=@n8)
  n11 = GetMaterialAttributes(None=StaticSwitchParameter(
    True=SetMaterialAttributes!(MaterialAttributes=@n1.MaterialAttributes, BaseColor=@n9, SubsurfaceColor=@n10) {attributeSetTypes [#1, #2]}
    False=NamedRerouteUsage()
   ) {defaultValue=true, group=Technic, parameterName=Technic Secondary Color}) {attributeGetTypes [#1, #3, #4]}
  n12 = MaterialFunctionCall(Float4=NamedRerouteUsage()) {materialFunction=&BreakOutFloat4Components}
  n13 = Saturate(None=TwoSidedSign())
  n14 = Add(B=@n13) {constA=0.25}
  n15 = SetMaterialAttributes!(
   MaterialAttributes=@n11.MaterialAttributes
   BaseColor=Multiply(A=Multiply(A=@n11.BaseColor, B=@n12.B), B=Saturate(None=@n14))
   Specular=Multiply(A=Multiply(A=@n11.Specular, B=@n12.G), B=@n13)
   AmbientOcclusion=Multiply(A=Multiply(A=@n11.AmbientOcclusion, B=@n12.B), B=@n13)
   OpacityMask=StaticSwitchParameter(True=@n12.A, False=@n12.G) {group=Technic, parameterName=Technic Parallax, sortPriority=3}
   Normal=NamedRerouteUsage()
  ) {attributeSetTypes [
    #1
    #3
    #4
    #5
    #6
   ]}
  FunctionOutput(None=StaticSwitchParameter(True=@n15, False=NamedRerouteUsage()) {group=Technic, parameterName=Technic, sortPriority=0})
  n16 = TextureCoordinate()
  n17 = TextureObjectParameter(UVs=@n16) {
   group=Technic
   parameterName=Technic Masks
   samplerType=SAMPLERTYPE_LinearColor
   sortPriority=1
   texture=/Engine/ArtTools/RenderToTexture/Textures/127grey
  }
  n18 = StaticSwitchParameter(
   True=MaterialFunctionCall!(
    Heightmap Texture=@n17
    Height Ratio=ScalarParameter() {defaultValue=-0.08, group=Technic, parameterName=Technic Depth, sortPriority=4}
    Min Steps
    Max Steps=Constant() {r=64}
    UVs=@n16
    Heightmap Channel=Constant4Vector() {constant=<rgba 0 0 0 1>}.RGBA
    Reference Plane
    -----------------------------
    Use World Coordinates
    Specify Manual Texture Size
    Manual Texture Size
    ------------------------------
    "Render Shadows (Occlusion Mapping)"
    Light Vector
    Shadow Steps
    Shadow Penumbra
    Transform To VertexNormal
   ) {materialFunction=/Engine/Functions/Engine_MaterialFunctions01/Texturing/ParallaxOcclusionMapping}.Parallax UVs
   False=@n16
  ) {group=Technic, parameterName=Technic Parallax}
  NamedRerouteDeclaration(None=TextureSample(UVs=@n18, Tex=@n17) {samplerType=SAMPLERTYPE_Masks, texture=&127grey}.RGBA) {name=Technic Masks, nodeColor=<rgba 1 0 0.2578 1>}
  n19 = TextureSampleParameter2D(UVs=@n18) {
   group=Technic
   parameterName=Technic Normal
   samplerSource=SSM_Wrap_WorldGroupSettings
   samplerType=SAMPLERTYPE_Normal
   sortPriority=2
   texture=/Engine/Engine_MI_Shaders/T_Normal
  }
  NamedRerouteDeclaration(None=@n19.RGB) {name=Technic Normal, nodeColor=<rgba 0 0.5781 1 1>}
 ]
}
