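# MF_WhiteoutBlendNormals is a material function that combines two normal inputs, Normal1 and Normal2, using whiteout blending.
# It adds the XY components of both normals and multiplies their Z components.
# The results are appended into a vector and normalized as the single output.
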MaterialFunction /AtomFramework/Materials/MatFunctions/MF_WhiteoutBlendNormals
graph {nodes [
  n1 = FunctionInput() {inputName=Normal1}
  n2 = FunctionInput() {inputName=Normal2}
  n3 = Add(A=ComponentMask(None=@n1) {g=true, r=true}, B=ComponentMask(None=@n2) {g=true, r=true})
  n4 = Multiply(A=ComponentMask(None=@n1) {b=true}, B=ComponentMask(None=@n2) {b=true})
  FunctionOutput(None=Normalize(VectorInput=AppendVector(A=@n3, B=@n4)))
 ]}
