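# Material function that fakes LEGO-style brick undersides (walls, pins, tubes) on incoming material attributes when DoFakeBottom is on.
# Texture coordinates encode brick dimensions and quadrant, driving borders, wall thickness and shape choice.
# Use Analytic Setup selects custom HLSL ray-cylinder/box intersection, producing normals and depth masks, or a flat path using a 2x2 mask atlas, parallax and faked sun shading.
# Results modify color, specular, roughness, normal, emissive and subsurface.
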MaterialFunction /AtomFramework/Materials/MatFunctions/MF_Undersides
graph {nodes [
  n1 = FunctionInput() {inputType=FunctionInput_MaterialAttributes}
  n2 = GetMaterialAttributes(None=@n1) {attributeGetTypes [
    #1
    #2
    #3
    #4
    #5
    #6
    #7
   ]}
  n3 = SetMaterialAttributes!(
   MaterialAttributes=@n2.MaterialAttributes
   BaseColor=NamedRerouteUsage()
   Metallic=NamedRerouteDeclaration(None=@n2.Metallic) {name=Metallic In, nodeColor=<rgba 0 0.3672 1 1>}
   Specular=NamedRerouteUsage()
   Roughness=NamedRerouteUsage()
   Normal=NamedRerouteUsage()
   EmissiveColor=NamedRerouteUsage()
   SubsurfaceColor=NamedRerouteUsage()
  ) {attributeSetTypes [
    #1
    #2
    #3
    #4
    #5
    #6
    #7
   ]}
  FunctionOutput(None=StaticSwitchParameter(True=@n3, False=@n1) {group=Undersides, parameterName=DoFakeBottom, sortPriority=-1})
  n4 = Multiply(A=NamedRerouteDeclaration(None=TextureCoordinate()) {name=TexCoord, nodeColor=<rgba 0.3828 0 1 1>}, B=Constant2Vector() {g=-1, r=1}) {constB=10, desc=Flip}
  n5 = ComponentMask(None=@n4) {r=true}
  n6 = Ceil(None=@n5)
  n7 = Floor(None=@n5)
  n8 = Saturate(None=Sign(None=@n5))
  n9 = ComponentMask(None=@n4) {g=true}
  n10 = LinearInterpolate(A=Ceil(None=@n9), B=Floor(None=@n9), Alpha=Saturate(None=Sign(None=@n9)))
  NamedRerouteDeclaration(None=AppendVector(A=LinearInterpolate(A=@n6, B=@n7, Alpha=@n8), B=Add(A=@n10))) {name=Brick Dimensions, nodeColor=<rgba 0.9762 0.0398 1 1>}
  n11 = Constant()
  n12 = ConstantBiasScale(None=Abs(None=NamedRerouteUsage())) {bias=-1, scale=0.17}
  n13 = MaterialFunctionCall(
   X=@n11
   Y=LinearInterpolate(B=ComponentMask(None=@n12) {r=true}, Alpha=NamedRerouteUsage()) {constA=0.25, constB=0.25}
   Z=@n11
   A=LinearInterpolate(B=ComponentMask(None=@n12) {g=true}, Alpha=NamedRerouteUsage()) {constA=0.25, constB=0.25}
  ) {materialFunction=&MakeFloat4}
  n14 = NamedRerouteUsage()
  NamedRerouteDeclaration(None=Floor(None=Divide(A=@n13.Result, B=@n14))) {name=UV Borders, nodeColor=<rgba 0 0.5781 1 1>}
  NamedRerouteDeclaration(None=Constant() {r=0.16}) {name=Size, nodeColor=<rgba 0 0.8359 1 1>}
  n15 = NamedRerouteUsage()
  n16 = Saturate(None=Subtract(A=Abs(None=@n15)))
  NamedRerouteDeclaration(None=ComponentMask(None=@n16) {r=true}) {name=Is Wider Than 1 Stud, nodeColor=<rgba 0.8281 1 0 1>}
  n17 = Sign(None=NamedRerouteUsage())
  NamedRerouteDeclaration(None=Saturate(None=ComponentMask(None=@n17) {r=true})) {name=Is UDIM X Positive, nodeColor=<rgba 0.0312 0 1 1>}
  NamedRerouteDeclaration(None=Saturate(None=ComponentMask(None=@n17) {g=true})) {name=Is UDIM Y Positive, nodeColor=<rgba 0 0.7422 1 1>}
  n18 = Constant2Vector() {g=-0.2, r=0.3}
  n19 = LinearInterpolate(A=Constant2Vector() {g=0.3, r=-0.2}, Alpha=NamedRerouteUsage()) {constA=0.3, constB=0.3}
  n20 = NamedRerouteUsage()
  NamedRerouteDeclaration(None=LinearInterpolate(A=Constant() {r=-0.2}, B=LinearInterpolate(A=@n18, B=@n19, Alpha=@n20) {constB=0.3}, Alpha=NamedRerouteUsage()) {constA=-0.2, constB=-0.2}) {name=Wall Thickness, nodeColor=<rgba 0.2188 1 0 1>}
  NamedRerouteDeclaration(None=ComponentMask(None=@n16) {g=true}) {name=Is Longer Than 1 Stud, nodeColor=<rgba 0 1 0.3906 1>}
  n21 = Abs(None=Floor(None=NamedRerouteUsage()))
  n22 = ComponentMask(None=@n21) {r=true}
  n23 = ComponentMask(None=@n21) {g=true}
  NamedRerouteDeclaration(None=Saturate(None=Min(A=@n22, B=@n23))) {name=Is Not Empty, nodeColor=<rgba 0.2891 1 0 1>}
  n24 = LinearInterpolate(
   A=Constant2Vector() {r=-0.9057}
   B=LinearInterpolate(A=Constant2Vector() {g=-0.0315, r=-0.937}, B=Constant2Vector() {g=-0.0312, r=-0.9057}, Alpha=NamedRerouteUsage()) {constA=0.3, constB=-0.0312}
   Alpha=NamedRerouteUsage()
  ) {constB=0.3}
  n25 = LinearInterpolate(
   A=Constant2Vector() {r=-0.0312}
   B=LinearInterpolate(A=Constant2Vector() {g=-0.0312}, Alpha=NamedRerouteUsage()) {constA=0.3, constB=-0.0312}
   Alpha=NamedRerouteUsage()
  ) {constB=0.3}
  n26 = NamedRerouteUsage()
  NamedRerouteDeclaration(None=LinearInterpolate(B=LinearInterpolate(A=@n24, B=@n25, Alpha=@n26), Alpha=NamedRerouteUsage()) {constB=-0.2}) {name=Per Quadrant Offset, nodeColor=<rgba 0 1 0.7188 1>}
  n27 = AppendVector(A=Add(A=NamedRerouteUsage()) {constB=-1}, B=Constant())
  n28 = Frac(None=NamedRerouteUsage())
  n29 = ComponentMask(None=@n28) {g=true}
  n30 = LinearInterpolate(A=@n28, B=AppendVector(A=ComponentMask(None=@n28) {r=true}, B=OneMinus(None=@n29)), Alpha=NamedRerouteUsage())
  n31 = NamedRerouteUsage()
  n32 = Divide(A=Multiply(A=Add(A=@n30, B=@n31) {constB=-0.02}) {constB=2.56, desc=Texel density scale}, B=NamedRerouteUsage())
  NamedRerouteDeclaration(None=LinearInterpolate(A=Add(A=@n27, B=@n32), B=@n32, Alpha=NamedRerouteUsage()) {desc=Fix Issues in Quadrant II Negative X}) {name=UV, nodeColor=<rgba 0 0.5547 1 1>}
  n33 = Floor(None=Divide(A=NamedRerouteUsage(), B=NamedRerouteUsage()))
  n34 = Floor(None=Divide(A=NamedRerouteUsage(), B=NamedRerouteUsage()))
  n35 = NamedRerouteDeclaration(None=Multiply(A=Saturate(None=@n33), B=Saturate(None=@n34))) {name=Shape Area Mask, nodeColor=<rgba 0 1 0.9297 1>}
  n36 = NamedRerouteUsage()
  NamedRerouteDeclaration(None=Add(A=NamedRerouteUsage(), B=Multiply(A=@n35, B=@n36))) {name=Flat Underside Mask, nodeColor=<rgba 1 0 0.1641 1>}
  NamedRerouteDeclaration(None=@n2.BaseColor) {name=Base Color In, nodeColor=<rgba 0.6172 1 0 1>}
  n37 = Multiply(A=NamedRerouteUsage(), B=NamedRerouteUsage())
  n38 = Constant3Vector() {desc="R: away from sun G: facing sun B: facing down ", constant=<rgba -0.775 0.75 -0.65 0.25>}
  n39 = NamedRerouteUsage()
  n40 = Add(A=@n39) {constB=-0.75}
  n41 = ComponentMask(None=VertexNormalWS()) {b=true}
  n42 = Multiply(
   A=@n38.G
   B=LinearInterpolate(A=Add(A=@n39) {constA=-8.1419, constB=-0.25}, B=Max(A=@n40) {constB=-1}, Alpha=Saturate(None=@n41)) {constA=-0.25, constB=-0.75}
  )
  n43 = MaterialFunctionCall(Float3=LinearInterpolate(
    B=MaterialFunctionCall(X=@n38.R, Y=@n42, Z=@n38.R) {materialFunction=&MakeFloat3}.Result
    Alpha=NamedRerouteUsage()
   ) {constB=0}) {materialFunction=&BreakOutFloat3Components}
  n44 = LinearInterpolate(
   A=LinearInterpolate(
    A=MaterialFunctionCall!(Base=@n37, Add=@n43.R) {materialFunction=/Engine/Functions/Engine_MaterialFunctions02/Math/MultiplyAdd}.Result
    B=MaterialFunctionCall(Base=@n37, Add=@n43.G) {materialFunction=&MultiplyAdd}.Result
    Alpha=NamedRerouteUsage()
   )
   B=MaterialFunctionCall(Base=@n37, Add=@n43.B) {materialFunction=&MultiplyAdd}.Result
   Alpha=NamedRerouteUsage()
  )
  NamedRerouteDeclaration(None=LinearInterpolate(A=Saturate(None=@n44), B=NamedRerouteUsage(), Alpha=NamedRerouteUsage())) {name=Base Color Flat Out, nodeColor=<rgba 0.9219 0 1 1>}
  n45 = NamedRerouteUsage()
  NamedRerouteDeclaration(None=Multiply(A=Saturate(None=@n45), B=NamedRerouteUsage())) {name=Specular Flat Out, nodeColor=<rgba 1 0.5391 0 1>}
  NamedRerouteDeclaration(None=@n2.Specular) {name=Specular In, nodeColor=<rgba 1 0.5859 0 1>}
  n46 = Frac(None=NamedRerouteUsage())
  n47 = Constant2Vector() {desc=Empty, g=0.5, r=0.5}
  n48 = NamedRerouteUsage()
  n49 = NamedRerouteUsage()
  n50 = LinearInterpolate(
   A=LinearInterpolate(A=Constant2Vector() {desc=Hollow Pins}, B=Constant2Vector() {desc=Pins, r=0.5}, Alpha=NamedRerouteUsage())
   B=Constant2Vector() {desc=Tubes, g=0.5}
   Alpha=NamedRerouteDeclaration(None=Min(A=@n48, B=@n49)) {name="One Side > 1 Stud", nodeColor=<rgba 0 1 0.3672 1>}
  ) {constB=0.25}
  n51 = NamedRerouteUsage()
  NamedRerouteDeclaration(None=Add(A=Divide(A=@n46) {desc=2x2 Atlas}, B=LinearInterpolate(A=@n47, B=@n50, Alpha=@n51))) {name=Atlas UVs, nodeColor=<rgba 0.9453 0 1 1>}
  n52 = NamedRerouteUsage()
  n53 = Max(A=Add(A=Abs(None=@n52)) {constB=-1})
  NamedRerouteDeclaration(None=ComponentMask(None=@n53) {r=true}) {name=Brick X, nodeColor=<rgba 0.0312 1 0 1>}
  NamedRerouteDeclaration(None=ComponentMask(None=@n53) {g=true}) {name=Brick Y, nodeColor=<rgba 0.4062 0 1 1>}
  n54 = Ceil(None=NamedRerouteUsage())
  NamedRerouteDeclaration(None=ComponentMask(None=@n54) {r=true}) {name=Raw U Ceil, nodeColor=<rgba 0 1 0.1328 1>}
  NamedRerouteDeclaration(None=ComponentMask(None=@n54) {g=true}) {name=Raw V Ceil, nodeColor=<rgba 0.2188 0 1 1>}
  n55 = Abs(None=Floor(None=NamedRerouteUsage()))
  NamedRerouteDeclaration(None=ComponentMask(None=Saturate(None=@n55)) {g=true}) {name=Is Empty Deep, nodeColor=<rgba 0 0.8125 1 1>}
  NamedRerouteDeclaration(None=LinearInterpolate(
    A=Constant3Vector() {desc=Empty, constant=<rgba 0 0 0.4 0.1>}
    B=Constant3Vector() {desc=Empty Deep, constant=<rgba 0 0 1 0.1>}
    Alpha=NamedRerouteUsage()
   )) {name=Empties, nodeColor=<rgba 1 0.8906 0 1>}
  NamedRerouteDeclaration(None=LinearInterpolate(
    A=Constant3Vector() {desc=Hollow Pin Shallow, constant=<rgba 0.2 0.085 0.4 0.7>}
    B=Constant3Vector() {desc=Pin Shallow, constant=<rgba 0.2 0 0.4 0.1>}
    Alpha=NamedRerouteUsage()
   )) {name=Shallow Pins, nodeColor=<rgba 1 0.8203 0 1>}
  NamedRerouteDeclaration(None=LinearInterpolate(
    A=Constant3Vector() {desc=Hollow Pin Deep, constant=<rgba 0.2 0.085 1 0.1>}
    B=Constant3Vector() {desc=Pin Deep, constant=<rgba 0.2 0 1 0.1>}
    Alpha=NamedRerouteUsage()
   )) {name=Deep Pins, nodeColor=<rgba 0 1 0.8125 1>}
  NamedRerouteDeclaration(None=LinearInterpolate(
    A=Constant3Vector() {desc=Tube Deep, constant=<rgba 0.425 0.275 1 0.1>}
    B=Constant3Vector() {desc=Tube Shallow, constant=<rgba 0.425 0.275 0.3 0.8>}
    Alpha=NamedRerouteUsage()
   )) {name=Tubes, nodeColor=<rgba 0 0.0859 1 1>}
  NamedRerouteDeclaration(None=LinearInterpolate(A=NamedRerouteUsage(), B=NamedRerouteUsage(), Alpha=NamedRerouteUsage())) {name=Pins, nodeColor=<rgba 1 0 0.0703 1>}
  n56 = LinearInterpolate(
   A=NamedRerouteUsage()
   B=LinearInterpolate(A=NamedRerouteUsage(), B=NamedRerouteUsage(), Alpha=NamedRerouteUsage()) {constB=0.25}
   Alpha=NamedRerouteUsage()
  )
  n57 = MaterialFunctionCall(Float3=Min(A=Saturate(None=@n56), B=Constant3Vector() {desc=Avoid bottomless interior, constant=<rgba 1 1 0.95 0>})) {materialFunction=&BreakOutFloat3Components}
  NamedRerouteDeclaration(None=@n57.R) {name=Outer Radius, nodeColor=<rgba 0.0312 1 0 1>}
  NamedRerouteDeclaration(None=@n57.G) {name=Inner Radius, nodeColor=<rgba 0 0.8125 1 1>}
  NamedRerouteDeclaration(None=@n57.B) {name=Depth, nodeColor=<rgba 0.4297 1 0 1>}
  n58 = NamedRerouteUsage()
  NamedRerouteDeclaration(None=LinearInterpolate(Alpha=Round(None=@n58)) {constA=3, constB=5}) {name=Iterations, nodeColor=<rgba 0 1 0.5078 1>}
  n59 = Transform(None=CameraVectorWS()) {transformSourceType=TRANSFORMSOURCE_World, transformType=TRANSFORM_Tangent}
  n60 = LinearInterpolate(A=@n59, B=Multiply(A=@n59, B=Constant3Vector() {constant=<rgba 1 -1 1 0>}) {desc=Flip V}, Alpha=NamedRerouteUsage())
  n61 = ComponentMask(None=NamedRerouteDeclaration(None=Multiply(A=@n60) {constB=-1, desc=Flip RD}) {name=Ray Direction, nodeColor=<rgba 0 1 0.2969 1>}) {g=true, r=true}
  NamedRerouteDeclaration(None=Multiply(A=DotProduct(A=@n61, B=@n61)) {constB=2}) {name=Quadratic A, nodeColor=<rgba 0.4062 1 0 1>}
  n62 = Frac(None=NamedRerouteUsage())
  NamedRerouteDeclaration(None=AppendVector(A=Subtract(A=@n62) {constB=0.5}, B=Constant())) {name=Ray Origin, nodeColor=<rgba 0.3359 0 1 1>}
  MaterialExpressionNamedRerouteUsage_350 = NamedRerouteUsage()
  MaterialExpressionNamedRerouteUsage_351 = NamedRerouteUsage()
  MaterialExpressionNamedRerouteUsage_352 = NamedRerouteUsage()
  MaterialExpressionNamedRerouteUsage_353 = NamedRerouteUsage()
  MaterialExpressionNamedRerouteUsage_354 = NamedRerouteUsage()
  MaterialExpressionNamedRerouteUsage_355 = NamedRerouteUsage()
  MaterialExpressionNamedRerouteUsage_356 = NamedRerouteUsage()
  MaterialExpressionNamedRerouteUsage_357 = NamedRerouteUsage()
  MaterialExpressionNamedRerouteUsage_358 = NamedRerouteUsage()
  MaterialExpressionNamedRerouteUsage_359 = NamedRerouteUsage()
  MaterialExpressionConstant_22 = Constant() {r=0.65}
  n63 = Custom!(
   RawUV=@MaterialExpressionNamedRerouteUsage_350
   RO=@MaterialExpressionNamedRerouteUsage_351
   RD=@MaterialExpressionNamedRerouteUsage_352
   a=@MaterialExpressionNamedRerouteUsage_353
   R=@MaterialExpressionNamedRerouteUsage_354
   MaxOverlapSteps=@MaterialExpressionNamedRerouteUsage_355
   Depth=@MaterialExpressionNamedRerouteUsage_356
   Border=@MaterialExpressionNamedRerouteUsage_357
   Inner_R=@MaterialExpressionNamedRerouteUsage_358
   WallThickness=@MaterialExpressionNamedRerouteUsage_359
   BottomAO=@MaterialExpressionConstant_22
  ) {
   code="// INPUTS\r\n//input float2\tRawUV\t\t\t\t: Previously divided by cell size\r\n//input float3\tRO\t\t\t\t   : Basically float3(frac(RawUV) - 0.5, 0);  \r\n//input float3\tRD\t\t\t\t   : Camera vector * -1.0\r\n//input float\t a\t\t\t\t\t: Could be step size\r\n//input float\t R\t\t\t\t\t: Outer radius of the cylinder/pin\r\n//input int\t   MaxOverlapSteps\t  : Maximum iterations required\r\n//input float\t Depth\t\t\t\t: Depth of the underside\r\n//input float4\tBorder\t\t\t   : Arbitrary UV borders for removing pins out of border\r\n//input float\t Inner_R\t\t\t  : Inner radius of the cylinder/pin\r\n//input float2\tWallThickness\t\t: Adjusting the box separately\r\n\r\nfloat \tR_sq = R * R;\r\nfloat \tInner_R_sq = Inner_R * Inner_R;\r\nfloat2\tsigns = sign(RD.xy);\r\nBorderMask = 0;\r\n\r\n//first intersect the outer bounding box\r\nfloat2 \touterbox_first =   (float2(Border.x - WallThickness.x, Border.z - WallThickness.y) - RawUV) / RD;\r\nfloat2 \touterbox_second =  (float2(Border.y + WallThickness.x, Border.w + WallThickness.y) - RawUV) / RD;\r\nfloat2 \touterbox_farplanes = max(outerbox_first, outerbox_second);\r\n\r\nif(RawUV.x < Border.x - WallThickness.x || RawUV.x > Border.y + WallThickness.x || RawUV.y < Border.z - WallThickness.y || RawUV.y > Border.w + WallThickness.y)\r\n{\r\n\tBorderMask = 1;\r\n\treturn float4(0,0,1,0);\r\n}\r\n\r\n\r\n//Main intersection loop\r\nfor(int i = 0; i < MaxOverlapSteps; i++)\r\n{\t\r\n\t//quadratic b term is shared between different branches\r\n\tfloat b = 2*(dot(RO.xy, RD.xy));\r\n\t\r\n\t//intersect a box around each peg\r\n\tfloat3 firstplanes =   (float3(-0.5,-0.5, -Depth) - RO) / RD;\r\n\tfloat3 secondplanes =  (float3( 0.5, 0.5,      0) - RO) / RD;\r\n\tfloat3 farplanes = max(firstplanes, secondplanes);\r\n\t\r\n\t//Is the Ray within a cell with a Cylinder?\r\n\tif(RawUV.x > Border.x &&&& RawUV.x < Border.y &&&& RawUV.y > Border.z &&&& RawUV.y < Border.w)\r\n\t{\r\n\t\t//Is the ray starting within the top cap?\r\n\t\tif(dot(RO.xy,RO.xy) <= R_sq)\r\n\t\t{\r\n\t\t\t//Is the ray starting within the inner, negative cylinder\r\n\t\t\tif(dot(RO.xy,RO.xy) <= Inner_R_sq)\r\n\t\t\t{\r\n\t\t\t\t//quadratic equation, b term is shared\r\n\t\t\t\tfloat c = dot(RO.xy,RO.xy) - Inner_R_sq;\r\n\t\t\t\tfloat disc = b*b - (2.0 * a * c);\r\n\t\t\t\tfloat t1 = (-b + sqrt(disc)) / (a);\r\n\t\t\t\t\r\n\t\t\t\tif(t1>0)\r\n\t\t\t\t{\r\n\t\t\t\t\tfloat zgrad = max(0.05, saturate(1 -t1/Depth*abs(RD.z)));\r\n\t\t\t\t\tfloat3 pegnormal = float3(-normalize(RO.xy + RD.xy*t1),0);\r\n\t\t\t\t\t//pegnormal = lerp(float3(0,0,1), pegnormal, saturate((t1/Depth*abs(RD.z) * 10)));\r\n\r\n\t\t\t\t\tif (zgrad < min(1 - Depth, Depth))\r\n\t\t\t\t\t{   \r\n\t\t\t\t\t\treturn float4(0, 0, 1, -BottomAO * 1.5);\r\n\t\t\t\t\t}\r\n\t\t\t\t\t\r\n\t\t\t\t\treturn float4(pegnormal,zgrad);\r\n\t\t\t\t}\r\n\t\t\t}\r\n\t\t\t//Generate cylinderical normals\r\n\t\t\tfloat3 normal = normalize(float3(RO.xy,0.0));\r\n\t\t\t\r\n\t\t\t//Generate a round, outer edge mask via pythagorean theorum\r\n\t\t\tfloat dist = -(sqrt(dot(RO.xy,RO.xy)) - R);\r\n\t\t\tdist = 1.0-saturate(dist*50);\r\n\t\t\tdist = sqrt(1.0-(dist*dist));\r\n\t\t\t\r\n\t\t\tnormal = lerp(normal, float3(0,0,1), dist);\r\n\t\t\t\r\n\t\t\t//Inner edge mask\r\n\t\t\tdist = (sqrt(dot(RO.xy,RO.xy)) - Inner_R);\r\n\t\t\tdist = 1.0-saturate(dist*50);\r\n\t\t\tdist = sqrt(1.0-(dist*dist));\r\n\t\t\t\r\n\t\t\t//Blend cylinder normals with flat cap\r\n\t\t\tnormal = lerp(normalize(float3(RO.xy,0.0))*float3(-1,-1,1),normal, dist);\r\n\t\t\treturn float4(normal,1);\r\n\t\t}\r\n\t\r\n\t\t//quadratic equation, b term is shared\r\n\t\tfloat c = dot(RO.xy,RO.xy) - R_sq;\r\n\t\tfloat disc = b*b - (2.0 * a * c);\r\n\t\tfloat t0 = (-b - sqrt(disc)) / (a);\r\n\t\r\n\t\t//Check if the ray hits the cylinder and if that is before the floor\r\n\t\tif(t0>0 &&&& t0<farplanes.z)\r\n\t\t{\r\n\t\t\tfloat zgrad = saturate(1 - t0/Depth*abs(RD.z));\r\n\t\t\tfloat3 pegnormal = float3(normalize(RO.xy + RD.xy*t0),0);\r\n\t\t\t//pegnormal = lerp(float3(0,0,1), pegnormal, saturate((t0/Depth*abs(RD.z) * 10)));\r\n\t\t\tpegnormal = lerp(float3(0,0,1), pegnormal, saturate(zgrad*10));\r\n\t\t\t\r\n\t\t\tzgrad = saturate((zgrad -0.1) / 0.9);\r\n\t\t\t\r\n\t\t\treturn float4(pegnormal, zgrad);\r\n\t\t}\r\n\t}\r\n\t\r\n\t//If the bottom plane is hit, return bottom AO, also check for ou\r\n\tif(farplanes.z < farplanes.x &&&& farplanes.z < farplanes.y || i==(MaxOverlapSteps-1))\r\n\t{\r\n\t\r\n\t\t//check against the outer bounding box\r\n\t\tif(outerbox_farplanes.x < farplanes.z || outerbox_farplanes.y < farplanes.z)\r\n\t\t{\r\n\t\t\tfloat zgrad = saturate(1 - min(outerbox_farplanes.x,outerbox_farplanes.y)/Depth*abs(RD.z));\r\n\t\t\tfloat3 sidenormal =  outerbox_farplanes.y > outerbox_farplanes.x ? float3(-signs.x,0,0) : float3(0,-signs.y,0);\r\n\t\t\tsidenormal = lerp(float3(0,0,1), sidenormal, saturate(zgrad*40));\r\n\t\t\tsidenormal = lerp(float3(0,0,1), sidenormal, saturate((1-zgrad)*40));\r\n\t\t\t\r\n\t\t\treturn float4(sidenormal, zgrad);\r\n\t\t}\r\n\t\r\n\t\r\n\t\tRO += RD * farplanes.z;\r\n\t\tRO = frac(RO+0.5)-0.5;\r\n\t\tfloat dist = max(0,sqrt(dot(RO.xy,RO.xy)) - R);\r\n\t\tfloat botAO = 1.0-exp(-dist * BottomAO);\r\n\t\treturn float4(0,0,1, -BottomAO + botAO);\r\n\t}\r\n\r\n\t//After each iteration, we offset the UVs by one grid cell in the direction of the intersected box face\r\n\tRO-= farplanes.y > farplanes.x ? float3(signs.x,0,0) : float3(0,signs.y,0);\r\n\tRawUV+= farplanes.y > farplanes.x ? float3(signs.x,0,0) : float3(0,signs.y,0);\r\n\r\n}\r\n\r\nreturn 0;"
   description=Lego Undersides Analytic
   outputType=CMOT_Float4
   additionalOutputs [{outputName=BorderMask, outputType=CMOT_Float1}]
   inputs [
    {inputName=RawUV, input {expression=&self:MaterialExpressionNamedRerouteUsage_350}}
    {inputName=RO, input {expression=&self:MaterialExpressionNamedRerouteUsage_351}}
    {inputName=RD, input {expression=&self:MaterialExpressionNamedRerouteUsage_352}}
    {inputName=a, input {expression=&self:MaterialExpressionNamedRerouteUsage_353}}
    {inputName=R, input {expression=&self:MaterialExpressionNamedRerouteUsage_354}}
    {inputName=MaxOverlapSteps, input {expression=&self:MaterialExpressionNamedRerouteUsage_355}}
    {inputName=Depth, input {expression=&self:MaterialExpressionNamedRerouteUsage_356}}
    {inputName=Border, input {expression=&self:MaterialExpressionNamedRerouteUsage_357}}
    {inputName=Inner_R, input {expression=&self:MaterialExpressionNamedRerouteUsage_358}}
    {inputName=WallThickness, input {expression=&self:MaterialExpressionNamedRerouteUsage_359}}
    {inputName=BottomAO, input {expression=&self:MaterialExpressionConstant_22}}
   ]
  }
  NamedRerouteDeclaration(None=ComponentMask(None=@n63.return) {a=true}) {name=Top Shape Mask, nodeColor=<rgba 0 1 0.9297 1>}
  n64 = ComponentMask(None=@n63.return) {b=true, g=true, r=true}
  n65 = Multiply(A=@n64, B=Constant3Vector() {constant=<rgba 1 -1 1 0>}) {desc=Correct Y}
  n66 = NamedRerouteUsage()
  NamedRerouteDeclaration(None=MaterialFunctionCall!(Normal2=LinearInterpolate(A=@n64, B=@n65, Alpha=@n66), Normal1=NamedRerouteUsage()) {materialFunction=&/MF_WhiteoutBlendNormals}.Result) {name=Normal Analytic Out, nodeColor=<rgba 1 0.8906 0 1>}
  n67 = Add(A=ComponentMask(None=NamedRerouteUsage()) {b=true}) {constB=0.15}
  NamedRerouteDeclaration(None=Multiply(A=Saturate(None=@n67), B=NamedRerouteUsage())) {name=Specular Anlaytic Out, nodeColor=<rgba 1 0.9844 0 1>}
  NamedRerouteDeclaration(None=@n2.Normal) {name=Normal In, nodeColor=<rgba 0 1 0.8828 1>}
  n68 = NamedRerouteUsage()
  n69 = Multiply(A=@n68) {constB=1.35}
  n70 = NamedRerouteUsage()
  NamedRerouteDeclaration(None=Saturate(None=LinearInterpolate(A=@n68, B=@n69, Alpha=@n70))) {name=Base Color Analytic Out, nodeColor=<rgba 0 1 0.0391 1>}
  NamedRerouteDeclaration(None=StaticBoolParameter() {group=Undersides, parameterName=Use Analytic Setup, sortPriority=0}) {name=Use Analytic Setup, nodeColor=<rgba 1 0 0.5625 1>}
  NamedRerouteDeclaration(None=StaticSwitch(True=NamedRerouteUsage(), False=NamedRerouteUsage(), Value=NamedRerouteUsage())) {name=Base Color Out, nodeColor=<rgba 0.8516 1 0 1>}
  NamedRerouteDeclaration(None=StaticSwitch(True=NamedRerouteUsage(), False=NamedRerouteUsage(), Value=NamedRerouteUsage())) {name=Specular Out, nodeColor=<rgba 1 0 0.0469 1>}
  NamedRerouteDeclaration(None=StaticSwitch(True=NamedRerouteUsage(), False=NamedRerouteUsage(), Value=NamedRerouteUsage())) {name=Normal Out, nodeColor=<rgba 0.1719 1 0 1>}
  NamedRerouteDeclaration(None=@n2.Roughness) {name=Roughness In, nodeColor=<rgba 0 1 0.2969 1>}
  NamedRerouteDeclaration(None=@n2.EmissiveColor) {name=Emissive In, nodeColor=<rgba 0 1 0.6484 1>}
  NamedRerouteDeclaration(None=@n2.SubsurfaceColor) {name=Subsurface In, nodeColor=<rgba 0 1 0.4141 1>}
  n71 = NamedRerouteUsage()
  n72 = ComponentMask(None=NamedRerouteUsage()) {b=true}
  NamedRerouteDeclaration(None=LinearInterpolate(A=Multiply(A=@n71) {constB=0.01}, B=@n71, Alpha=Round(None=@n72))) {name=Emissive Analytic Out, nodeColor=<rgba 0.1484 0 1 1>}
  n73 = NamedRerouteUsage()
  n74 = Multiply(A=@n73) {constB=0.25}
  n75 = Multiply(A=@n73) {constB=0.085}
  n76 = NamedRerouteUsage()
  NamedRerouteDeclaration(None=LinearInterpolate(A=LinearInterpolate(A=@n74, B=@n75, Alpha=@n76), B=NamedRerouteUsage(), Alpha=NamedRerouteUsage())) {name=Emissive Flat Out, nodeColor=<rgba 0.8281 0 1 1>}
  n77 = NamedRerouteUsage()
  n78 = LinearInterpolate(B=@n77, Alpha=NamedRerouteUsage()) {constA=0.75, constB=0.75}
  NamedRerouteDeclaration(None=LinearInterpolate(A=@n77, B=Saturate(None=@n78), Alpha=NamedRerouteUsage())) {name=Roughness Flat Out, nodeColor=<rgba 0 0.0625 1 1>}
  n79 = NamedRerouteUsage()
  n80 = Multiply(A=@n79) {constB=0.75}
  n81 = Multiply(A=@n79) {constB=0.5}
  n82 = NamedRerouteUsage()
  NamedRerouteDeclaration(None=LinearInterpolate(A=LinearInterpolate(A=@n80, B=@n81, Alpha=@n82), B=NamedRerouteUsage(), Alpha=NamedRerouteUsage())) {name=Subsurface Flat Out, nodeColor=<rgba 0.4297 0 1 1>}
  NamedRerouteDeclaration(None=StaticSwitch(True=NamedRerouteUsage(), False=NamedRerouteUsage(), Value=NamedRerouteUsage())) {name=Roughness Out, nodeColor=<rgba 1 0 0.5156 1>}
  n83 = NamedRerouteUsage()
  n84 = StaticSwitchParameter(True=@n83, False=NamedRerouteUsage()) {parameterName=Glitter}
  NamedRerouteDeclaration(None=StaticSwitch(True=NamedRerouteUsage(), False=StaticSwitchParameter(True=@n83, False=@n84) {parameterName=Opalescence}, Value=NamedRerouteUsage())) {name=Subsurface Out, nodeColor=<rgba 1 0 0.5391 1>}
  n85 = StaticSwitch(True=NamedRerouteUsage(), False=NamedRerouteUsage(), Value=NamedRerouteUsage())
  n86 = NamedRerouteUsage()
  NamedRerouteDeclaration(None=StaticSwitchParameter(True=@n85, False=StaticSwitchParameter(True=@n85, False=@n86) {parameterName=Glitter}) {parameterName=Opalescence}) {name=Emissive Out, nodeColor=<rgba 0.2656 1 0 1>}
  NamedRerouteDeclaration(None=@n63.BorderMask) {name=Analytic Border Mask, nodeColor=<rgba 1 0.7031 0 1>}
  n87 = NamedRerouteUsage()
  n88 = NamedRerouteUsage()
  n89 = NamedRerouteUsage()
  NamedRerouteDeclaration(None=LinearInterpolate(A=Multiply(A=@n87) {constB=0.75}, B=@n87, Alpha=Add(A=@n88, B=@n89))) {name=Subsurface Analytic Out, nodeColor=<rgba 0 0.2031 1 1>}
  MaterialExpressionNamedRerouteUsage_15 = NamedRerouteUsage()
  MaterialExpressionNamedRerouteUsage_399 = NamedRerouteUsage()
  MaterialExpressionNamedRerouteUsage_400 = NamedRerouteUsage()
  n90 = Custom!(
   RawUV=@MaterialExpressionNamedRerouteUsage_15
   Border=@MaterialExpressionNamedRerouteUsage_399
   WallThickness=@MaterialExpressionNamedRerouteUsage_400
  ) {
   code="float BottomBorder = saturate(step(RawUV.y, (Border.z - WallThickness.y)));\r\n\r\nfloat TopBorder = saturate(1 - step(RawUV.y, (Border.w + WallThickness.y)));\r\n\r\nfloat LeftBorder = saturate(step(RawUV.x, (Border.x - WallThickness.x)));\r\n\r\nfloat RightBorder = saturate(1 - step(RawUV.x, (Border.y + WallThickness.x)));\r\n\r\nreturn saturate(TopBorder + BottomBorder + LeftBorder + RightBorder);"
   description=Brick Borders
   outputType=CMOT_Float1
   inputs [
    {inputName=RawUV, input {expression=&self:MaterialExpressionNamedRerouteUsage_15}}
    {inputName=Border, input {expression=&self:MaterialExpressionNamedRerouteUsage_399}}
    {inputName=WallThickness, input {expression=&self:MaterialExpressionNamedRerouteUsage_400}}
   ]
  }
  NamedRerouteDeclaration(None=@n90) {name=Underside Borders, nodeColor=<rgba 0 0.7891 1 1>}
  NamedRerouteDeclaration(None=TextureSample!(UVs=NamedRerouteUsage(), Tex=NamedRerouteUsage(), "DDX(UVs)"=NamedRerouteUsage(), "DDY(UVs)"=NamedRerouteUsage(), Apply View MipBias) {mipValueMode=TMVM_Derivative, samplerSource=SSM_Wrap_WorldGroupSettings, samplerType=SAMPLERTYPE_Masks}.R) {name="Shape Mask (Tex)", nodeColor=<rgba 0.5703 1 0 1>}
  NamedRerouteDeclaration(None=Max(A=NamedRerouteUsage(), B=NamedRerouteUsage())) {name=Larger Than 1x1, nodeColor=<rgba 0 0.1094 1 1>}
  n91 = Transform(None=Multiply(A=CameraVectorWS()) {constB=-1}) {transformSourceType=TRANSFORMSOURCE_World, transformType=TRANSFORM_Tangent}
  n92 = Divide(A=ComponentMask(None=@n91) {g=true, r=true}, B=Multiply(A=ComponentMask(None=@n91) {b=true}) {constA=1, constB=5})
  n93 = Constant2Vector() {g=1, r=-1}
  NamedRerouteDeclaration(None=Clamp(None=Multiply(A=@n92, B=@n93)) {maxDefault=0.2, minDefault=-0.2}) {name=Parallax Offset, nodeColor=<rgba 1 0.8438 0 1>}
  MaterialExpressionAdd_47 = Add(A=@MaterialExpressionNamedRerouteUsage_15, B=NamedRerouteUsage())
  n94 = Custom(RawUV=@MaterialExpressionAdd_47, Border=@MaterialExpressionNamedRerouteUsage_399, WallThickness=@MaterialExpressionNamedRerouteUsage_400) {
   code="float BottomBorder = saturate(step(RawUV.y, (Border.z - WallThickness.y)));\r\n\r\nfloat TopBorder = saturate(1 - step(RawUV.y, (Border.w + WallThickness.y)));\r\n\r\nfloat LeftBorder = saturate(step(RawUV.x, (Border.x - WallThickness.x)));\r\n\r\nfloat RightBorder = saturate(1 - step(RawUV.x, (Border.y + WallThickness.x)));\r\n\r\nreturn saturate(TopBorder + BottomBorder + LeftBorder + RightBorder);"
   description=Brick Borders
   outputType=CMOT_Float1
   inputs [
    {inputName=RawUV, input {expression=&self:MaterialExpressionAdd_47}}
    {inputName=Border, input {expression=&self:MaterialExpressionNamedRerouteUsage_399}}
    {inputName=WallThickness, input {expression=&self:MaterialExpressionNamedRerouteUsage_400}}
   ]
  }
  NamedRerouteDeclaration(None=@n94) {name=Borders Parallax, nodeColor=<rgba 0.2656 1 0 1>}
  n95 = Subtract(A=NamedRerouteUsage(), B=NamedRerouteUsage())
  n96 = Multiply(A=NamedRerouteUsage(), B=Subtract(A=NamedRerouteUsage(), B=NamedRerouteUsage()))
  NamedRerouteDeclaration(None=Saturate(None=Add(A=@n95, B=@n96))) {name=Parallax Mask, nodeColor=<rgba 1 0 0.375 1>}
  n97 = Multiply(A=WorldPosition().XY) {constB=0.01}
  NamedRerouteDeclaration(None=DDX(Value=@n97)) {name=DDX Cells, nodeColor=<rgba 0 1 0.4375 1>}
  NamedRerouteDeclaration(None=DDY(Value=@n97)) {name=DDY Cells, nodeColor=<rgba 1 0 0.75 1>}
  n98 = NamedRerouteUsage()
  n99 = NamedRerouteUsage()
  n100 = Add(A=Add(A=NamedRerouteUsage(), B=NamedRerouteUsage()), B=Multiply(A=Add(A=@n98, B=@n99), B=NamedRerouteUsage()))
  n101 = OneMinus(None=NamedRerouteUsage())
  NamedRerouteDeclaration(None=Add(A=Saturate(None=@n100), B=Max(A=@n101) {constB=0.9}) {constB=0.8}) {name=Bottom Plane AO, nodeColor=<rgba 0.875 1 0 1>}
  n102 = ComponentMask(None=SkyAtmosphereLightDirection()) {g=true, r=true}
  n103 = ComponentMask(None=ReflectionVectorWS()) {g=true, r=true}
  n104 = NamedRerouteDeclaration(None=Multiply(A=DotProduct(A=@n102, B=@n103)) {constB=-1}) {name=Sun Dir Mult -1 to 1, nodeColor=<rgba 0.7812 0 1 1>}
  NamedRerouteDeclaration(None=Add(A=Multiply(A=@n104) {constB=0.5}) {constB=0.5}) {name=Sun Dir Mult 0 to 1, nodeColor=<rgba 0 0.6484 1 1>}
  n105 = VertexNormalWS()
  n106 = Multiply(A=@n105) {constB=-1}
  NamedRerouteDeclaration(None=Saturate(None=ComponentMask(None=@n106) {b=true})) {name=Underside Facing Down, nodeColor=<rgba 0.0078 0 1 1>}
  n107 = SkyAtmosphereLightDirection()
  NamedRerouteDeclaration(None=ConstantBiasScale(None=DotProduct(A=@n107, B=@n105)) {desc=Facing the sun? 1 : 0}) {name=Underside Facing Sun, nodeColor=<rgba 1 0 0.7734 1>}
  NamedRerouteDeclaration(None=TextureObject() {samplerType=SAMPLERTYPE_Masks, texture=/AtomFramework/Materials/Textures/FlatUndersides_Shape_Mask}) {name=Tex Object, nodeColor=<rgba 0 1 0.4375 1>}
  n108 = NamedRerouteUsage()
  n109 = Multiply(A=NamedRerouteUsage()) {constB=0.2}
  NamedRerouteDeclaration(None=TextureSample(UVs=Add(A=@n108, B=@n109), Tex=NamedRerouteUsage(), "DDX(UVs)"=NamedRerouteUsage(), "DDY(UVs)"=NamedRerouteUsage()) {mipValueMode=TMVM_Derivative, samplerSource=SSM_Wrap_WorldGroupSettings, samplerType=SAMPLERTYPE_Masks}.R) {name="Shape Parallax Mask (Tex)", nodeColor=<rgba 0.2891 0 1 1>}
 ]}
